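# Material function exposed to the library under Misc and Fortnite that supplies a white-to-black distance fade for the light function.
# It measures the distance from the camera to the world position and divides it by the SSSDistancefade scalar parameter.
# The clamped, inverted result is white at the camera and fades linearly to black at that parameter's distance.
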MaterialFunction /Game/Environments/World/LightFunctions/ForniteLightFunctionFalloff
graph {
 properties {bExposeToLibrary=true, description=White to black fade that the light function uses, libraryCategoriesText [Misc,Fortnite]}
 nodes [
  n1 = Divide(A=Distance(A=CameraPositionWS(), B=WorldPosition().XYZ), B=ScalarParameter() {defaultValue=4096, parameterName=SSSDistancefade})
  FunctionOutput(None=OneMinus(None=Clamp(None=@n1)))
 ]
}
