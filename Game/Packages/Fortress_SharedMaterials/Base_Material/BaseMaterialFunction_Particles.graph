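# Base lit particle material, usable on mesh particles, with static switches for SubUV, normals, jiggle WPO, SSS and near-camera fade.
# Diffuse can be tinted from a world-position-indexed color chart and hue-shifted, then is multiplied by particle color.
# Emissive comes from color, diffuse or a dynamic parameter, with optional masking; particle alpha feeds opacity.
# Attributes pass through MF_CreativeMoveTool; destruction and deformation Code parameters are declared but unconnected.
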Material /Game/Packages/Fortress_SharedMaterials/Base_Material/BaseMaterialFunction_Particles
graph {
 properties {
  bUseMaterialAttributes=true
  bUsedWithMeshParticles=true
  opacityMaskClipValue=0.16
  pixelDepthOffsetMode=PDOM_Legacy
  translucencyLightingMode=TLM_VolumetricPerVertexNonDirectional
  usedShadingModels=Default Lit
 }
 nodes [
  n1 = ParticleColor()
  n2 = VectorParameter() {group=World Space Color Picker, parameterName=World Space Color Picker Color Override, defaultValue=<rgba 1 1 1 1>}
  n3 = ComponentMask(None=TextureCoordinate() {coordinateIndex=1}) {g=true}
  n4 = ScalarParameter() {
   defaultValue=0.1
   desc="1/number of colors. ie 1/10 = .1"
   group=World Space Color Picker
   parameterName="Number Of Colors in Color Chart (1/Number of colors)"
  }
  n5 = Divide(
   A=MaterialFunctionCall!(f2, f3=ParticlePositionWS(), f4) {materialFunction=/Game/Packages/Fortnite_UniqueMaterials_1/VectorComponentOps}.f3 Add
   B=ScalarParameter() {defaultValue=16.71, group=World Space Color Picker, parameterName="Color Change Frequency (WS)"}
  )
  n6 = StaticSwitchParameter(True=Add(A=Multiply(A=@n3, B=@n4), B=@n5), False=@n5) {group=World Space Color Picker, parameterName=Offset Color Index with UV Channel 2 G}
  n7 = TextureSampleParameterSubUV!(UVs, Apply View MipBias) {group=Diffuse, parameterName=Diffuse_SubUV, texture=/Engine/EngineMaterials/DefaultDiffuse}
  n8 = TextureSampleParameter2D() {group=Diffuse, parameterName=Diffuse, texture=&DefaultDiffuse}
  n9 = ComponentMask(None=StaticSwitchParameter(True=@n7.RGB, False=@n8.RGB) {group=Diffuse, parameterName=Use SubUV}) {b=true, g=true, r=true}
  n10 = Multiply(
   A=StaticSwitchParameter(
    True=@n2.RGB
    False=TextureSampleParameter2D(UVs=@n6) {group=World Space Color Picker, parameterName=Color Chart, texture=/Game/Global/Textures/Default/ColorSets/CarColors_Interiors}.RGB
   ) {group=World Space Color Picker, parameterName=Override World Space Color Picker Color}
   B=@n9
  )
  n11 = TextureSampleParameter2D() {
   group=World Space Color Picker
   parameterName=World Space Color Picker Mask
   texture=/Game/Vehicles/FORT_Vehicles_01/Materials/M_FORT_Truck_01_M_New
  }
  n12 = StaticComponentMaskParameter(None=AppendVector(A=@n11.RGB, B=@n11.RGB)) {defaultB=true, group=World Space Color Picker, parameterName=World Space Color Picker Mask Channel}
  n13 = StaticSwitchParameter(
   True=StaticSwitchParameter(True=LinearInterpolate(A=@n10, B=@n9, Alpha=@n12), False=@n10) {group=World Space Color Picker, parameterName=Mask World Position Color Picker}
   False=@n9
  ) {group=World Space Color Picker, parameterName=Use World Position Coloring}
  n14 = Multiply(A=ScalarParameter() {defaultValue=1, group=Diffuse, parameterName=Diff_Mult}, B=@n13)
  n15 = StaticSwitchParameter(
   True=MaterialFunctionCall(Hue Shift Percentage=ScalarParameter() {group=Diffuse, parameterName=Hue Shift Percentage}, Texture=@n14) {materialFunction=&HueShift}.Result
   False=@n14
  ) {group=Diffuse, parameterName=HueShift Diffuse}
  n16 = MaterialFunctionCall(Float3=TextureSampleParameter2D() {parameterName=SpecMasks, samplerType=SAMPLERTYPE_Masks, texture=&T_Fortnite_Default_S}.RGB) {materialFunction=&BreakOutFloat3Components}
  n17 = ScalarParameter() {group=Specular, parameterName=SpecRoughnessMin}
  n18 = ScalarParameter() {defaultValue=1, group=Specular, parameterName=SpecRoughnessMax}
  n19 = StaticSwitchParameter(
   True=MaterialFunctionCall(Hue Shift Percentage=ScalarParameter() {parameterName=Hue Shift Percentage}, Texture=@n13) {materialFunction=&HueShift}.Result
   False=@n13
  ) {parameterName=HueShift Diffuse}
  n20 = ScalarParameter() {defaultValue=1, group=Emissive, parameterName=Emissive_Boost}
  n21 = StaticSwitchParameter(
   True=DynamicParameter() {paramNames [Emissive,Param2,Param3,Param4]}.Emissive
   False=StaticSwitchParameter(True=Multiply(A=@n19, B=@n20), False=VectorParameter() {parameterName=Emissive_Color}.RGB) {group=Emissive, parameterName=Use Diffuse in Emissive}
  ) {group=Emissive, parameterName=Use Dynamic Param Emissive}
  n22 = TextureSampleParameter2D() {parameterName=Emissive Mask Texture, texture=&DefaultDiffuse}
  n23 = StaticComponentMaskParameter(None=AppendVector(A=@n22.RGB, B=@n22.RGB)) {defaultR=true, parameterName=Emissive Mask Texture Channel}
  n24 = Multiply(A=@n1.RGB, B=StaticSwitchParameter(True=Multiply(A=@n21, B=@n23), False=@n21) {group=Emissive, parameterName=Use Emissive Mask Texture})
  n25 = ScalarParameter() {defaultValue=0.25, group=SSS, parameterName=Amount Of SubSurface Scattering}
  n26 = OneMinus(None=@n25)
  n27 = StaticSwitchParameter(
   True=StaticSwitchParameter(True=Multiply(A=@n15, B=@n26), False=@n26) {defaultValue=true, group=SSS, parameterName=Reference Diffuse In SSS}
   False=Constant() {r=1}
  ) {group=SSS, parameterName=Has SSS}
  n28 = ParticleColor()
  n29 = StaticSwitchParameter(
   True=StaticSwitchParameter(True=@n7.A, False=@n8.A) {group=Diffuse, parameterName=Use SubUV}
   False=TextureSampleParameter2D() {parameterName=Opacity_Mask_Texture}.R
  ) {group=Diffuse, parameterName=Use Diffuse_Alpha as Masked}
  n30 = Multiply(A=@n28.A, B=@n29)
  n31 = Multiply(
   A=MaterialFunctionCall(Distance=ScalarParameter() {parameterName=Near Camera Fade Distance}) {materialFunction=&Near_Camera_Fade}.Result
   B=@n30
  )
  n32 = Multiply(A=@n28.A, B=@n29)
  n33 = TextureSampleParameter2D() {group=Normal, parameterName=Normals, samplerType=SAMPLERTYPE_Normal, texture=/Engine/EngineMaterials/DefaultNormal}
  n34 = Constant3Vector() {constant=<rgba 0 0 1 0>}
  n35 = MaterialFunctionCall!(
   Speed=ScalarParameter() {defaultValue=0.5, group=WorldPositionOffset, parameterName=WPO_Speed}
   LargeWaveScle=ScalarParameter() {defaultValue=1, group=WorldPositionOffset, parameterName=WPO_LargeWaveScale}
   WaveScale=ScalarParameter() {defaultValue=0.5, group=WorldPositionOffset, parameterName=WPO_WaveScale}
  ) {materialFunction=/Game/Effects/Fort_Effects/Materials/_Functions/Jiggle_WPO_VertColor_Red_3Vec}
  n36 = Constant()
  n37 = VectorParameter() {group=SSS, parameterName=SSS Color, defaultValue=<rgba 0.489 1 0.05 1>}
  n38 = Multiply(A=ComponentMask(None=@n15) {r=true}, B=@n37.RGB)
  n39 = Multiply(
   A=StaticSwitchParameter(True=Multiply(A=@n25, B=@n38), False=Multiply(A=@n25, B=@n37.RGB)) {defaultValue=true, group=SSS, parameterName=Reference Diffuse In SSS}
   B=MaterialFunctionCall() {materialFunction=/Game/Environments/World/LightFunctions/ForniteLightFunctionFalloff}.Result
  )
  n40 = Constant()
  MaterialFunctionCall(MA In=MakeMaterialAttributes!(
    BaseColor=Multiply(A=@n1.RGB, B=@n15)
    Metallic=@n16.G
    Specular=@n16.G
    Roughness=LinearInterpolate(A=@n17, B=@n18, Alpha=@n16.B)
    Anisotropy
    EmissiveColor=Multiply(A=@n24, B=@n27)
    Opacity=StaticSwitchParameter(True=@n31, False=@n30) {parameterName=Use Near Camera Fade}
    OpacityMask=StaticSwitchParameter(True=@n32, False=@n29) {parameterName=Use Opacity Masked Particle Color}
    Normal=StaticSwitchParameter(True=@n33.RGB, False=@n34) {group=Normal, parameterName=Normal_Switch}
    Tangent
    WorldPositionOffset=StaticSwitchParameter(True=@n35.Result, False=@n36) {group=WorldPositionOffset, parameterName=Use JIggle WPO}
    SubsurfaceColor=StaticSwitchParameter(True=@n39, False=@n40) {group=SSS, parameterName=Has SSS}
    ClearCoat
    ClearCoatRoughness
    AmbientOcclusion
    Refraction
    CustomizedUVs_0
    CustomizedUVs_1
    CustomizedUVs_2
    CustomizedUVs_3
    CustomizedUVs_4
    CustomizedUVs_5
    CustomizedUVs_6
    CustomizedUVs_7
    PixelDepthOffset
    ShadingModel
    Displacement
   )) {materialFunction=&MF_CreativeMoveTool}
  ScalarParameter() {group=Code, parameterName=Code_DestructionAmount}
  ScalarParameter() {group=Code, parameterName=WSDamageRadius}
  ScalarParameter() {group=Code, parameterName=Code_HitAnimation}
  ScalarParameter() {group=WorldPositionOffset, parameterName=DistanceCullFadeAnimationOffset}
  VectorParameter() {group=Code, parameterName=Code_DeformationVector, defaultValue=<rgba 0 0 1 0>}
  VectorParameter() {group=Code, parameterName=WSDeformationCenter}
 ]
}
