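# Fortnite material function that swings mesh parts around a supplied pivot in the wind, like a ball-and-socket joint, producing World Position Offset and rotated normals.
# Time-panned turbulence texture samples, offset by pivot and X-axis extent, perturb wind direction and set gust angle.
# The rotation axis crosses wind with either the part's X axis or the pivot-to-vertex direction (wrap-around).
# A hierarchy depth mask gates results, with optional vertex-alpha modulation and shear.
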MaterialFunction /Game/Packages/Fortress_SharedMaterials/Base_Material_Functions/Fornite_BallAndSocketConstraint
graph {
 properties {bExposeToLibrary=true, libraryCategoriesText [Fortnite]}
 nodes [
  n1 = FunctionInput() {bUsePreviewValueAsDefault=true, inputName=Model Depth, inputType=FunctionInput_Scalar, sortPriority=99, previewValue=<1 0 0 1>}
  n2 = Clamp(None=Subtract(
    A=Add(A=@n1)
    B=FunctionInput() {bUsePreviewValueAsDefault=true, inputName=Effect's Target Depth, inputType=FunctionInput_Scalar, sortPriority=99}
   ) {desc=0 matches level. positive is per pivot. everything else is per vertex})
  n3 = FunctionInput(Preview=WorldPosition() {worldPositionShaderOffset=WPT_ExcludeAllShaderOffsets}.XYZ) {bUsePreviewValueAsDefault=true, inputName=World Position, sortPriority=50}
  n4 = ComponentMask(None=FunctionInput() {inputName=Position, sortPriority=-5}) {b=true, g=true, r=true}
  n5 = Normalize(VectorInput=Subtract(A=@n3, B=@n4))
  n6 = FunctionInput() {
   bUsePreviewValueAsDefault=true
   inputName=Random Rotation Influence
   inputType=FunctionInput_Scalar
   sortPriority=12
   previewValue=<0.25 0 0 1>
  }
  n7 = Multiply(A=Time() {bIgnorePause=true}, B=FunctionInput() {inputName=TimeRate_CodeReferenced1, inputType=FunctionInput_Scalar, sortPriority=200})
  n8 = FunctionInput() {inputName=TimeOffset_CodeControlled1, inputType=FunctionInput_Scalar, sortPriority=201}
  n9 = FunctionInput() {inputName=X-Axis Texture And X Extent Alpha, inputType=FunctionInput_Vector4, sortPriority=-4}
  n10 = ComponentMask(None=@n9) {b=true, g=true, r=true}
  n11 = Multiply(A=@n10, B=ComponentMask(None=@n9) {a=true})
  n12 = Add(A=Multiply(A=Add(A=@n7, B=@n8), B=Constant2Vector() {r=1}), B=ComponentMask(None=Add(A=@n4, B=@n11)) {g=true, r=true})
  n13 = FunctionInput() {
   bUsePreviewValueAsDefault=true
   inputName=Random Rotation Texture Sample Scale
   inputType=FunctionInput_Scalar
   sortPriority=11
   previewValue=<250 0 0 1>
  }
  n14 = TextureSample(UVs=Frac(None=Divide(A=@n12, B=@n13))) {
   samplerType=SAMPLERTYPE_LinearColor
   texture=/Engine/Functions/Engine_MaterialFunctions02/ExampleContent/Textures/WindTurblenceVectorAndGustMagnitude
  }
  n15 = ConstantBiasScale(None=@n14.RGB) {bias=-0.5, scale=2}
  n16 = FunctionInput() {bUsePreviewValueAsDefault=true, inputName=Wind Direction, sortPriority=1000, previewValue=<1 0 0 1>}
  n17 = Normalize(VectorInput=Add(A=Multiply(A=@n6, B=@n15), B=@n16)) {desc=Wind Vector}
  n18 = Multiply(A=Power(Base=DotProduct(A=@n16, B=@n10)) {constExponent=5}, B=Constant3Vector() {constant=<rgba 0 0.05 -0.2 0>})
  n19 = CrossProduct(A=Normalize(VectorInput=Add(A=@n10, B=@n18)), B=@n17)
  n20 = FunctionInput(Preview=StaticBool()) {bUsePreviewValueAsDefault=true, inputName=UseWrapAroundLogic, inputType=FunctionInput_StaticBool}
  n21 = StaticSwitch(True=Normalize(VectorInput=CrossProduct(A=@n5, B=@n17)), False=Normalize(VectorInput=@n19), Value=@n20) {desc=Axis}
  n22 = FunctionInput() {bUsePreviewValueAsDefault=true, inputName=Wind Gust Angle Rotation, inputType=FunctionInput_Scalar, previewValue=<0.1 0 0 1>}
  n23 = TextureSample(UVs=Frac(None=@n12)) {
   samplerType=SAMPLERTYPE_LinearColor
   texture=&WindTurblenceVectorAndGustMagnitude
  }
  n24 = FunctionInput() {bUsePreviewValueAsDefault=true, inputName=Wind Gust Offset, inputType=FunctionInput_Scalar, previewValue=<-0.125 0 0 1>}
  n25 = Add(A=@n23.A, B=@n24) {constB=-0.25}
  n26 = StaticSwitch(
   True=Multiply(A=VertexColor().A, B=@n25)
   False=@n25
   Value=FunctionInput(Preview=StaticBool()) {bUsePreviewValueAsDefault=true, inputName=Modulate Rotation Results by the Vertex Alpha, inputType=FunctionInput_StaticBool}
  )
  n27 = Multiply(A=@n22, B=@n26)
  n28 = Abs(None=DotProduct(A=@n10, B=@n17))
  n29 = StaticSwitch(True=Multiply(A=@n27, B=OneMinus(None=@n28)), False=@n27, Value=@n20) {desc=Angle}
  n30 = RotateAboutAxis(NormalizedRotationAxis=@n21, RotationAngle=@n29, PivotPoint=@n4, Position=@n3)
  n31 = FunctionInput() {bUsePreviewValueAsDefault=true, inputName=Max Sheer Offset, sortPriority=1000, previewValue=<5 5 0 1>}
  n32 = Multiply(A=@n19, B=@n31)
  n33 = StaticSwitch(
   True=Add(A=@n30, B=Multiply(A=@n26, B=@n32))
   False=@n30
   Value=FunctionInput(Preview=StaticBool()) {bUsePreviewValueAsDefault=true, inputName=Add Shear, inputType=FunctionInput_StaticBool, sortPriority=1000}
  )
  n34 = FunctionInput(Preview=StaticBool() {value=true}) {
   bUsePreviewValueAsDefault=true
   description="Defaults to true and outputs the correct World Position Offset and Rotated Normal. "
   inputName=Enable
   inputType=FunctionInput_StaticBool
   sortPriority=-10
  }
  FunctionOutput(None=StaticSwitch(True=Multiply(A=@n2, B=@n33), False=Constant(), Value=@n34)) {outputName=Add to WPO}
  n35 = FunctionInput() {bUsePreviewValueAsDefault=true, inputName=Optional - world space pixel normal, previewValue=<0 0 1 1>}
  n36 = LinearInterpolate(
   A=@n35
   B=Add(
    A=@n35
    B=MaterialFunctionCall(Rotation Axis=@n21, Rotation Angle=@n29, World Space Vertex Normals=@n35) {materialFunction=&FixRotateAboutAxisNormals}.New World Space Vertex Normal
   )
   Alpha=@n2
  )
  n37 = FunctionInput(Preview=StaticBool() {value=true}) {
   bUsePreviewValueAsDefault=true
   description="Defaults to true. "
   inputName=Enable Normal Rotation
   inputType=FunctionInput_StaticBool
   sortPriority=-9
  }
  FunctionOutput(None=StaticSwitch(True=StaticSwitch(True=@n36, False=@n35, Value=@n37), False=@n35, Value=@n34)) {outputName=Rotated World Space Pixel Normal, sortPriority=10}
 ]
}
