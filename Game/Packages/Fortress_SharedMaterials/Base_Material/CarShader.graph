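# Fortnite clear-coat master material for cars and similar props, built on material attributes with static switches for optional features.
# Base color combines tiling or triplanar detail, decals and color-chart tints picked by actor world position.
# Code-driven height-lerp damage alters color, specular, normals and clear coat, while two animated emissive layers add glow.
# World position offset adds destruction, morph and wind motion; snow, lava-edge and game-wide functions wrap the result.
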Material /Game/Packages/Fortress_SharedMaterials/Base_Material/CarShader
graph {
 properties {
  bAutomaticallySetUsageInEditor=false
  bUseMaterialAttributes=true
  bUsedWithInstancedStaticMeshes=true
  bUsedWithNanite=true
  ditheredLODTransition=true
  pixelDepthOffsetMode=PDOM_Legacy
  shadingModel=MSM_ClearCoat
  usedShadingModels=Clear Coat
 }
 nodes [
  ActorPositionWS()
  Constant()
  n1 = ScalarParameter() {defaultValue=572.252, group=Custom Tile, parameterName=WorldSpaceBlendSize}
  n2 = ScalarParameter() {defaultValue=4, group=Custom Tile, parameterName=Base Color Pattern Tiling}
  Divide(A=@n1, B=@n2) {constA=256}
  n3 = StaticBool() {value=true}
  n4 = Divide(
   A=MaterialFunctionCall!(World Position) {materialFunction=/Engine/Functions/Engine_MaterialFunctions02/Utility/DepthFromWorldPosition}.Result
   B=ScalarParameter() {defaultValue=5000, group=EmissiveDistanceFade, parameterName=Fade Distance}
  )
  n5 = MaterialFunctionCall(
   Invert Fresnel=StaticBoolParameter() {defaultValue=true, group=Emissive, parameterName=InvertEmissiveFresnel}
   Power=ScalarParameter() {defaultValue=2, group=Emissive, parameterName=EmissveFresnelExponent}
  ) {materialFunction=&Fresnel_Function}
  n6 = VectorParameter() {group=Emissive, parameterName="EmissiveUVPositioning (RG)UpperLeft (BA)LowerRight", defaultValue=<rgba 0 0 1 1>}
  n7 = VectorParameter() {group=World Space Color Picker, parameterName=World Space Color Picker Color Override, defaultValue=<rgba 1 1 1 1>}
  n8 = ComponentMask(None=TextureCoordinate() {coordinateIndex=1}) {g=true}
  n9 = ScalarParameter() {
   defaultValue=0.1
   desc="1/number of colors. ie 1/10 = .1"
   group=World Space Color Picker
   parameterName="Number Of Colors in Color Chart (1/Number of colors)"
  }
  n10 = MaterialFunctionCall!(f2, f3=ActorPositionWS(), f4) {materialFunction=/Game/Packages/Fortnite_UniqueMaterials_1/VectorComponentOps}
  n11 = Divide(A=@n10.f3 Add, B=ScalarParameter() {defaultValue=16.71, group=World Space Color Picker, parameterName="Color Change Frequency (WS)"})
  n12 = StaticSwitchParameter(True=Add(A=Multiply(A=@n8, B=@n9), B=@n11), False=@n11) {group=World Space Color Picker, parameterName=Offset Color Index with UV Channel 2 G}
  n13 = TextureCoordinate() {coordinateIndex=1}
  n14 = TextureSampleParameter2D(UVs=@n13) {group=Decal, parameterName=decal, texture=&T_BlacksRGB}
  n15 = MaterialFunctionCall(Float2=@n13) {materialFunction=&BreakOutFloat2Components}
  n16 = Constant() {r=0.5}
  n17 = Constant() {r=1}
  n18 = Multiply(A=Multiply(A=SphereMask(A=@n15.R, B=@n16, Radius=@n16, Hardness=@n17), B=SphereMask(A=@n15.G, B=@n16, Radius=@n16, Hardness=@n17)), B=@n14.A)
  n19 = TextureCoordinate()
  n20 = VectorParameter() {group=Diffuse, parameterName=Base Color Offset RG Tiling BA, defaultValue=<rgba 0 0 1 1>}
  n21 = ComponentMask(None=@n20.RGB) {g=true, r=true}
  n22 = StaticSwitchParameter(True=Multiply(A=Add(A=@n19, B=@n21), B=AppendVector(A=@n20.B, B=@n20.B)), False=@n19) {group=Diffuse, parameterName=Manipulate Base Color and Spec UVS}
  n23 = TextureSampleParameter2D(UVs=@n22) {desc=Use for exterior walls and floors, group=Diffuse, parameterName=Base Color}
  n24 = TextureObjectParameter() {group=Custom Tile, parameterName=TilingBaseColor}
  n25 = MaterialFunctionCall!(
   TextureObject=@n24
   BlendContrast=ScalarParameter() {defaultValue=6.6987, group=Custom Tile, parameterName=WorldSpaceBlendContrast}
   TextureSize=@n1
  ) {materialFunction=/Game/Packages/Fortress_SharedMaterials/Base_Material_Functions/CheapTriPlanerBlend}
  n26 = TextureCoordinate()
  n27 = TextureCoordinate() {coordinateIndex=1}
  n28 = StaticBoolParameter() {group=Custom Tile, parameterName=Use2ndUVForTiling}
  n29 = TextureSample(UVs=Multiply(A=StaticSwitch(True=@n26, False=@n27, Value=@n28), B=@n2), Tex=@n24)
  n30 = AppendVector(A=@n29.RGB, B=@n29.RGB)
  n31 = MaterialFunctionCall(Float4=StaticSwitch(
    True=QualitySwitch(Default=@n25.Result, Low=@n30, High=@n25.Result, Medium=@n25.Result)
    False=@n30
    Value=StaticBoolParameter() {group=Custom Tile, parameterName=UseTriPlaneBlend}
   )) {materialFunction=&BreakOutFloat4Components}
  n32 = AppendVector(A=AppendVector(A=@n31.R, B=@n31.R), B=@n31.B)
  n33 = Multiply(A=@n32, B=ScalarParameter() {defaultValue=0.4, group=Custom Tile, parameterName=TilingDiffuseStainDarkness})
  n34 = TextureSample(UVs=@n22) {desc=TilingTextureBlendMask, samplerType=SAMPLERTYPE_LinearColor, texture=/Game/Athena/Aircraft/Textures/T_BattleBus_CustomTextureMask}
  n35 = StaticSwitch(
   True=LinearInterpolate(A=@n23.RGB, B=LinearInterpolate(A=@n32, B=@n33, Alpha=@n34.R), Alpha=@n34.G)
   False=@n23.RGB
   Value=StaticBoolParameter() {group=Custom Tile, parameterName=UseTilingBaseColorTexture}
  )
  n36 = ComponentMask(None=StaticSwitchParameter(True=LinearInterpolate(A=@n35, B=@n14.RGB, Alpha=@n18), False=@n35) {group=Decal, parameterName=Apply decal via uv 1}) {b=true, g=true, r=true}
  n37 = Multiply(
   A=StaticSwitchParameter(
    True=@n7.RGB
    False=TextureSampleParameter2D(UVs=@n12) {group=World Space Color Picker, parameterName=Color Chart, texture=/Game/Global/Textures/Default/ColorSets/CarColors_Interiors}.RGB
   ) {group=World Space Color Picker, parameterName=Override World Space Color Picker Color}
   B=@n36
  )
  n38 = TextureSampleParameter2D(UVs=@n11) {group=World Space Color Picker, parameterName=Color Chart 2, texture=/Game/Global/Textures/Default/ColorSets/CarColors}
  n39 = TextureSampleParameter2D() {
   group=World Space Color Picker
   parameterName=World Space Color Picker Mask
   texture=/Game/Vehicles/FORT_Vehicles_01/Materials/M_FORT_Truck_01_M_New
  }
  n40 = AppendVector(A=@n39.RGB, B=@n39.RGB)
  n41 = StaticComponentMaskParameter(None=@n40) {defaultB=true, group=World Space Color Picker, parameterName=World Space Color Picker Mask Channel}
  n42 = LinearInterpolate(
   A=LinearInterpolate(A=@n37, B=Multiply(A=@n38.RGB, B=@n36), Alpha=@n41)
   B=@n36
   Alpha=StaticComponentMaskParameter(None=@n40) {defaultG=true, group=World Space Color Picker, parameterName=World Space Color Picker 2 Mask Channel}
  )
  n43 = LinearInterpolate(A=@n37, B=@n36, Alpha=@n41)
  n44 = StaticSwitchParameter(
   True=StaticSwitchParameter(
    True=StaticSwitchParameter(True=@n42, False=@n43) {group=World Space Color Picker, parameterName=Use 2 World Position Coloring Colors}
    False=@n37
   ) {group=World Space Color Picker, parameterName=Mask World Position Color Picker}
   False=@n36
  ) {group=World Space Color Picker, parameterName=Use World Position Coloring}
  n45 = Multiply(A=@n44, B=VectorParameter() {group=Diffuse, parameterName=Base Color Multiplier, defaultValue=<rgba 1 1 1 1>}.RGB)
  n46 = TextureSampleParameter2D() {
   group=World Space Color Picker
   parameterName=World Space Color Picker Mask
   texture=&M_FORT_Truck_01_M_New
  }
  n47 = LinearInterpolate(A=@n45, B=@n44, Alpha=@n46.G)
  n48 = MaterialFunctionCall!(
   S
   V2
   V3=StaticSwitchParameter(True=StaticSwitchParameter(True=@n47, False=@n45) {group=Diffuse, parameterName=Modulate Base With Mask}, False=@n44) {group=Diffuse, parameterName=Modulate Base Color}
   V4
  ) {desc=Diffuse, materialFunction=/Engine/Functions/Engine_MaterialFunctions02/PassThrough}
  n49 = TextureCoordinate()
  n50 = ScalarParameter() {defaultValue=1.5, group=CustomDamage, parameterName=Custom Damage Texture Tiling}
  n51 = Multiply(
   A=@n48.V3
   B=TextureSampleParameter2D(UVs=Multiply(A=@n49, B=@n50)) {group=CustomDamage, parameterName=Custom Damage Diffuse Texture, texture=/Game/Packages/Fortress_ShaderTests/Textures/charcoal}.RGB
  )
  n52 = VectorParameter() {group=CustomDamage, parameterName=Custom Damage Diffuse Color, defaultValue=<rgba 0.18 0.18 0.18 1>}
  n53 = StaticSwitchParameter(
   True=MaterialFunctionCall(Speed=Constant() {r=0.3333}) {materialFunction=&TimeWithSpeedVariable}.Frac Time
   False=ScalarParameter() {group=Code, parameterName=Code_CustomDamagePercentage}
  ) {group=CustomDamage, parameterName=Debug custom damage animation}
  n54 = Multiply(A=@n53, B=ScalarParameter() {defaultValue=0.4, group=CustomDamage, parameterName=Custom Damage DSSm Height Max}) {constB=0.1}
  n55 = TextureSampleParameter2D() {
   group=CustomDamage
   parameterName=Custom Damage Height Map
   samplerType=SAMPLERTYPE_Masks
   texture=/Game/Packages/Fortress_ShaderTests/Textures/T_CopCar_DamageMask
  }
  n56 = StaticComponentMaskParameter(None=AppendVector(A=@n55.RGB, B=@n55.RGB)) {defaultG=true, group=CustomDamage, parameterName=Custom Damage Height Map Channel}
  n57 = Clamp(None=Multiply(
    A=MaterialFunctionCall(Transition Phase=@n54, Height Texture=@n56) {materialFunction=&HeightLerp}.Lerp Alpha No Contrast
    B=ScalarParameter() {defaultValue=726.3647, group=CustomDamage, parameterName=Custom Damage Contrast}
   ) {constB=10})
  n58 = StaticSwitchParameter(
   True=LinearInterpolate(
    A=@n48.V3
    B=StaticSwitchParameter(True=@n51, False=@n52.RGB) {group=CustomDamage, parameterName=Use Custom Damage Diffuse Texture}
    Alpha=@n57
   ) {constB=0, desc=1 is damaged}
   False=@n48.V3
  ) {defaultValue=true, desc=d, group=CustomDamage, parameterName=Use Custom Damage}
  n59 = MaterialFunctionCall!(
   Manipulate Texture UVs=StaticBoolParameter() {group=Emissive, parameterName=Manipulate Emissive Uvs}
   ActorPositionAdd3=@n10.f3 Add
   Mask Cropped Emissive=StaticBoolParameter() {defaultValue=true, group=Emissive, parameterName="Mask Cropped Emissive "}
   Texture=TextureObjectParameter() {group=Emissive, parameterName=EmissiveTexture}
   Animated emissive black out time=ScalarParameter() {defaultValue=0.5, desc=should be a value from 0-1, group=Emissive, parameterName=Animated Emissive Blackout Ratio}
   Texture Multiplier=VectorParameter() {group=Emissive, parameterName=Emissive Multiplier, defaultValue=<rgba 1 1 1 1>}.RGB
   Modulate Texture=StaticBoolParameter() {group=Emissive, parameterName=Modulate Emissive}
   UVPositioning LowerRight=AppendVector(A=@n6.B, B=@n6.B)
   UVPositioning UpperLeft=ComponentMask(None=@n6.RGB) {g=true, r=true}
   "------------------- "
   "Animate Using Texture Based Animation(T) or a Flicker(F)"=StaticBoolParameter() {group=Emissive, parameterName=AnimateEmissiveUsingAGradient}
   Animate=StaticBoolParameter() {group=Emissive, parameterName=AnimateEmissive}
   -------------------
   "World Position Animation Offset Multiplier (WS)"=ScalarParameter() {defaultValue=0.01, group=Emissive, parameterName=Animated emissive world position animation offset amount}
   Use World Position to offset the animation=StaticBoolParameter() {group=Emissive, parameterName=Use World position to offset the emissive animation}
   Animated Radius=ScalarParameter() {defaultValue=0.1, group=Emissive, parameterName=Emissive Timeline Animation Falloff Radius}
   Animated Hardness=ScalarParameter() {defaultValue=0.1, group=Emissive, parameterName=Emissive Timeline Animation Falloff Hardness}
   "Loop(T) or Ping  Pong(F) Animation"=StaticBoolParameter() {group=Emissive, parameterName="Animate Emissive Using a loop(T) or Ping Pong(F)"}
   Modulate with diffuse=StaticBoolParameter() {group=Emissive, parameterName=Modulate Emissive with Diffuse}
   Diffuse=@n58
   ---------------------
   ControlTimeManually
   ManualTime
   Sequence animation speed=ScalarParameter() {defaultValue=0.25, group=Emissive, parameterName=EmissiveAnimationSpeed}
  ) {materialFunction=/Game/Packages/Fortress_SharedMaterials/Base_Material_Functions/SequenceAnimation}
  n60 = VectorParameter() {group=Emissive2, parameterName="Emissive 2 UV Positioning (RG)UpperLeft (BA)LowerRight", defaultValue=<rgba 0 0 1 1>}
  n61 = MaterialFunctionCall(
   Manipulate Texture UVs=StaticBoolParameter() {group=Emissive2, parameterName=Manipulate Emissive 2 Uvs}
   ActorPositionAdd3=@n10.f3 Add
   Mask Cropped Emissive=StaticBoolParameter() {defaultValue=true, group=Emissive2, parameterName=Mask Cropped Emissive 2}
   Texture=TextureObjectParameter() {group=Emissive2, parameterName=Emissive Texture 2}
   Animated emissive black out time=ScalarParameter() {defaultValue=0.5, desc=should be a value from 0-1, group=Emissive2, parameterName=Animated Emissive Blackout Ratio 2}
   Texture Multiplier=VectorParameter() {group=Emissive2, parameterName=Emissive Multiplier 2, defaultValue=<rgba 1 1 1 1>}.RGB
   Modulate Texture=StaticBoolParameter() {group=Emissive2, parameterName=Modulate Emissive 2}
   UVPositioning LowerRight=AppendVector(A=@n60.B, B=@n60.B)
   UVPositioning UpperLeft=ComponentMask(None=@n60.RGB) {g=true, r=true}
   "Animate Using Texture Based Animation(T) or a Flicker(F)"=StaticBoolParameter() {group=Emissive2, parameterName=Animate Emissive Using the emissive alpha as a timeline 2}
   Animate=StaticBoolParameter() {group=Emissive2, parameterName=Animate Emissive 2}
   "World Position Animation Offset Multiplier (WS)"=ScalarParameter() {defaultValue=0.01, group=Emissive2, parameterName=Animated emissive world position animation offset amount 2}
   Use World Position to offset the animation=StaticBoolParameter() {group=Emissive2, parameterName=Use World position to offset the emissive animation 2}
   Animated Radius=ScalarParameter() {defaultValue=0.1, group=Emissive2, parameterName=Emissive Timeline Animation Falloff Radius 2}
   Animated Hardness=ScalarParameter() {defaultValue=0.1, group=Emissive2, parameterName=Emissive Timeline Animation Falloff Hardness 2}
   "Loop(T) or Ping  Pong(F) Animation"=StaticBoolParameter() {group=Emissive2, parameterName="Animate Emissive Using a loop(T) or Ping Pong(F) 2"}
   Modulate with diffuse=StaticBoolParameter() {group=Emissive2, parameterName=Modulate Emissive with Diffuse 2}
   Diffuse=@n58
   Sequence animation speed=ScalarParameter() {defaultValue=0.25, group=Emissive2, parameterName=Emissive Animation Speed 2}
  ) {materialFunction=&SequenceAnimation}
  n62 = StaticSwitchParameter(
   True=StaticSwitchParameter(True=Add(A=@n59.Result, B=@n61.Result), False=@n59.Result) {group=Emissive2, parameterName=Use Emissive 2}
   False=Constant()
  ) {group=Emissive, parameterName=Use Emissive}
  n63 = MaterialFunctionCall!(
   DiffuseFresnelAmount=ScalarParameter() {defaultValue=0.75, group=Diffuse, parameterName=BaseColorFresnelAmount}
   Emissive=StaticSwitchParameter(True=Multiply(A=@n5.Result, B=@n62), False=@n62) {defaultValue=true, group=Emissive, parameterName=Emissive Fresnel}
   Base Color=@n58
  ) {desc=diffuse, materialFunction=/Game/Packages/Fortress_SharedMaterials/Base_Material_Functions/FresnelFunction}
  n64 = ScalarParameter() {bUseCustomPrimitiveData=true, group=z_GameplayParameters_DestructibleDamage, parameterName=Code_DestructionAmount, primitiveDataIndex=13}
  n65 = Add(
   A=MaterialFunctionCall(DestructibleDamageAmount=@n64) {materialFunction=&MF_FN_DestructibleDamage}.DamageAnimWPO
   B=MaterialFunctionCall!(Use Tree Vertex Shader=StaticBoolParameter() {group=Special Cases, parameterName=Use Tree Vertex Shader}) {materialFunction=/Game/Packages/Fortress_SharedMaterials/Base_Material_Functions/FortniteTreeAnimation_VS}.Result
  )
  n66 = MaterialFunctionCall() {materialFunction=/Engine/Functions/Engine_MaterialFunctions02/WorldPositionOffset/StaticMeshMorphTargets}
  n67 = ScalarParameter() {group=MorphTargets, parameterName=MorphTargetStrength}
  n68 = StaticSwitchParameter(True=Add(A=@n65, B=Multiply(A=@n66.Morph Target 1 WPO, B=@n67)) {desc=7 instructions}, False=@n65) {group=MorphTargets, parameterName=UseMorphTargets}
  n69 = VectorParameter() {group=WorldPositionOffset, parameterName=WorldPositionOffset_ModulatedByVertexColor, defaultValue=<rgba 0 0 0 1>}
  n70 = VertexColor()
  n71 = StaticComponentMaskParameter(None=AppendVector(A=@n70, B=@n70)) {defaultA=true, group=WorldPositionOffset, parameterName=WorldPositionOffsetVertexColorModulationChannel}
  n72 = StaticSwitchParameter(True=Add(A=@n68, B=Multiply(A=@n69.RGB, B=@n71)), False=@n68) {group=WorldPositionOffset, parameterName=UseWorldPositionOffsetModulatedByVertexColors}
  n73 = Transform(None=Constant3Vector() {constant=<rgba 0 0 1 0>}) {transformSourceType=TRANSFORMSOURCE_Local}
  n74 = MaterialFunctionCall!(
   Enable=StaticBoolParameter() {defaultValue=true, group=Wind Setting 1, parameterName=Animate Level 1}
   Enable Normal Rotation=StaticBoolParameter() {group=Wind Setting 1, parameterName=Animate Level 1 Normals}
   Position=ActorPositionWS()
   X-Axis Texture And X Extent Alpha=AppendVector(A=Normalize(VectorInput=@n73), B=ObjectRadius())
   UseWrapAroundLogic=StaticBoolParameter() {group=Wind Setting 1, parameterName=Wind 1 - Wrap Deformation Around Pivot Point}
   Wind Gust Angle Rotation=Clamp(None=ScalarParameter() {defaultValue=-0.01, group=Wind Setting 1, parameterName=Max Rotation 1, sliderMax=10})
   Wind Gust Offset=ScalarParameter() {group=Wind Setting 1, parameterName=Wind offset 1}
   Optional - world space pixel normal
   Modulate Rotation Results by the Vertex Alpha=StaticBool() {value=true}
   Random Rotation Texture Sample Scale=ScalarParameter() {defaultValue=2.5, group=Wind Setting 1, parameterName=Wind 1 Random Rotation Change Rate}
   Random Rotation Influence=ScalarParameter() {defaultValue=1, group=Wind Setting 1, parameterName=Wind 1 Random Rotation Influence}
   World Position=Add(A=@n72, B=WorldPosition().XYZ)
   Model Depth=Constant()
   Effect's Target Depth=Constant()
   TimeRate_CodeReferenced1=ScalarParameter() {defaultValue=0.75, group=Wind Setting 1, parameterName=Wind_1_Speed_CodeReferenced}
   TimeOffset_CodeControlled1=ScalarParameter() {group=Code Controlled, parameterName=Wind_1_Offset_CodeReferenced}
   Add Shear=StaticBoolParameter() {group=Wind Setting 1, parameterName=Wind 1 - Add Shear}
   Max Sheer Offset=VectorParameter() {group=Wind Setting 1, parameterName=Wind 1 Shear Amount, defaultValue=<rgba 0 0 0 1>}.RGB
   Wind Direction=VectorParameter() {group=Code Controlled, parameterName=WindVector_CodeControlled, defaultValue=<rgba 1 0 0 1>}.RGB
  ) {materialFunction=/Game/Packages/Fortress_SharedMaterials/Base_Material_Functions/Fornite_BallAndSocketConstraint}
  n75 = TextureSampleParameter2D() {
   group=Specular
   parameterName=SpecularMasks
   samplerType=SAMPLERTYPE_Masks
   texture=&T_Fortnite_Default_S
  }
  n76 = ScalarParameter() {defaultValue=1, group=ClearCoat, parameterName=ClearCoat Rouhness}
  n77 = TextureSampleParameter2D(UVs=@n22) {
   group=Specular
   parameterName=SpecularMasks
   samplerType=SAMPLERTYPE_Masks
   texture=&T_Fortnite_Default_S
  }
  n78 = StaticSwitchParameter(True=ComponentMask(None=@n77.RGB) {b=true, g=true, r=true}, False=Constant3Vector() {constant=<rgba 1 0 0.2667 0>}) {desc=Specular Power, group=Specular, parameterName=Use SpecularMasks}
  n79 = MaterialFunctionCall(S=ComponentMask(None=@n78) {b=true, desc=roughness}) {desc=Normals, materialFunction=&PassThrough}
  n80 = LinearInterpolate(A=@n79.S, B=ScalarParameter() {defaultValue=0.6, group=CustomDamage, parameterName=Custom Damage Spec Roughness}, Alpha=@n57) {constB=0}
  n81 = Multiply(A=ScalarParameter() {defaultValue=0.5, group=Specular, parameterName=SpecLevel}, B=ComponentMask(None=@n78) {desc=spec, r=true})
  n82 = LinearInterpolate(A=@n81, Alpha=@n57) {constA=1, constB=0.5}
  n83 = ComponentMask(None=@n78) {desc=metalness, g=true}
  n84 = LinearInterpolate(A=@n83, Alpha=@n57)
  n85 = MaterialFunctionCall(Float3=MaterialFunctionCall(
    X=StaticSwitchParameter(True=@n82, False=@n81) {defaultValue=true, group=CustomDamage, parameterName=Use Custom Damage}
    Y=StaticSwitchParameter(True=@n84, False=@n83) {defaultValue=true, desc=Spec Metalness, group=CustomDamage, parameterName=Use Custom Damage}
    Z=StaticSwitchParameter(True=@n80, False=@n79.S) {defaultValue=true, desc=Spec Roughness, group=CustomDamage, parameterName=Use Custom Damage}
   ) {desc=Specular Masks, materialFunction=&MakeFloat3}.Result) {materialFunction=&BreakOutFloat3Components}
  n86 = StaticSwitchParameter(True=Multiply(A=@n75.A, B=OneMinus(None=@n57)), False=@n75.A) {defaultValue=true, desc=d, group=CustomDamage, parameterName=Use Custom Damage}
  n87 = ScalarParameter() {group=Snow, parameterName=SnowNormalContrast1}
  n88 = ScalarParameter() {defaultValue=1, group=Snow, parameterName=SnowNormalContrast2}
  n89 = TextureCoordinate()
  n90 = VectorParameter() {group=Normal, parameterName=Normal Offset RG Tiling BA, defaultValue=<rgba 0 0 1 1>}
  n91 = Add(A=@n89, B=ComponentMask(None=@n90.RGB) {g=true, r=true})
  n92 = AppendVector(A=@n90.B, B=@n90.B)
  n93 = TextureSampleParameter2D(UVs=StaticSwitchParameter(True=Multiply(A=@n91, B=@n92), False=@n89) {group=Normal, parameterName=Manipulate Normal UVS}) {group=Normal, parameterName=Normals, samplerType=SAMPLERTYPE_Normal, texture=/Engine/EngineMaterials/FlatNormal}
  n94 = StaticSwitchParameter(
   True=Multiply(A=@n93.RGB, B=VectorParameter() {group=Normal, parameterName=Normal Modulation, defaultValue=<rgba 1 1 1 1>}.RGB)
   False=@n93.RGB
  ) {group=Normal, parameterName=ModulateNormals}
  n95 = LinearInterpolate(
   A=@n94
   B=MaterialFunctionCall() {materialFunction=&StaticMeshMorphTargets}.Morph Target 1 Normals
   Alpha=ScalarParameter() {defaultValue=1, group=MorphTargets, parameterName=MorphTargetStrength}
  )
  n96 = MaterialFunctionCall(V3=StaticSwitchParameter(True=StaticSwitchParameter(True=@n95, False=@n94) {group=MorphTargets, parameterName=MorphNormals}, False=@n94) {group=MorphTargets, parameterName=UseMorphTargets}) {desc=Normals, materialFunction=&PassThrough}
  n97 = TextureSampleParameter2D(UVs=Multiply(A=TextureCoordinate(), B=ScalarParameter() {defaultValue=1, group=CustomDamage, parameterName=Custom Damage Height Texture Tiling})) {
   group=CustomDamage
   parameterName=Custom Damage Normal
   samplerType=SAMPLERTYPE_Normal
   texture=/Game/Global/Textures/Default/Tileable_Patterns/CarDamageNormal
  }
  n98 = MaterialFunctionCall(
   BaseNormal=@n96.V3
   AdditionalNormal=LinearInterpolate(
    A=Constant3Vector() {constant=<rgba 0 0 1 0>}
    B=@n97.RGB
    Alpha=ScalarParameter() {defaultValue=0.5, group=CustomDamage, parameterName=Custom Damage Normal Map Intensity}
   )
  ) {materialFunction=&BlendAngleCorrectedNormals}
  n99 = ScalarParameter() {defaultValue=0.9, group=CustomDamage, parameterName=Custom Damage N Height Max}
  n100 = ScalarParameter() {defaultValue=10, group=CustomDamage, parameterName=Custom Damage Normal Map Height Map Final Reveal Mask Multiplier}
  n101 = Multiply(
   A=MaterialFunctionCall(Transition Phase=Multiply(A=@n53, B=@n99), Height Texture=@n56) {materialFunction=&HeightLerp}.Lerp Alpha No Contrast
   B=Clamp(None=Multiply(A=@n56, B=@n100) {constB=10})
  )
  n102 = StaticSwitchParameter(
   True=StaticSwitchParameter(True=LinearInterpolate(A=@n96.V3, B=@n98.Result, Alpha=@n101), False=@n96.V3) {defaultValue=true, desc=N, group=CustomDamage, parameterName=Use Custom Damage Normal Map}
   False=@n96.V3
  ) {defaultValue=true, desc=Normal, group=CustomDamage, parameterName=Use Custom Damage}
  n103 = ComponentMask(None=Transform(None=@n102)) {b=true}
  n104 = ScalarParameter() {group=Snow, parameterName=SnowDiffuseContrast1}
  n105 = ScalarParameter() {defaultValue=420.3344, group=Snow, parameterName=SnowDiffuseContrast2}
  n106 = ComponentMask(None=@n63.d) {r=true}
  n107 = Multiply(A=Clamp(None=LinearInterpolate(A=@n87, B=@n88, Alpha=@n103)), B=Clamp(None=LinearInterpolate(A=@n104, B=@n105, Alpha=@n106)))
  n108 = StaticSwitchParameter(True=LinearInterpolate(A=@n86, B=Constant(), Alpha=@n107), False=@n86) {parameterName=UseSnow}
  n109 = ScalarParameter() {group=Specular, parameterName=PaintMetallic}
  n110 = VectorParameter() {group=Snow, parameterName=SnowColor_1, defaultValue=<rgba 0.7262 0.7933 0.83 1>}
  n111 = AppendVector(A=@n77.RGB, B=@n77.RGB)
  n112 = StaticSwitchParameter(
   True=LinearInterpolate(
    A=@n63.d
    B=LinearInterpolate(A=@n63.d, B=@n110.RGB, Alpha=@n107)
    Alpha=ChannelMaskParameter(None=@n111) {group=Snow, maskChannel=Green, parameterName=SnowMaskChannel, defaultValue=<rgba 0 1 0 0>}.R
   )
   False=@n63.d
  ) {group=Snow, parameterName=UseSnow}
  n113 = LinearInterpolate(
   A=ScalarParameter() {group=Specular, parameterName=SpecRoughnessMin}
   B=ScalarParameter() {defaultValue=1, group=Specular, parameterName=SpecRoughnessMax}
   Alpha=@n85.B
  )
  n114 = LinearInterpolate(
   A=ScalarParameter() {group=Custom Tile, parameterName=TilingDiffuseRoughness_Min}
   B=ScalarParameter() {defaultValue=1, group=Custom Tile, parameterName=TilingDiffuseRoughness_Max}
   Alpha=@n31.A
  )
  n115 = SetMaterialAttributes!(
   MaterialAttributes
   Base Color=@n112
   Metallic=LinearInterpolate(A=@n85.G, B=Multiply(A=@n108, B=@n109), Alpha=@n86)
   Specular=@n85.R
   Roughness=StaticSwitch(
    True=LinearInterpolate(A=@n113, B=@n114, Alpha=@n34.G)
    False=@n113
    Value=StaticBoolParameter() {group=Custom Tile, parameterName=UseTilingBaseColorTexture}
   )
   Emissive Color=StaticSwitchParameter(True=LinearInterpolate(A=@n63.Emissive, Alpha=@n4) {constA=1, constB=0}, False=@n63.Emissive) {group=EmissiveDistanceFade, parameterName=EmissiveDistanceFade}
   Normal=@n102
   World Position Offset=ShadingPathSwitch(Default=Add(A=@n72, B=@n74.Add to WPO), Mobile=@n72) {desc=Migrated from FeatureLevelSwitch}
   Clear Coat=@n108
   Clear Coat Roughness=StaticSwitchParameter(
    True=Multiply(A=@n75.B, B=@n76)
    False=StaticSwitchParameter(True=@n76, False=@n76) {defaultValue=true, desc=d, group=CustomDamage, parameterName=Use Custom Damage}
   ) {desc=d, group=CustomDamage, parameterName=UseTextureForClearCoatRoughnes}
  ) {attributeSetTypes [
    #1
    #2
    #3
    #4
    #5
    #6
    #7
    #8
    #9
   ]}
  n116 = MaterialFunctionCall!(ExtraSnowMask, Simple Mask=@n3, ForceUseInWSNormal=@n3, UseActorPosition, UseSnowMaskNormal, UseColorTest, UseColorTestRock, In=@n115) {materialFunction=/Game/Packages/Fortnite_UniqueMaterials_1/MaterialFunctions/Trees/MF_Tree_Snow_Legacy}
  n117 = GetMaterialAttributes(None=@n116.Result) {attributeGetTypes [#8]}
  n118 = StaticSwitchParameter(
   True=SetMaterialAttributes!(MaterialAttributes=@n117.MaterialAttributes, Clear Coat=LinearInterpolate(A=@n117.Clear Coat, Alpha=@n116.SnowMask) {constB=0}) {attributeSetTypes [#8]}
   False=@n115
  ) {group=Snow, parameterName=UseSnowLegacy}
  n119 = StaticSwitchParameter(
   True=MaterialFunctionCall!(In=@n118, UVs, TransitionGradient) {materialFunction=/Game/Environments/Sets/Vulcano/Materials/MF_LavaEdgeGradientSimple}.Result
   False=@n118
  ) {parameterName=UseLavaEdgeGradient}
  MaterialFunctionCall!(In=@n119, UseSphericalMappingOverride) {materialFunction=/Game/Packages/Fortress_SharedMaterials/Base_Material_Functions/MF_CurieFallbackEffect}
  MaterialFunctionCall!(In=MaterialFunctionCall(MA In=@n119) {materialFunction=&MF_CreativeMoveTool}.MA Out) {materialFunction=/Game/Athena/WinterArt/Materials/MF_GameWideShared}
  Reroute(None=@n102)
  ScalarParameter() {defaultValue=1, group=CustomDamage, parameterName=Custom Damage Spec Metallic}
  ScalarParameter() {defaultValue=1, group=CustomDamage, parameterName=Custom Damage Spec Level}
  SetMaterialAttributes!(MaterialAttributes=GetMaterialAttributes() {attributeGetTypes [#1]}.MaterialAttributes, Base Color=@n112) {attributeSetTypes [#1]}
  VertexInterpolator(VS=MaterialFunctionCall() {materialFunction=&ComponentPivotLocation}.Result)
 ]
}
