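# DisableAO is a material function in the shared base character shader set that neutralizes ambient occlusion in a material.
# It takes a MaterialAttributes input, overrides only the AmbientOcclusion attribute with a constant of 1 through SetMaterialAttributes, and outputs the result.
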MaterialFunction /Game/Packages/Fortress_SharedMaterials/Base_Material_Functions/BaseCharacterShaderFunctions/DisableAO
graph {nodes [FunctionOutput(None=SetMaterialAttributes!(
    MaterialAttributes=FunctionInput() {bUsePreviewValueAsDefault=true, inputName=MaterialAttributes, inputType=FunctionInput_MaterialAttributes}
    AmbientOcclusion=Constant() {r=1}
   ) {attributeSetTypes [#1]})]}
