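# Legacy character master material, marked do-not-use, for skinned meshes with Subsurface Profile shading.
# A mask texture drives AO and blends skin color boost, specular and roughness, with mobile and quality switches.
# Emissive combines rim light, optionally animated or cropped glow, and pulsing hit glow; opacity masking handles res-in and ninja dissolve.
# Assembled material attributes pass through DeRez transition, skin FX, AO-disable and tech-art override functions.
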Material /Game/Packages/Fortress_SharedMaterials/Base_Material/CharacterShader_Do_Not_Use_LegacyOnly
graph {
 properties {
  bAutomaticallySetUsageInEditor=false
  bUseMaterialAttributes=true
  bUsedWithClothing=true
  bUsedWithMeshDeformer=true
  bUsedWithMorphTargets=true
  bUsedWithSkeletalMesh=true
  numCustomizedUVs=3
  physMaterial=/Game/FortressPhysicalMaterials/Player_Flesh
  pixelDepthOffsetMode=PDOM_Legacy
  refractionMethod=RM_IndexOfRefraction
  shadingModel=MSM_SubsurfaceProfile
  subsurfaceProfile=/Game/Packages/Fortress_SharedMaterials/SubSurfaceProfile/Character_SubsurfaceProfile
  usedShadingModels=Subsurface Profile
 }
 nodes [
  n1 = TextureSampleParameter2D() {
   desc="Masks R=AO G= Fresnel B = Tmission/Skin"
   group=Masks
   parameterName=M
   samplerType=SAMPLERTYPE_Masks
   texture=/Engine/EngineMaterials/DefaultDiffuse_TC_Masks
  }
  n2 = Multiply(A=@n1.R, B=@n1.R)
  n3 = VectorParameter() {group=SkinColor, parameterName=Skin Peach Fuzz Color and Fresnel Exponent, defaultValue=<rgba 0.0708 0.0852 0.14 2>}
  n4 = VectorParameter() {group=SkinColor, parameterName=Skin Boost Color And Exponent, defaultValue=<rgba 0.87 0.6688 0.6948 2>}
  n5 = TextureSampleParameter2D() {
   desc="Masks R=AO G= Fresnel B = Tmission/Skin"
   group=SkinColor
   parameterName=FemConstructorSkinColor
   samplerType=SAMPLERTYPE_Masks
   texture=&DefaultDiffuse_TC_Masks
  }
  n6 = TextureSampleParameter2D() {
   desc="Masks R=AO G= Fresnel B = Tmission/Skin"
   group=Masks
   parameterName=M
   samplerType=SAMPLERTYPE_Masks
   texture=&DefaultDiffuse_TC_Masks
  }
  n7 = OneMinus(None=@n6.B)
  n8 = TextureSampleParameter2D() {group=Diffuse, parameterName=Diffuse, texture=/Engine/EngineMaterials/DefaultDiffuse}
  n9 = StaticSwitchParameter(True=Add(A=Multiply(A=@n5.RGB, B=@n6.B), B=Multiply(A=@n7, B=@n8.RGB)), False=@n8.RGB) {group=SkinColor, parameterName=IsFemaleConstructor}
  n10 = ShadingPathSwitch(Default=Desaturation(None=@n9, Fraction=ScalarParameter() {group=Code, parameterName=DesaturationPercentage}), Mobile=@n9) {desc=Migrated from FeatureLevelSwitch}
  n11 = QualitySwitch(Default=@n10, Low=@n9, High=@n10, Medium=@n10)
  n12 = Multiply(A=@n11, B=@n4.RGB)
  n13 = Multiply(A=OneMinus(None=Fresnel(ExponentIn=@n4.A) {exponent=2}), B=@n12)
  n14 = LinearInterpolate(A=@n11, B=Multiply(A=ScalarParameter() {defaultValue=0.75, parameterName=Global Skin Boost EX}, B=Add(A=@n13, B=@n12)), Alpha=@n1.B)
  Add(A=Multiply(A=Multiply(A=@n2, B=@n3.RGB), B=Fresnel(ExponentIn=@n3.A) {exponent=3}), B=@n14)
  n15 = TextureSampleParameter2D() {
   desc="Masks R=AO G= Fresnel B = Tmission/Skin"
   group=Masks
   parameterName=M
   samplerType=SAMPLERTYPE_Masks
   texture=&DefaultDiffuse_TC_Masks
  }
  n16 = TextureSampleParameter2D() {
   desc="R = SpecMask G = Metallic B = Roughness"
   group=Specular
   parameterName=SpecularMasks
   samplerType=SAMPLERTYPE_Masks
   texture=&T_Fortnite_Default_S
  }
  n17 = TextureSampleParameter2D() {
   desc="R = SpecMask G = Metallic B = Roughness"
   group=Specular
   parameterName=SpecularMasks
   samplerType=SAMPLERTYPE_Masks
   texture=&T_Fortnite_Default_S
  }
  n18 = Multiply(A=Multiply(A=@n17.R, B=Constant() {r=1}) {constB=0.5}) {constB=0.5}
  n19 = Constant() {r=0.35}
  n20 = ScalarParameter() {defaultValue=0.6, parameterName=MatchNewSKin}
  n21 = LinearInterpolate(
   A=ScalarParameter() {group=Roughness, parameterName=SpecRoughnessMin}
   B=ScalarParameter() {defaultValue=1, group=Roughness, parameterName=SpecRoughnessMax}
   Alpha=@n16.B
  )
  n22 = ScalarParameter() {group=Roughness, parameterName=MobileSpecRoughnessMin}
  n23 = ScalarParameter() {defaultValue=1, group=Roughness, parameterName=MobileSpecRoughnessMax}
  n24 = ShadingPathSwitch(
   Default=@n21
   Mobile=StaticSwitchParameter(True=LinearInterpolate(A=@n22, B=@n23, Alpha=@n16.B), False=@n21) {parameterName=OverrideRoughnessOnMobile}
  ) {desc=Migrated from FeatureLevelSwitch}
  n25 = Multiply(A=@n15.B, B=ScalarParameter() {defaultValue=1, group=SkinColor, parameterName=SSS Brightness})
  n26 = TextureSampleParameter2D() {group=Normal, parameterName=Normals, samplerType=SAMPLERTYPE_Normal, texture=/Engine/EngineMaterials/DefaultNormal}
  n27 = MaterialFunctionCall!(WPO) {materialFunction=/Game/Packages/Fortress_SharedMaterials/Base_Material_Functions/Fortnite_HideElementsWithVertexColor}
  n28 = Constant()
  n29 = PreSkinnedPosition()
  n30 = CollectionParameter() {
   collection=&FortniteMaterialParameters
   parameterName=PlayerHeight_DoNotSetDynamically
  }
  n31 = Constant()
  n32 = ShadingPathSwitch(
   Default=Divide(
    A=MaterialFunctionCall(Float3=@n29) {materialFunction=&BreakOutFloat3Components}.B
    B=@n30
   ) {constB=175}
   Mobile=@n31
  ) {desc=Migrated from FeatureLevelSwitch}
  n33 = VectorParameter() {group=Opacity Mask, parameterName="Opacity Mask Texture Cropping ", defaultValue=<rgba 0 0 1 1>}
  n34 = MaterialFunctionCall!(
   UVs
   TextureIn=TextureObjectParameter() {group=Opacity Mask, parameterName=Opacity Mask}
   UpperLeftCorner=AppendVector(A=@n33.R, B=@n33.R)
   LowerRightCorner=AppendVector(A=@n33.B, B=@n33.B)
   ExportFloat4=StaticBool() {value=true}
  ) {materialFunction=/Engine/Functions/Engine_MaterialFunctions01/Texturing/TextureCropping}
  n35 = Saturate(None=Add(A=@n34.Cropped, B=OneMinus(None=@n34.Crop Mask)))
  n36 = TextureSampleParameter2D() {group=Opacity Mask, parameterName=Opacity Mask}
  n37 = AppendVector(A=@n36.RGB, B=@n36.RGB)
  n38 = MaterialFunctionCall!(Animation Gradient, Rez In Percentage) {materialFunction=/Game/Packages/Fortress_SharedMaterials/Base_Material_Functions/CharacterResInEffect}
  n39 = Panner(Coordinate=TextureCoordinate() {vTiling=4}) {speedX=0.65, speedY=0.5}
  n40 = Panner(Coordinate=TextureCoordinate() {vTiling=8}) {speedX=-1.325, speedY=-0.9}
  n41 = Saturate(None=Add(
    A=TextureSample(UVs=@n39) {texture=&T_NoisePack_01}.R
    B=TextureSample(UVs=@n40) {texture=&T_NoisePack_01}.B
   ))
  n42 = ScalarParameter() {defaultValue=1, parameterName=NinjaDissolve}
  n43 = StaticSwitchParameter(True=Multiply(A=@n38.Modulative Opacity, B=Add(A=@n41, B=@n42) {constB=0}), False=@n38.Modulative Opacity) {parameterName=UseNinjaShadowStance?}
  n44 = Multiply(
   A=StaticComponentMaskParameter(None=StaticSwitchParameter(True=@n35, False=@n37) {group=Opacity Mask, parameterName=Control Mask UVS}) {defaultR=true, group=Opacity Mask, parameterName=Opacity Mask Channel}
   B=@n43
  )
  n45 = Saturate(None=DotProduct(A=PixelNormalWS(), B=CameraVectorWS()))
  n46 = MaterialFunctionCall!(BalanceDistance, AO Texture=@n6.R, Diffuse Texture=@n9, Fresnel N dot C=@n45) {materialFunction=/Game/Packages/Fortress_SharedMaterials/Base_Material_Functions/Character_Rim}
  n47 = VectorParameter() {group=Emissive, parameterName=EmissiveUVs_RG_UpperLeftCorner_BA_LowerRightCorner, defaultValue=<rgba 0 0 1 1>}
  n48 = TextureObjectParameter() {
   group=Emissive
   parameterName=EmissiveDistanceField
   samplerType=SAMPLERTYPE_LinearGrayscale
   texture=/Game/Global/Textures/Default/Blanks/T_BlackLinearGrayscale
  }
  n49 = MaterialFunctionCall!(
   SubUV_Frames=ScalarParameter() {defaultValue=2, group=Emissive, parameterName=SubUV_Frames}
   SubUV_TextureObject=@n48
   SubUV_Speed=ScalarParameter() {defaultValue=2, group=Emissive, parameterName=SubUV_Speed}
  ) {materialFunction=/Game/Packages/Fortress_SharedMaterials/Base_Material_Functions/MF_DistanceFieldSubUV}
  n50 = TextureSampleParameter2D() {group=Emissive, parameterName=Emissive, texture=&T_BlacksRGB}
  n51 = StaticSwitchParameter(
   True=MaterialFunctionCall(
    TextureIn=TextureObjectParameter() {group=Emissive, parameterName=Emissive, texture=&T_BlacksRGB}
    UpperLeftCorner=ComponentMask(None=@n47.RGB) {g=true, r=true}
    LowerRightCorner=AppendVector(A=@n47.B, B=@n47.B)
   ) {materialFunction=&TextureCropping}.CroppedMasked
   False=StaticSwitchParameter(True=Multiply(A=@n49.Result, B=@n50.RGB), False=@n50.RGB) {group=Emissive, parameterName=UseAnimatedEmissive}
  ) {desc=CROP_EMIS, group=Emissive, parameterName=CroppedEmissive}
  n52 = StaticSwitchParameter(True=Multiply(A=@n51, B=@n51), False=@n51) {group=Emissive, parameterName=LinearEmissiveMask}
  n53 = Multiply(
   A=StaticSwitchParameter(True=Multiply(A=@n52, B=@n45) {desc=Emis*Facing}, False=@n52) {group=Emissive, parameterName=CameraFacing}
   B=ScalarParameter() {defaultValue=1, group=Emissive, parameterName=emissive mult}
  )
  n54 = VectorParameter() {group=Emissive, parameterName=Emissive Color, defaultValue=<rgba 1 1 1 1>}
  n55 = VectorParameter() {group=Hit Effects, parameterName=HitGlowOuter, defaultValue=<rgba 0 0 0 1>}
  n56 = VectorParameter() {group=Hit Effects, parameterName=HitGlow, defaultValue=<rgba 0 0 0 1>}
  n57 = MaterialFunctionCall(
   Brightness=ScalarParameter() {defaultValue=1.4, group=Hit Effects, parameterName=HitGlowFresnelBrightness}
   Exponent=ScalarParameter() {defaultValue=1, group=Hit Effects, parameterName=HitGlowFresnelExponent}
   Vertex / Pixel Normal=Transform(None=@n26.RGB)
  ) {materialFunction=&MF_Fresnel_FX}
  n58 = ConstantBiasScale(None=Sine(None=Time() {bOverride_Period=true, period=0.5}) {period=0.5}) {bias=1.5, scale=0.4}
  n59 = ScalarParameter() {defaultValue=1, group=Hit Effects, parameterName=HitGlowUsesPulse}
  n60 = Add(
   A=StaticSwitchParameter(True=Multiply(A=@n53, B=@n54.RGB), False=Constant()) {group=Emissive, parameterName=Emissive}
   B=Multiply(A=LinearInterpolate(A=@n55.RGB, B=@n56.RGB, Alpha=@n57.Fresnel), B=LinearInterpolate(B=@n58, Alpha=@n59) {constA=1})
  )
  n61 = MaterialFunctionCall!(
   Fresnel=@n45
   Material=MakeMaterialAttributes!(
    BaseColor=@n14
    Metallic=@n16.G
    Specular=LinearInterpolate(A=@n18, B=@n19, Alpha=@n1.B)
    Roughness=LinearInterpolate(A=@n24, B=@n20, Alpha=@n1.B)
    Anisotropy
    EmissiveColor=Add(A=@n46.Result, B=@n60)
    Opacity=@n25
    OpacityMask=StaticSwitchParameter(True=@n44, False=@n43) {group=Opacity Mask, parameterName=Use Opacity Mask Texture}
    Normal=@n26.RGB
    Tangent
    WorldPositionOffset=StaticSwitchParameter(True=@n27.Result, False=@n28) {group=Opacity Mask, parameterName=Use Vertex Colors for Mask}
    SubsurfaceColor
    ClearCoat
    ClearCoatRoughness
    AmbientOcclusion=@n1.R
    Refraction
    CustomizedUVs_0
    CustomizedUVs_1
    CustomizedUVs_2=QualitySwitch(Default=@n32, Low=@n31, High=@n32, Medium=@n31)
    CustomizedUVs_3
    CustomizedUVs_4
    CustomizedUVs_5
    CustomizedUVs_6
    CustomizedUVs_7
    PixelDepthOffset
    ShadingModel
    Displacement
   )
  ) {materialFunction=/Game/Packages/Fortress_SharedMaterials/Base_Material_Functions/CharacterDeRez_Transition}
  MaterialFunctionCall(In=MaterialFunctionCall!(MaterialAttributes=MaterialFunctionCall(Gmap=@n15.G, MaterialAttributes=@n61.Result) {materialFunction=&CharacterSkinFX}.Result) {materialFunction=/Game/Packages/Fortress_SharedMaterials/Base_Material_Functions/BaseCharacterShaderFunctions/DisableAO}.Result) {materialFunction=&MF_TechArtOverrides}
  Reroute(None=@n25)
 ]
}
