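# Debug material function that displays a checker texture on the second UV channel (index 1) of a material, for inspecting UV layout.
# The checker, scaled by a UV Tiling input and multiplied by the incoming material's ambient occlusion, becomes Base Color.
# The other wired attributes, including Normal and Opacity, all read the incoming Metallic output.
# The "Debug UV Channel 2" static switch selects this view or passes the material through unchanged.
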MaterialFunction /Game/Packages/Fortress_SharedMaterials/Base_Material_Functions/DEMO_View_UVs_CH2
graph {nodes [
  n1 = FunctionInput(Preview=Constant() {r=1}) {bUsePreviewValueAsDefault=true, inputName=UV Tiling, inputType=FunctionInput_Scalar, sortPriority=3}
  n2 = TextureCoordinate() {coordinateIndex=1}
  n3 = FunctionInput(Preview=MakeMaterialAttributes!(
    BaseColor
    Metallic
    Specular
    Roughness
    Anisotropy
    EmissiveColor
    Opacity
    OpacityMask
    Normal
    Tangent
    WorldPositionOffset
    SubsurfaceColor
    ClearCoat
    ClearCoatRoughness
    AmbientOcclusion
    Refraction
    CustomizedUVs_0
    CustomizedUVs_1
    CustomizedUVs_2
    CustomizedUVs_3
    CustomizedUVs_4
    CustomizedUVs_5
    CustomizedUVs_6
    CustomizedUVs_7
    PixelDepthOffset
    ShadingModel
    Displacement
   )) {bUsePreviewValueAsDefault=true, inputName=Material, inputType=FunctionInput_MaterialAttributes}
  n4 = BreakMaterialAttributes!(Attr=@n3)
  n5 = Multiply(
   A=TextureSample(UVs=Multiply(A=@n1, B=@n2)) {texture=/Game/Packages/Fortress_SharedMaterials/Textures/T_UV_Checker_D}.RGB
   B=@n4.AmbientOcclusion
  )
  FunctionOutput(None=StaticSwitchParameter(
    True=MakeMaterialAttributes(
     BaseColor=@n5
     Metallic=@n4.Metallic
     Specular=@n4.Metallic
     Roughness=@n4.Metallic
     Anisotropy=@n4.Metallic
     EmissiveColor=@n4.Metallic
     Opacity=@n4.Metallic
     OpacityMask=@n4.Metallic
     Normal=@n4.Metallic
     Tangent=@n4.Metallic
     WorldPositionOffset=@n4.Metallic
     SubsurfaceColor=@n4.Metallic
     ClearCoat=@n4.Metallic
     ClearCoatRoughness=@n4.Metallic
     AmbientOcclusion=@n4.Metallic
     Refraction=@n4.Metallic
     CustomizedUVs_0=@n4.Metallic
     CustomizedUVs_1=@n4.Metallic
     CustomizedUVs_2=@n4.Metallic
     CustomizedUVs_3=@n4.Metallic
     CustomizedUVs_4=@n4.Metallic
     CustomizedUVs_5=@n4.Metallic
     CustomizedUVs_6=@n4.Metallic
     CustomizedUVs_7=@n4.Metallic
     PixelDepthOffset=@n4.Metallic
     ShadingModel=@n4.Metallic
    )
    False=@n3
   ) {group=Debug Tools, parameterName=Debug UV Channel 2})
 ]}
