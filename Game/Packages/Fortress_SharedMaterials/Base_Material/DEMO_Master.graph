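# DEMO master material: a gray base surface with optional high-gloss look and debug views, usable on skeletal, Nanite and Niagara meshes.
# A static High Gloss switch swaps color boost, metallic, roughness and Default Lit versus Clear Coat shading, over tiled normal and AO textures.
# Attributes chain through Color ID mask and two UV-view functions, then are rebuilt into a CharacterDeRez transition.
# Vertex-color element hiding drives world position offset.
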Material /Game/Packages/Fortress_SharedMaterials/Base_Material/DEMO_Master
graph {
 properties {
  bUseMaterialAttributes=true
  bUsedWithClothing=true
  bUsedWithMeshDeformer=true
  bUsedWithMorphTargets=true
  bUsedWithNanite=true
  bUsedWithNiagaraMeshParticles=true
  bUsedWithSkeletalMesh=true
  pixelDepthOffsetMode=PDOM_Legacy
  refractionMethod=RM_IndexOfRefraction
  shadingModel=MSM_FromMaterialExpression
  usedShadingModels=Default Lit | Clear Coat
 }
 nodes [
  n1 = Constant() {r=1}
  n2 = Constant()
  n3 = StaticBoolParameter() {group=Debug Tools, parameterName=High Gloss}
  n4 = Constant() {r=0.3}
  n5 = Constant() {desc=.45, r=0.45}
  n6 = Constant() {r=1}
  n7 = Constant()
  n8 = Constant() {r=1}
  n9 = Constant()
  n10 = ShadingModel() {shadingModel=MSM_ClearCoat}
  n11 = ShadingModel()
  n12 = TextureSample() {desc=Flat Normal, samplerType=SAMPLERTYPE_Normal, texture=&FlatNormalBlue}
  n13 = ScalarParameter() {group=Textures, parameterName=Texture MipBias, sliderMax=15}
  n14 = ScalarParameter() {defaultValue=1, group=Textures, parameterName=Texture Tiling}
  n15 = Multiply(A=TextureCoordinate(), B=@n14)
  n16 = TextureSampleParameter2D!(UVs=@n15, Bias=@n13, Apply View MipBias) {
   group=Textures
   mipValueMode=TMVM_MipBias
   parameterName=Normal
   samplerType=SAMPLERTYPE_Normal
   texture=&FlatNormalBlue
  }
  n17 = ScalarParameter() {defaultValue=1, group=Textures, parameterName=Normal Map Influence, sliderMax=1}
  n18 = TextureSampleParameter2D(UVs=@n15, Bias=@n13) {
   group=Textures
   mipValueMode=TMVM_MipBias
   parameterName=AO M Texture/RGB Packed
   samplerType=SAMPLERTYPE_Masks
   texture=&T_WhiteMask
  }
  n19 = LinearInterpolate(
   A=Constant() {r=1}
   B=StaticSwitchParameter(
    True=@n18.R
    False=TextureSampleParameter2D(UVs=@n15, Bias=@n13) {group=Textures, mipValueMode=TMVM_MipBias, parameterName=AO, texture=&T_WhitesRGB}.RGB
   ) {group=Textures, parameterName=AO uses _M Texture}
   Alpha=ScalarParameter() {defaultValue=1, group=Textures, parameterName=AO Opacity, sliderMax=1}
  )
  n20 = VectorParameter() {group=Base Color, parameterName=Base Color, defaultValue=<rgba 0.3 0.3 0.3 1>}
  n21 = Multiply(A=Multiply(A=Constant() {desc=Base Color Boost, r=2.5}, B=@n20.RGB), B=@n19)
  n22 = Multiply(A=@n20.RGB, B=@n19)
  n23 = MaterialFunctionCall!(WPO) {materialFunction=/Game/Packages/Fortress_SharedMaterials/Base_Material_Functions/Fortnite_HideElementsWithVertexColor}
  n24 = Constant()
  n25 = MaterialFunctionCall!(
   Material=MakeMaterialAttributes!(
    BaseColor=StaticSwitch(True=@n21, False=@n22, Value=@n3) {desc=Base Color}
    Metallic=StaticSwitch(True=@n1, False=@n2, Value=@n3) {desc=Metallic}
    Specular=Constant() {desc=Specular, r=1}
    Roughness=StaticSwitch(True=@n4, False=@n5, Value=@n3) {desc=Roughness}
    Anisotropy
    EmissiveColor
    Opacity
    OpacityMask
    Normal=LinearInterpolate(A=@n12.RGB, B=@n16.RGB, Alpha=@n17)
    Tangent
    WorldPositionOffset=StaticSwitchParameter(True=@n23.Result, False=@n24) {group=Opacity Mask, parameterName=Use Vertex Colors for Mask}
    SubsurfaceColor
    ClearCoat=StaticSwitch(True=@n6, False=@n7, Value=@n3) {desc=Clear Coat on/off}
    ClearCoatRoughness=StaticSwitch(True=@n8, False=@n9, Value=@n3) {desc=Clear Coat Roughness}
    AmbientOcclusion=@n19
    Refraction
    CustomizedUVs_0
    CustomizedUVs_1
    CustomizedUVs_2
    CustomizedUVs_3
    CustomizedUVs_4
    CustomizedUVs_5
    CustomizedUVs_6
    CustomizedUVs_7
    PixelDepthOffset
    ShadingModel=StaticSwitch(True=@n10, False=@n11, Value=@n3) {desc=Shading Model}
    Displacement
   )
   Color ID Texture=TextureObjectParameter() {group=Textures, parameterName=Color ID/Clown Mask, texture=/Game/Global/Textures/Default/Masks_and_Gradients/3pxRGBImage}
   UV Tiling=@n15
   MipBias=@n13
  ) {materialFunction=/Game/Packages/Fortress_SharedMaterials/Base_Material_Functions/DEMO_View_ColorID_Mask}
  n26 = BreakMaterialAttributes!(Attr=MaterialFunctionCall!(
    Material=MaterialFunctionCall!(Material=@n25.Result, UV Tiling=@n14) {materialFunction=/Game/Packages/Fortress_SharedMaterials/Base_Material_Functions/DEMO_View_UVs}.Result
    UV Tiling=@n14
   ) {materialFunction=/Game/Packages/Fortress_SharedMaterials/Base_Material_Functions/DEMO_View_UVs_CH2}.Result)
  MaterialFunctionCall!(
   Fresnel
   Material=MakeMaterialAttributes(
    BaseColor=@n26.BaseColor
    Metallic=@n26.BaseColor
    Specular=@n26.BaseColor
    Roughness=@n26.BaseColor
    Anisotropy=@n26.BaseColor
    EmissiveColor=@n26.BaseColor
    Opacity=@n26.BaseColor
    OpacityMask=@n26.BaseColor
    Normal=@n26.BaseColor
    Tangent=@n26.BaseColor
    WorldPositionOffset=MaterialFunctionCall(WPO=@n26.WorldPositionOffset) {materialFunction=&Fortnite_HideElementsWithVertexColor}.Result
    SubsurfaceColor=@n26.BaseColor
    ClearCoat=@n26.BaseColor
    ClearCoatRoughness=@n26.BaseColor
    AmbientOcclusion=Constant() {r=1}
    Refraction=@n26.BaseColor
    CustomizedUVs_0=@n26.BaseColor
    CustomizedUVs_1=@n26.BaseColor
    CustomizedUVs_2=@n26.BaseColor
    CustomizedUVs_3=@n26.BaseColor
    CustomizedUVs_4=@n26.BaseColor
    CustomizedUVs_5=@n26.BaseColor
    CustomizedUVs_6=@n26.BaseColor
    CustomizedUVs_7=@n26.BaseColor
    PixelDepthOffset=@n26.BaseColor
    ShadingModel=@n26.BaseColor
   )
  ) {materialFunction=/Game/Packages/Fortress_SharedMaterials/Base_Material_Functions/CharacterDeRez_Transition}
 ]
}
