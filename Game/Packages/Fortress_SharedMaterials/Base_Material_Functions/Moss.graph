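# Material function that layers moss over a base material's color, specular masks, normal and emissive, gated by the Use Transylvania Moss static switch.
# World-aligned moss textures are height-lerped in by a transition phase (vertex red by default), with view-angle shading on the moss and darkened base color in the blend.
# Mossy areas lose specular, become fully rough and mask out emissive.
# Low quality passes the input normal through, saving about 60 instructions.
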MaterialFunction /Game/Packages/Fortress_SharedMaterials/Base_Material_Functions/Moss
graph {
 properties {description=only adds 30 instructions on low end. Normals add roughtly 60 more instructions}
 nodes [
  Constant3Vector() {desc=slime color, constant=<rgba 0.0463 0.0862 0 0>}
  n1 = Multiply(A=@n11.Lerp Alpha No Contrast) {constB=3}
  n2 = Clamp(None=DotProduct(A=CameraVectorWS(), B=PixelNormalWS()))
  n3 = OneMinus(None=@n2)
  n4 = Add(A=Multiply(A=Power(Base=@n3) {constExponent=6}) {constB=0.8, desc=edge}, B=OneMinus(None=Multiply(A=@n2) {constB=0.9, desc=core}))
  n5 = FunctionInput(Preview=StaticBool()) {
   bUsePreviewValueAsDefault=true
   description=Defaults to false
   inputName=Use World Space Normals
   inputType=FunctionInput_StaticBool
   sortPriority=55
  }
  n6 = MaterialFunctionCall!(
   Diffuse Texture Object=TextureObject() {texture=/Game/Packages/Fortress_SharedMaterials/Textures/Moss_DiffuseWithAlphaMask_Proto}
   Spec Texture Object
   TextureSize=Constant() {r=256}
   --------------------------
   Use seperate settings for Z projection
   Use seperate texture for Z projection
   Texture Size Z projection
   Z Plane Diffuse Texture Object
   Z Plane Spec Texture Object
   "--------------------------      "
   ProjectionTransitionContrast=Constant() {r=5}
   Use Texture Alpha Heightmap to Refine the Transistion
   HeightLerpTransitionContrast
   PickTheDiffuseHeightChannel
   TheDiffuseHeightChannel
   "--------------------------       "
   NormalTextureObject=TextureObject() {samplerType=SAMPLERTYPE_Normal, texture=/Game/Packages/Fortress_SharedMaterials/Textures/Moss_Normals}
   World Space Normals=@n5
   "--------------------------  "
   Export Float 4 Diffuse=StaticBool() {value=true}
   Export Float 4 Spec
   "--------------------------     "
   WorldPosition=WorldPosition() {worldPositionShaderOffset=WPT_ExcludeAllShaderOffsets}.XYZ
   World Space Normal
  ) {materialFunction=/Engine/Functions/Engine_MaterialFunctions02/Texturing/WorldAlignedTextures_Complex}
  n7 = ComponentMask(None=@n6.XYZ Diffuse Texture) {b=true, g=true, r=true}
  n8 = FunctionInput() {bUsePreviewValueAsDefault=true, inputName=Base Color, sortPriority=3, previewValue=<1 1 1 1>}
  n9 = Clamp(None=LinearInterpolate(Alpha=@n11.Lerp Alpha No Contrast) {constA=-2, constB=2})
  n10 = FunctionInput(Preview=VertexColor().R) {
   bUsePreviewValueAsDefault=true
   description=Define the 0-1 information used to lerp between the moss and the base layer. This defaults to the red vertex color channel
   inputName=Transition Phase
   inputType=FunctionInput_Scalar
   sortPriority=1
   previewValue=<0.5 1 1 1>
  }
  n11 = MaterialFunctionCall(
   A=Multiply(A=Add(A=@n1), B=Multiply(A=@n4, B=@n7) {desc=diffuse})
   B=Multiply(A=@n8, B=LinearInterpolate(Alpha=@n9) {desc=just darken})
   Transition Phase=@n10
   Height Texture=ComponentMask(None=@n6.XYZ Diffuse Texture) {a=true}
   Contrast=Constant() {r=1}
  ) {materialFunction=&HeightLerp}
  n12 = FunctionInput(Preview=StaticBool()) {
   bUsePreviewValueAsDefault=true
   description=defaults to false
   inputName=Use Transylvania Moss
   inputType=FunctionInput_StaticBool
   sortPriority=-1
  }
  FunctionOutput(None=StaticSwitch(True=@n11.Results, False=@n8, Value=@n12)) {outputName=Base Color}
  n13 = FunctionInput() {bUsePreviewValueAsDefault=true, inputName=SpecularMasks, sortPriority=4}
  n14 = StaticSwitch(
   True=LinearInterpolate(A=Constant3Vector() {constant=<rgba 0 0 1 0>}, B=@n13, Alpha=@n11.Alpha) {desc=disable spec in mossy areas}
   False=@n13
   Value=@n12
  )
  FunctionOutput(None=@n14) {outputName=SpecularMasks, sortPriority=1}
  FunctionOutput(None=ComponentMask(None=@n14) {r=true}) {outputName=Specular, sortPriority=1}
  FunctionOutput(None=ComponentMask(None=@n14) {g=true}) {outputName=Metallic, sortPriority=1}
  FunctionOutput(None=ComponentMask(None=@n14) {b=true}) {outputName=Roughness, sortPriority=1}
  n15 = FunctionInput(Preview=StaticSwitch(True=VertexNormalWS(), False=Constant3Vector() {constant=<rgba 0 0 1 0>}, Value=@n5)) {bUsePreviewValueAsDefault=true, description=Defaults to VertexNormalWS, inputName=Normal, sortPriority=50}
  n16 = StaticSwitch(True=LinearInterpolate(A=@n6.Normal XYZ Texture, B=@n15, Alpha=@n11.Alpha), False=@n15, Value=@n12)
  FunctionOutput(None=QualitySwitch(Default=@n16, Low=@n15, High=@n16) {desc=normal output - removing normal support removes 60 instructions from the shader}) {outputName=Normal Output, sortPriority=10}
  n17 = FunctionInput() {bUsePreviewValueAsDefault=true, inputName=Emissive, sortPriority=6}
  FunctionOutput(None=StaticSwitch(True=Multiply(A=@n11.Alpha, B=@n17), False=@n17, Value=@n12)) {outputName=Emissive, sortPriority=3}
  QualitySwitch()
 ]
}
