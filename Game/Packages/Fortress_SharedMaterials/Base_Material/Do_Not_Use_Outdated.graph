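# Outdated masked, two-sided base master material, named "Do Not Use", that covers props, buildings, vehicles and melee weapons through static switches.
# Diffuse layers blend by texture, height or vertex color, then take world-position color-chart tinting, faked reflections and height-lerped custom damage.
# Shared functions add animated emissive, moss and supercharged glow.
# World position offset combines hit effects, tree sway and morph targets, and attributes pass through creative-move and crack functions.
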Material /Game/Packages/Fortress_SharedMaterials/Base_Material/Do_Not_Use_Outdated
graph {
 properties {
  bUseMaterialAttributes=true
  bUsedWithInstancedStaticMeshes=true
  bUsedWithMeshDeformer=true
  bUsedWithNanite=true
  bUsedWithSkeletalMesh=true
  blendMode=BLEND_Masked
  pixelDepthOffsetMode=PDOM_Legacy
  twoSided=true
  usedShadingModels=Default Lit
 }
 nodes [
  n1 = VectorParameter() {group=Emissive2, parameterName="Emissive 2 UV Positioning (RG)UpperLeft (BA)LowerRight", defaultValue=<rgba 0 0 1 1>}
  AppendVector(A=@n1.RGB, B=@n1.RGB)
  n2 = StaticSwitchParameter(
   True=MaterialFunctionCall(Speed=Constant() {r=0.3333}) {materialFunction=&TimeWithSpeedVariable}.Frac Time
   False=ScalarParameter() {group=Code, parameterName=Code_CustomDamagePercentage}
  ) {group=CustomDamage, parameterName=Debug custom damage animation}
  n3 = TextureSampleParameter2D() {
   group=CustomDamage
   parameterName=Custom Damage Height Map
   samplerType=SAMPLERTYPE_Masks
   texture=/Game/Packages/Fortress_ShaderTests/Textures/T_CopCar_DamageMask
  }
  n4 = StaticComponentMaskParameter(None=AppendVector(A=@n3.RGB, B=@n3.RGB)) {defaultG=true, group=CustomDamage, parameterName=Custom Damage Height Map Channel}
  n5 = MaterialFunctionCall(
   Transition Phase=Multiply(A=@n2, B=ScalarParameter() {defaultValue=0.9, group=CustomDamage, parameterName=Custom Damage N Height Max})
   Height Texture=@n4
  ) {materialFunction=&HeightLerp}
  Clamp(None=Multiply(A=@n5.Lerp Alpha No Contrast) {constB=5})
  n6 = VectorParameter() {group=Emissive, parameterName="EmissiveUVPositioning (RG)UpperLeft (BA)LowerRight", defaultValue=<rgba 0 0 1 1>}
  n7 = MaterialFunctionCall!(
   Manipulate Texture UVs=StaticBoolParameter() {group=Emissive, parameterName=Manipulate Emissive Uvs}
   ActorPositionAdd3
   Mask Cropped Emissive
   Texture=TextureObjectParameter() {group=Emissive, parameterName=EmissiveTexture}
   Animated emissive black out time=ScalarParameter() {defaultValue=0.5, desc=should be a value from 0-1, group=Emissive, parameterName=Animated Emissive Blackout Ratio}
   Texture Multiplier=VectorParameter() {group=Emissive, parameterName=Emissive Multiplier, defaultValue=<rgba 1 1 1 1>}.RGB
   Modulate Texture=StaticBoolParameter() {group=Emissive, parameterName=Modulate Emissive}
   UVPositioning LowerRight=AppendVector(A=@n6.B, B=@n6.B)
   UVPositioning UpperLeft=ComponentMask(None=@n6.RGB) {g=true, r=true}
   "------------------- "
   "Animate Using Texture Based Animation(T) or a Flicker(F)"=StaticBoolParameter() {group=Emissive, parameterName=AnimateEmissiveUsingAGradient}
   Animate=StaticBoolParameter() {group=Emissive, parameterName=AnimateEmissive}
   -------------------
   "World Position Animation Offset Multiplier (WS)"=ScalarParameter() {defaultValue=0.01, group=Emissive, parameterName=Animated emissive world position animation offset amount}
   Use World Position to offset the animation=StaticBoolParameter() {group=Emissive, parameterName=Use World position to offset the emissive animation}
   Animated Radius=ScalarParameter() {defaultValue=0.1, group=Emissive, parameterName=Emissive Timeline Animation Falloff Radius}
   Animated Hardness=ScalarParameter() {defaultValue=0.1, group=Emissive, parameterName=Emissive Timeline Animation Falloff Hardness}
   "Loop(T) or Ping  Pong(F) Animation"=StaticBoolParameter() {group=Emissive, parameterName="Animate Emissive Using a loop(T) or Ping Pong(F)"}
   Modulate with diffuse
   Diffuse
   ---------------------
   ControlTimeManually
   ManualTime
   Sequence animation speed=ScalarParameter() {defaultValue=0.25, group=Emissive, parameterName=EmissiveAnimationSpeed}
  ) {materialFunction=/Game/Packages/Fortress_SharedMaterials/Base_Material_Functions/SequenceAnimation}
  n8 = Add(
   A=@n7.Result
   B=MaterialFunctionCall(
    Manipulate Texture UVs=StaticBoolParameter() {group=Emissive2, parameterName=Manipulate Emissive 2 Uvs}
    Texture=TextureObjectParameter() {group=Emissive2, parameterName=Emissive Texture 2}
    Animated emissive black out time=ScalarParameter() {defaultValue=0.5, desc=should be a value from 0-1, group=Emissive2, parameterName=Animated Emissive Blackout Ratio 2}
    Texture Multiplier=VectorParameter() {group=Emissive2, parameterName=Emissive Multiplier 2, defaultValue=<rgba 1 1 1 1>}.RGB
    Modulate Texture=StaticBoolParameter() {group=Emissive2, parameterName=Modulate Emissive 2}
    UVPositioning LowerRight=AppendVector(A=@n1.B, B=@n1.B)
    UVPositioning UpperLeft=ComponentMask(None=@n1.RGB) {g=true, r=true}
    "Animate Using Texture Based Animation(T) or a Flicker(F)"=StaticBoolParameter() {group=Emissive2, parameterName=Animate Emissive Using the emissive alpha as a timeline 2}
    Animate=StaticBoolParameter() {group=Emissive2, parameterName=Animate Emissive 2}
    "World Position Animation Offset Multiplier (WS)"=ScalarParameter() {defaultValue=0.01, group=Emissive2, parameterName=Animated emissive world position animation offset amount 2}
    Use World Position to offset the animation=StaticBoolParameter() {group=Emissive2, parameterName=Use World position to offset the emissive animation 2}
    Animated Radius=ScalarParameter() {defaultValue=0.1, group=Emissive2, parameterName=Emissive Timeline Animation Falloff Radius 2}
    Animated Hardness=ScalarParameter() {defaultValue=0.1, group=Emissive2, parameterName=Emissive Timeline Animation Falloff Hardness 2}
    "Loop(T) or Ping  Pong(F) Animation"=StaticBoolParameter() {group=Emissive2, parameterName="Animate Emissive Using a loop(T) or Ping Pong(F) 2"}
    Sequence animation speed=ScalarParameter() {defaultValue=0.25, group=Emissive2, parameterName=Emissive Animation Speed 2}
   ) {materialFunction=&SequenceAnimation}.Result
  )
  n9 = TextureCoordinate() {coordinateIndex=1}
  n10 = ScalarParameter() {
   defaultValue=0.1
   desc="1/number of colors. ie 1/10 = .1"
   group=World Space Color Picker
   parameterName="Number Of Colors in Color Chart (1/Number of colors)"
  }
  n11 = MaterialFunctionCall!(
   f2
   f3=Divide(A=ActorPositionWS(), B=ScalarParameter() {defaultValue=16.71, group=World Space Color Picker, parameterName="Color Change Frequency (WS)"})
   f4
  ) {materialFunction=/Game/Packages/Fortnite_UniqueMaterials_1/VectorComponentOps}
  n12 = Add(A=Multiply(A=ComponentMask(None=@n9) {g=true}, B=@n10), B=@n11.f3 Add)
  n13 = WorldPosition() {worldPositionShaderOffset=WPT_ExcludeAllShaderOffsets}
  n14 = MaterialFunctionCall(
   TextureObject=TextureObjectParameter() {desc=Use for exterior walls and floors, group=Diffuse, parameterName=Diffuse}
   TextureSize=VectorParameter() {group=Diffuse, parameterName=WorldPositionCoordinates_diffuse, defaultValue=<rgba 256 256 256 1>}.RGB
   WorldPosition=@n13.XYZ
   Export Float 4=StaticBool() {value=true}
  ) {materialFunction=&WorldAlignedTexture}
  n15 = StaticSwitchParameter(True=@n14.XY Texture, False=@n14.XY Texture) {group=Diffuse, parameterName=WorldPositionUV_XY_Or_Z_diffuse}
  n16 = TextureCoordinate()
  n17 = VectorParameter() {group=Diffuse, parameterName=Diffuse Offset RG Tiling BA, defaultValue=<rgba 0 0 1 1>}
  n18 = Add(A=@n16, B=ComponentMask(None=@n17.RGB) {g=true, r=true})
  n19 = AppendVector(A=@n17.B, B=@n17.B)
  n20 = TextureSampleParameter2D(UVs=StaticSwitchParameter(True=Multiply(A=@n18, B=@n19), False=@n16) {group=Diffuse, parameterName=Manipulate Diffuse UVS}) {desc=Use for exterior walls and floors, group=Diffuse, parameterName=Diffuse}
  n21 = VectorParameter() {group=Diffuse, parameterName=Diffuse Color Vector, defaultValue=<rgba 0.5 0.5 0.5 1>}
  n22 = StaticSwitchParameter(
   True=StaticSwitchParameter(
    True=StaticSwitchParameter(True=@n14.XYZ Texture, False=@n15) {group=Diffuse, parameterName=WorldPositionUV_XYZ_Diffuse}
    False=AppendVector(A=@n20.RGB, B=@n20.RGB)
   ) {group=Diffuse, parameterName=UseWorldPosition_diffuse}
   False=AppendVector(A=@n21.RGB, B=@n21.RGB)
  ) {defaultValue=true, group=Diffuse, parameterName="Use Diffuse Texture "}
  n23 = ComponentMask(None=@n22) {b=true, g=true, r=true}
  n24 = VectorParameter() {group=LerpTexture, parameterName=Offset World Position Lerp Coordinates, defaultValue=<rgba 0 0 0 1>}
  n25 = MaterialFunctionCall(
   TextureObject=TextureObjectParameter() {group=LerpTexture, parameterName=Diffuse Lerp Texture}
   TextureSize=VectorParameter() {group=LerpTexture, parameterName=WorldPositionCoordinates_Lerp, defaultValue=<rgba 256 256 256 1>}.RGB
   WorldPosition=StaticSwitchParameter(True=Add(A=@n13.XYZ, B=@n24.RGB), False=@n13.XYZ) {group=LerpTexture, parameterName=Offset World Position Lerp Coordinates}
   Export Float 4=StaticBool() {value=true}
  ) {materialFunction=&WorldAlignedTexture}
  n26 = StaticSwitchParameter(True=TextureCoordinate() {coordinateIndex=1}, False=TextureCoordinate()) {group=LerpTexture, parameterName="Use UV channel 0(false) or 1(true) for the lerp texture"}
  n27 = VectorParameter() {group=LerpTexture, parameterName=Lerp Diffuse UV Offset RG Tiling BA, defaultValue=<rgba 0 0 1 1>}
  n28 = Add(A=@n26, B=ComponentMask(None=@n27.RGB) {g=true, r=true})
  n29 = AppendVector(A=@n27.B, B=@n27.B)
  n30 = TextureSampleParameter2D(UVs=StaticSwitchParameter(True=Multiply(A=@n28, B=@n29), False=@n26) {group=LerpTexture, parameterName=Manipulate Lerp UVS}) {group=LerpTexture, parameterName=Diffuse Lerp Texture}
  n31 = StaticSwitchParameter(
   True=StaticSwitchParameter(
    True=@n25.XYZ Texture
    False=StaticSwitchParameter(True=@n25.XY Texture, False=@n25.XY Texture) {group=LerpTexture, parameterName=WorldPositionUV_XY_Or_Z_lerp}
   ) {group=LerpTexture, parameterName=WorldPositionUV_XYZ_Lerp}
   False=AppendVector(A=@n30.RGB, B=@n30.RGB)
  ) {group=LerpTexture, parameterName=UseWorldPosition_lerpTexture}
  n32 = ComponentMask(None=@n31) {b=true, g=true, r=true}
  n33 = VectorParameter() {group=LerpTexture, parameterName=Add Color to Diffuse Lerp Texture, defaultValue=<rgba 0 0 0 1>}
  n34 = StaticSwitchParameter(
   True=Multiply(
    A=Add(A=@n32, B=@n33.RGB)
    B=VectorParameter() {group=LerpTexture, parameterName=Multiply Diffuse Lerp Texture, defaultValue=<rgba 1 1 1 1>}.RGB
   )
   False=VectorParameter() {group=LerpTexture, parameterName=Lerp color, defaultValue=<rgba 0 0 0 1>}.RGB
  ) {group=LerpTexture, parameterName=Lerp using Texture or Vector}
  n35 = ComponentMask(None=@n31) {a=true}
  n36 = ComponentMask(None=@n22) {a=true}
  n37 = TextureSampleParameter2D() {group=LerpTexture, parameterName=Height Lerp Height Texture}
  n38 = StaticSwitchParameter(
   True=@n35
   False=StaticComponentMaskParameter(None=AppendVector(A=@n37.RGB, B=@n37.RGB)) {defaultR=true, group=LerpTexture, parameterName=Height Lerp Height Texture Channel}
  ) {group=LerpTexture, parameterName=Use Lerp Diffuse Alpha as Lerp Alpha}
  n39 = VertexColor()
  n40 = AppendVector(A=@n39, B=@n39)
  n41 = MaterialFunctionCall!(
   Lerp A
   Lerp B
   Height Texture=StaticSwitchParameter(True=@n35, False=StaticSwitchParameter(True=@n36, False=@n38) {group=LerpTexture, parameterName=Use Diffuse Alpha as Height Lerp As Map}) {group=LerpTexture, parameterName=Use Diffuse Lerp Texture Alpha as Height Map}
   Falloff Contrast=StaticSwitchParameter(
    True=StaticComponentMaskParameter(None=@n40) {defaultR=true, group=LerpTexture, parameterName=Vertex Color to Use For HeightLerp Falloff Contrast}
    False=ScalarParameter() {group=LerpTexture, parameterName=Height Lerp Transition Contrast}
   ) {group=LerpTexture, parameterName=Use Vertex Color for falloff contrast}
   Transition Phase=StaticSwitchParameter(
    True=StaticComponentMaskParameter(None=@n40) {defaultG=true, group=LerpTexture, parameterName="Vertex Color to Use for HeightLerp Transition  Phase"}
    False=ScalarParameter() {group=LerpTexture, parameterName=Height Lerp Transition Contrast}
   ) {group=LerpTexture, parameterName=Use Green Vertex Color Channel for Transition Phase}
  ) {materialFunction=/Game/Packages/Fortnight_MaterialFunctions/Temp/HeightLerp}
  n42 = VectorParameter() {group=LerpTexture, parameterName=Offset World Position Lerp Alpha Coordinates, defaultValue=<rgba 0 0 0 1>}
  n43 = MaterialFunctionCall(
   TextureObject=TextureObjectParameter() {group=LerpTexture, parameterName=Lerp Alpha}
   TextureSize=VectorParameter() {group=LerpTexture, parameterName=WorldPositionCoordinates_LerpAlpha, defaultValue=<rgba 256 256 256 1>}.RGB
   WorldPosition=StaticSwitchParameter(True=Add(A=@n13.XYZ, B=@n42.RGB), False=@n13.XYZ) {group=LerpTexture, parameterName=Offset World Position Lerp Alpha Coordinates}
  ) {materialFunction=&WorldAlignedTexture}
  n44 = StaticSwitchParameter(
   True=@n43.XYZ Texture
   False=StaticSwitchParameter(True=@n43.XY Texture, False=@n43.XY Texture) {group=LerpTexture, parameterName=Lerp Alpha Coordinates XY or Z}
  ) {group=LerpTexture, parameterName=Lerp Alpha Coordinates XYZ or 2 way}
  n45 = StaticSwitchParameter(True=TextureCoordinate() {coordinateIndex=1}, False=TextureCoordinate()) {group=LerpTexture, parameterName="Use UV channel 0(false) or 1(true) for the lerp texture alpha"}
  n46 = VectorParameter() {group=LerpTexture, parameterName=Lerp Alpha Offset RG Tiling BA, defaultValue=<rgba 0 0 1 1>}
  n47 = Add(A=@n45, B=ComponentMask(None=@n46.RGB) {g=true, r=true})
  n48 = AppendVector(A=@n46.B, B=@n46.B)
  n49 = TextureSampleParameter2D(UVs=StaticSwitchParameter(True=Multiply(A=@n47, B=@n48), False=@n45) {group=LerpTexture, parameterName=Manipulate Lerp Alpha UVS}) {group=LerpTexture, parameterName=Lerp Alpha}
  n50 = AppendVector(A=@n49.RGB, B=@n49.RGB)
  n51 = StaticComponentMaskParameter(None=StaticSwitchParameter(True=@n22, False=StaticSwitchParameter(True=@n44, False=@n50) {group=LerpTexture, parameterName=Use World Pos Coordinates for Lerp Alpha}) {group=LerpTexture, parameterName=Diffuse Channel to use as lerp alpha}) {defaultR=true, group=LerpTexture, parameterName=Lerp Alpha Texture}
  n52 = StaticSwitchParameter(
   True=StaticComponentMaskParameter(None=@n40) {defaultR=true, group=LerpTexture, parameterName=Vertex Color to Use as Lerp Alpha}
   False=StaticSwitchParameter(
    True=StaticSwitchParameter(True=@n35, False=@n51) {group=LerpTexture, parameterName=Use Lerp Diffuse Alpha as Lerp Alpha}
    False=ScalarParameter() {defaultValue=0.5, group=LerpTexture, parameterName=Lerp Alpha}
   ) {group=LerpTexture, parameterName=Use Texture as Lerp Alpha or Scalar}
  ) {group=LerpTexture, parameterName=Use Vertex Color As Alpha for Lerp Texture}
  n53 = StaticSwitchParameter(True=@n41.Lerp Alpha, False=@n52) {group=LerpTexture, parameterName=Use HeightLerp Texture Blending}
  n54 = StaticSwitchParameter(True=AppendVector(A=LinearInterpolate(A=@n23, B=@n34, Alpha=@n53), B=@n36), False=@n22) {group=LerpTexture, parameterName=Lerp Diffuse Using Texture}
  n55 = TextureSampleParameter2D() {
   desc=Use for interior walls and ceilings
   group=Multiple_Material_Textures
   parameterName=Diffuse_Texture_2
   texture=/Engine/EngineMaterials/DefaultDiffuse
  }
  n56 = TextureSampleParameter2D() {
   desc=Use for trims and accents
   group=Multiple_Material_Textures
   parameterName=Diffuse_Texture_3
   texture=&DefaultDiffuse
  }
  n57 = TextureSampleParameter2D() {desc=Use for doors, group=Multiple_Material_Textures, parameterName=Diffuse_Texture_4, texture=&DefaultDiffuse}
  n58 = TextureSampleParameter2D() {group=Multiple_Material_Textures, parameterName=Diffuse_Texture_5, texture=&DefaultDiffuse}
  n59 = TextureSampleParameter2D() {group=Multiple_Material_Textures, parameterName=Diffuse_Texture_6, texture=&DefaultDiffuse}
  n60 = TextureSampleParameter2D() {group=Multiple_Material_Textures, parameterName=Diffuse_Texture_7, texture=&DefaultDiffuse}
  n61 = TextureSampleParameter2D() {group=Multiple_Material_Textures, parameterName=Diffuse_Texture_8, texture=&DefaultDiffuse}
  n62 = TextureSampleParameter2D() {group=Multiple_Material_Textures, parameterName=Diffuse_Texture_9, texture=&DefaultDiffuse}
  n63 = TextureSampleParameter2D() {group=Multiple_Material_Textures, parameterName=Diffuse_Texture_10, texture=&DefaultDiffuse}
  n64 = VectorParameter() {group=Diffuse, parameterName=Dot Diffuse Color, defaultValue=<rgba 1 0 0 0>}
  n65 = AppendVector(A=@n64.RGB, B=@n64.RGB)
  n66 = StaticBoolParameter() {group=Diffuse, parameterName="Dot Texture 1 (Diiffuse)"}
  n67 = DotProduct(A=@n54, B=@n65)
  n68 = Constant4Vector() {constant=<rgba 1 1 1 1>}
  n69 = StaticSwitchParameter(
   True=MaterialFunctionCall!(
    Spec1
    Texture1=@n54
    Normal1
    Normal2
    Texture2=AppendVector(A=@n55.RGB, B=@n55.RGB)
    Spec2
    Texture3=AppendVector(A=@n56.RGB, B=@n56.RGB)
    Spec3
    Normal3
    Texture4=AppendVector(A=@n57.RGB, B=@n57.RGB)
    Spec4
    Normal4
    Texture5=AppendVector(A=@n58.RGB, B=@n58.RGB)
    Normal5
    Spec5
    Texture6=AppendVector(A=@n59.RGB, B=@n59.RGB)
    Spec6
    Normal6
    Texture7=AppendVector(A=@n60.RGB, B=@n60.RGB)
    Texture8=AppendVector(A=@n61.RGB, B=@n61.RGB)
    Texture9=AppendVector(A=@n62.RGB, B=@n62.RGB)
    Use_Dot_Texture9
    Texture10=AppendVector(A=@n63.RGB, B=@n63.RGB)
    Use 2 Diffuse Textures=StaticBoolParameter() {group=Multiple_Material_Textures, parameterName=Use 2 Materials}
    --------------
    Use 3 Diffuse Textures=StaticBoolParameter() {group=Multiple_Material_Textures, parameterName=Use 3 Materials}
    Use 4 Diffuse Textures=StaticBoolParameter() {group=Multiple_Material_Textures, parameterName=Use 4 Materials}
    Use 5 Diffuse Textures=StaticBoolParameter() {group=Multiple_Material_Textures, parameterName=Use 5 Materials}
    Use 6 Diffuse Textures=StaticBoolParameter() {group=Multiple_Material_Textures, parameterName=Use 6 Materials}
    Use 7 Diffuse Textures=StaticBoolParameter() {group=Multiple_Material_Textures, parameterName=Use 7 Materials}
    Use 8 Diffuse Textures=StaticBoolParameter() {group=Multiple_Material_Textures, parameterName=Use 8 Materials}
    Use 9 Diffuse Textures=StaticBoolParameter() {group=Multiple_Material_Textures, parameterName=Use 9 Materials}
    Use 10 Diffuse Textures=StaticBoolParameter() {group=Multiple_Material_Textures, parameterName=Use 10 Materials}
    "-------------- "
    Dot_Texture1=@n65
    TransitionInfo
    Dot_Texture2
    Dot_Texture3
    Dot_Texture4
    Dot_Texture5
    Dot_Texture6
    Dot_Texture7
    Dot_Texture8
    Dot_Texture9
    Dot_Texture10
    Use_Dot_Texture1=@n66
    Use_Dot_Texture2
    Use_Dot_Texture3
    Use_Dot_Texture4
    Use_Dot_Texture5
    Use_Dot_Texture6
    Use_Dot_Texture7
    Use_Dot_Texture8
    Use_Dot_Texture10
   ) {materialFunction=/Game/Packages/Fortress_SharedMaterials/Base_Material_Functions/Fortnite_TextureSetLerp}.Diffuse
   False=StaticSwitch(True=Multiply(A=@n67, B=@n68.RGBA), False=@n54, Value=@n66)
  ) {group=Multiple_Material_Textures, parameterName=Use_Multiple_Material_Textures}
  n70 = ComponentMask(None=@n69) {b=true, g=true, r=true}
  n71 = Multiply(
   A=TextureSampleParameter2D(UVs=StaticSwitchParameter(True=@n12, False=@n11.f3 Add) {group=World Space Color Picker, parameterName=Offset Color Index with UV Channel 2 G}) {group=World Space Color Picker, parameterName=Color Chart, texture=/Game/Global/Textures/Default/ColorSets/CarColors_Interiors}.RGB
   B=@n70
  )
  n72 = TextureSampleParameter2D(UVs=@n11.f3 Add) {group=World Space Color Picker, parameterName=Color Chart 2, texture=/Game/Global/Textures/Default/ColorSets/CarColors}
  n73 = TextureSampleParameter2D() {
   group=World Space Color Picker
   parameterName=World Space Color Picker Mask
   texture=/Game/Vehicles/FORT_Vehicles_01/Materials/M_FORT_Truck_01_M_New
  }
  n74 = AppendVector(A=@n73.RGB, B=@n73.RGB)
  n75 = StaticComponentMaskParameter(None=@n74) {defaultB=true, group=World Space Color Picker, parameterName=World Space Color Picker Mask Channel}
  n76 = LinearInterpolate(
   A=LinearInterpolate(A=@n71, B=Multiply(A=@n72.RGB, B=@n70), Alpha=@n75)
   B=@n70
   Alpha=StaticComponentMaskParameter(None=@n74) {defaultG=true, group=World Space Color Picker, parameterName=World Space Color Picker 2 Mask Channel}
  )
  n77 = LinearInterpolate(A=@n71, B=@n70, Alpha=@n75)
  n78 = StaticSwitchParameter(
   True=StaticSwitchParameter(
    True=StaticSwitchParameter(True=@n76, False=@n77) {group=World Space Color Picker, parameterName=Use 2 World Position Coloring Colors}
    False=@n71
   ) {group=World Space Color Picker, parameterName=Mask World Position Color Picker}
   False=@n70
  ) {group=World Space Color Picker, parameterName=Use World Position Coloring}
  n79 = ScalarParameter() {defaultValue=1, group=Diffuse, parameterName=DiffuseDesaturationAmount}
  n80 = ComponentMask(None=StaticSwitchParameter(True=Desaturation(None=@n78, Fraction=@n79), False=@n78) {group=Diffuse, parameterName=Desaturate Diffuse}) {b=true, desc=diffuse output, g=true, r=true}
  n81 = TextureCoordinate()
  n82 = ObjectPositionWS()
  n83 = Divide(
   A=AppendVector(A=MaterialFunctionCall(f2=@n82) {materialFunction=&VectorComponentOps}.f2 Add, B=Constant())
   B=ObjectRadius()
  )
  n84 = ScalarParameter() {defaultValue=1, group=ModulateTexture, parameterName=ModTextureWorldPosOffsetRate}
  n85 = StaticSwitchParameter(True=Add(A=@n81, B=Multiply(A=@n83, B=@n84)), False=@n81) {defaultValue=true, group=ModulateTexture, parameterName=WorldPosOffsetModulateTexture}
  n86 = VectorParameter() {group=ModulateTexture, parameterName=ModTexture RG_XYScaling B_AddToTexture A_MultiplyTexture, defaultValue=<rgba 1 1 0 0>}
  n87 = ComponentMask(None=@n86.RGB) {g=true, r=true}
  n88 = TextureSampleParameter2D(UVs=Add(A=Multiply(A=@n85, B=@n87), B=AppendVector(A=@n86.B, B=@n86.B))) {group=ModulateTexture, parameterName=ModulateDiffuseTexture}
  n89 = AppendVector(A=@n88.RGB, B=@n88.RGB)
  n90 = StaticSwitchParameter(
   True=Multiply(
    A=@n80
    B=StaticComponentMaskParameter(None=@n89) {defaultB=true, defaultG=true, defaultR=true, group=ModulateTexture, parameterName=ModulateDiffuseTexture}
   )
   False=@n80
  ) {desc="< Diffuse", group=ModulateTexture, parameterName=Texture Modulate Diffuse}
  n91 = VectorParameter() {group=Diffuse, parameterName=DiffuseColorMultiplier, defaultValue=<rgba 1 1 1 1>}
  n92 = Add(A=StaticSwitchParameter(True=Multiply(A=@n90, B=@n91.RGB), False=@n90) {group=Diffuse, parameterName=Modulate Diffuse}, B=Constant()) {desc="< Diffuse"}
  n93 = ScalarParameter() {defaultValue=1, group=reflection, parameterName=Reflection Fresnel Multiplier}
  n94 = TextureCoordinate()
  n95 = VectorParameter() {group=Normal, parameterName=Normal Offset RG Tiling BA, defaultValue=<rgba 0 0 1 1>}
  n96 = Multiply(A=Add(A=@n94, B=ComponentMask(None=@n95.RGB) {g=true, r=true}), B=AppendVector(A=@n95.B, B=@n95.B))
  n97 = StaticSwitchParameter(
   True=TextureSampleParameter2D(UVs=StaticSwitchParameter(True=@n96, False=@n94) {group=Normal, parameterName=Manipulate Normal UVS}) {group=Normal, parameterName=Normals, samplerType=SAMPLERTYPE_Normal, texture=/Engine/EngineMaterials/DefaultNormal}.RGB
   False=Constant3Vector() {constant=<rgba 0 0 1 0>}
  ) {group=Normal, parameterName=UseNormalTexture}
  n98 = Constant3Vector() {constant=<rgba 0 0 1 0>}
  n99 = StaticSwitchParameter(
   True=LinearInterpolate(A=@n97, B=@n98, Alpha=ScalarParameter() {defaultValue=0.5, group=Normal, parameterName=Soften Normal Amount})
   False=@n97
  ) {group=Normal, parameterName=Soften Normals}
  n100 = TextureCoordinate()
  n101 = VectorParameter() {group=Normal, parameterName=Lerp Normal Offset RG Tiling BA, defaultValue=<rgba 0 0 1 1>}
  n102 = Add(A=@n100, B=ComponentMask(None=@n101.RGB) {g=true, r=true})
  n103 = AppendVector(A=@n101.B, B=@n101.B)
  n104 = TextureSampleParameter2D(UVs=StaticSwitchParameter(True=Multiply(A=@n102, B=@n103), False=@n100) {group=Normal, parameterName=Manipulate Lerp Normal UVS}) {
   group=Normal
   parameterName=Second Normals for Lerp
   samplerType=SAMPLERTYPE_LinearColor
   texture=/Engine/EngineMaterials/DefaultNormal_Uncompressed
  }
  n105 = LinearInterpolate(A=@n104.RGB, B=@n98, Alpha=ScalarParameter() {defaultValue=0.5, group=Normal, parameterName=Soften Lerp Normal Amount})
  n106 = StaticSwitchParameter(
   True=LinearInterpolate(
    A=@n99
    B=StaticSwitchParameter(True=@n105, False=@n104.RGB) {group=Normal, parameterName=Soften Lerp Normals}
    Alpha=StaticSwitchParameter(True=@n41.Lerp Alpha, False=@n52) {group=LerpTexture, parameterName=Use Height Map for Lerp Alpha}
   )
   False=@n99
  ) {group=LerpTexture, parameterName=Lerp Normals}
  n107 = StaticSwitchParameter(True=Multiply(A=@n106, B=VectorParameter() {group=Normal, parameterName=Normal Modulation, defaultValue=<rgba 1 1 1 1>}.RGB), False=@n106) {group=Normal, parameterName=ModulateNormals}
  n108 = TextureCoordinate()
  n109 = AppendVector(
   A=ScalarParameter() {defaultValue=1, group=DetailNormalMap, parameterName=DetailNormalMapTilingX}
   B=ScalarParameter() {defaultValue=1, group=DetailNormalMap, parameterName=DetailNormalMapTilingY}
  )
  n110 = TextureSampleParameter2D(UVs=StaticSwitchParameter(True=Multiply(A=@n108, B=@n109), False=@n108) {group=DetailNormalMap, parameterName=Manipulate DetailMap UVS}) {group=DetailNormalMap, parameterName=DetailNormalMap, samplerType=SAMPLERTYPE_Normal, texture=&DefaultNormal}
  n111 = ScalarParameter() {defaultValue=1, group=DetailNormalMap, parameterName=DetailNormalStrength}
  n112 = AppendVector(A=AppendVector(A=@n111, B=@n111), B=Constant())
  n113 = StaticSwitchParameter(True=Add(A=@n107, B=Multiply(A=@n110.RGB, B=@n112)), False=@n107) {desc=Tangent Normal, group=DetailNormalMap, parameterName=DetailNormalMap}
  n114 = StaticSwitchParameter(True=@n113, False=@n107) {group=DetailNormalMap, parameterName=Use Detail Map in Fresnel}
  n115 = StaticSwitchParameter(
   True=LinearInterpolate(
    A=VertexNormalWS()
    B=StaticSwitchParameter(True=@n114, False=@n113) {group=DetailNormalMap, parameterName=DetailNormalMap}
    Alpha=ScalarParameter() {defaultValue=1, group=reflection, parameterName=Reflection Fresnel Smoothness}
   )
   False=@n113
  ) {group=reflection, parameterName=Smooth Reflection Fresnel}
  n116 = CameraVectorWS()
  n117 = Clamp(None=DotProduct(A=Transform(None=@n115), B=@n116))
  n118 = DotProduct(A=@n116, B=VertexNormalWS())
  n119 = Multiply(
   A=Power(Base=OneMinus(None=@n117), Exp=ScalarParameter() {defaultValue=2, group=reflection, parameterName=Reflection Fresnel Power})
   B=OneMinus(None=Clamp(None=@n118)) {desc=should probably remove}
  )
  n120 = Add(
   A=StaticSwitchParameter(True=Multiply(A=@n93, B=@n119), False=@n119) {group=reflection, parameterName=Modulate Reflection Fresnel}
   B=ScalarParameter() {defaultValue=1, group=Normal, parameterName=Reflection Fresnel Offset}
  )
  n121 = Constant() {r=1}
  n122 = ReflectionVectorWS()
  n123 = Transform(None=ReflectionVectorWS()) {transformSourceType=TRANSFORMSOURCE_World, transformType=TRANSFORM_View}
  n124 = Multiply(A=ConstantBiasScale(None=ComponentMask(None=@n123) {g=true, r=true}), B=Constant2Vector() {g=-1, r=1})
  n125 = Multiply(
   A=Multiply(
    A=StaticSwitchParameter(True=@n120, False=@n121) {group=reflection, parameterName=Use Reflection Fresnel}
    B=ScalarParameter() {defaultValue=1, group=reflection, parameterName=Reflection Multiplier}
   )
   B=StaticSwitchParameter(
    True=TextureSampleParameterCube!(UVs=@n122, Apply View MipBias) {group=reflection, parameterName=Reflection CubeMap}.RGB
    False=TextureSampleParameter2D(UVs=@n124) {group=reflection, parameterName=SphereReflectionMap}.RGB
   ) {group=reflection, parameterName=CubeMap Or ReflectionProbe}
  ) {desc=Image * multiplier * fresnel}
  n126 = TextureCoordinate()
  n127 = VectorParameter() {group=Normal, parameterName=Normal Offset RG Tiling BA, defaultValue=<rgba 0 0 1 1>}
  n128 = Multiply(A=Add(A=@n126, B=ComponentMask(None=@n127.RGB) {g=true, r=true}), B=AppendVector(A=@n127.B, B=@n127.B))
  n129 = StaticSwitchParameter(
   True=@n90
   False=TextureSampleParameter2D(UVs=StaticSwitchParameter(True=@n128, False=@n126) {group=Specular, parameterName=Manipulate Specular UVs}) {group=Specular, parameterName=Specular}.RGB
  ) {defaultValue=true, group=Specular, parameterName=Use Diffuse as Spec}
  n130 = TextureCoordinate()
  n131 = VectorParameter() {group=LerpTexture, parameterName=ReflectionMask Offset RG Tiling BA, defaultValue=<rgba 0 0 1 1>}
  n132 = Add(A=@n130, B=ComponentMask(None=@n131.RGB) {g=true, r=true})
  n133 = AppendVector(A=@n131.B, B=@n131.B)
  n134 = TextureSampleParameter2D(UVs=StaticSwitchParameter(True=Multiply(A=@n132, B=@n133), False=@n130) {group=reflection, parameterName=Manipulate ReflectionMask UVS}) {group=reflection, parameterName=ReflectionMask}
  n135 = AppendVector(A=@n134.RGB, B=@n134.RGB)
  n136 = StaticSwitchParameter(True=@n129, False=@n135) {group=Specular, parameterName=Use Spec as Reflection Mask}
  n137 = StaticComponentMaskParameter(None=StaticSwitchParameter(True=@n90, False=StaticSwitchParameter(True=@n136, False=@n135) {group=Specular, parameterName=Use Spec}) {group=Specular, parameterName=Use Diffuse as Reflection Mask}) {defaultR=true, group=reflection, parameterName=ReflectionMaskChannel}
  n138 = MaterialFunctionCall(In=@n137, Contrast=ScalarParameter() {group=reflection, parameterName=Reflection Mask Cheap Contrast}) {materialFunction=&CheapContrast}
  n139 = StaticSwitchParameter(
   True=StaticSwitchParameter(True=Clamp(None=@n138.Result), False=@n137) {group=reflection, parameterName=Apply Cheap Contrast To Reflection Mask}
   False=ScalarParameter() {defaultValue=0.5, group=reflection, parameterName=Reflection Amount}
  ) {group=reflection, parameterName=Mask Reflection Map With Image}
  n140 = Multiply(A=@n125, B=@n139)
  n141 = StaticSwitchParameter(
   True=StaticSwitchParameter(True=LinearInterpolate(A=@n92, B=@n125, Alpha=@n139) {desc="a=0 b=1"}, False=Add(A=@n92, B=@n140)) {group=reflection, parameterName=Lerp Blend Or Add Reflection}
   False=@n92
  ) {desc=Diffuse, group=reflection, parameterName=Use Reflection Map}
  n142 = Multiply(A=TextureCoordinate(), B=ScalarParameter() {defaultValue=1.5, group=CustomDamage, parameterName=Custom Damage Texture Tiling})
  n143 = StaticSwitchParameter(
   True=Multiply(
    A=@n141
    B=TextureSampleParameter2D(UVs=@n142) {group=CustomDamage, parameterName=Custom Damage Diffuse Texture, texture=/Game/Packages/Fortress_ShaderTests/Textures/charcoal}.RGB
   )
   False=VectorParameter() {group=CustomDamage, parameterName=Custom Damage Diffuse Color, defaultValue=<rgba 0 0 0 1>}.RGB
  ) {group=CustomDamage, parameterName=Use Custom Damage Diffuse Texture}
  n144 = Multiply(A=@n2, B=ScalarParameter() {defaultValue=0.4, group=CustomDamage, parameterName=Custom Damage DSSm Height Max}) {constB=0.1}
  n145 = Clamp(None=Multiply(A=MaterialFunctionCall(Transition Phase=@n144, Height Texture=@n4) {materialFunction=/Engine/Functions/Engine_MaterialFunctions02/Texturing/HeightLerp}.Lerp Alpha No Contrast) {constB=10})
  n146 = MaterialFunctionCall!(
   DiffuseFresnelAmount=ScalarParameter() {defaultValue=0.75, group=Diffuse, parameterName=DiffuseFresnelAmount}
   Emissive=StaticSwitchParameter(True=StaticSwitchParameter(True=@n8, False=@n7.Result) {group=Emissive2, parameterName=Use Emissive 2}, False=Constant()) {group=Emissive, parameterName=Use Emissive}
   Base Color=StaticSwitchParameter(True=LinearInterpolate(A=@n141, B=@n143, Alpha=@n145) {constB=0, desc=1 is damaged}, False=@n141) {desc=d, group=CustomDamage, parameterName=Use Custom Damage}
  ) {desc=diffuse, materialFunction=/Game/Packages/Fortress_SharedMaterials/Base_Material_Functions/FresnelFunction}
  n147 = ScalarParameter() {defaultValue=1, group=Specular, parameterName=SpecLevel}
  n148 = StaticSwitchParameter(True=Desaturation(None=@n129, Fraction=ScalarParameter() {defaultValue=1, group=Specular, parameterName=Spec Desaturation}), False=@n129) {group=Specular, parameterName=Desaturate Spec Texture}
  n149 = MaterialFunctionCall(In=@n148, Contrast=ScalarParameter() {group=Specular, parameterName=Spec Cheap Contrast Amount}) {materialFunction=&CheapContrast}
  n150 = VectorParameter() {group=Specular, parameterName=SpecColor, defaultValue=<rgba 1 1 1 1>}
  n151 = Multiply(A=StaticSwitchParameter(True=@n149.Result, False=@n148) {group=Specular, parameterName=Use Cheap Contrast On Spec}, B=@n150.RGB)
  n152 = StaticSwitchParameter(True=@n149.Result, False=@n148) {group=Specular, parameterName=Use Cheap Contrast On Spec}
  n153 = Multiply(
   A=@n147
   B=StaticSwitchParameter(True=StaticSwitchParameter(True=@n151, False=@n152) {group=Specular, parameterName=Modulate Spec Color}, False=@n150.RGB) {group=Specular, parameterName=Use Spec Texture}
  )
  n154 = LinearInterpolate(A=@n153, B=ScalarParameter() {defaultValue=10, group=CustomDamage, parameterName=Custom Damage Spec Level}, Alpha=@n145) {constA=1}
  n155 = ScalarParameter() {defaultValue=1, group=Code, parameterName=SuperChargedFlickerBias}
  n156 = ScalarParameter() {group=Code, parameterName=SuperChargedEffect}
  n157 = StaticBool()
  n158 = LinearInterpolate(
   A=@n113
   B=MaterialFunctionCall() {materialFunction=/Engine/Functions/Engine_MaterialFunctions02/WorldPositionOffset/StaticMeshMorphTargets}.Morph Target 1 Normals
   Alpha=ScalarParameter() {defaultValue=1, group=MorphTargets, parameterName=MorphTargetStrength}
  )
  n159 = MaterialFunctionCall!(
   S
   V2
   V3=StaticSwitchParameter(True=StaticSwitchParameter(True=@n158, False=@n113) {group=MorphTargets, parameterName=MorphNormals}, False=@n113) {group=MorphTargets, parameterName=UseMorphTargets}
   V4
  ) {desc=Normals, materialFunction=/Engine/Functions/Engine_MaterialFunctions02/PassThrough}
  n160 = TextureSampleParameter2D(UVs=Multiply(A=TextureCoordinate(), B=ScalarParameter() {defaultValue=1, group=CustomDamage, parameterName=Custom Damage Height Texture Tiling})) {
   group=CustomDamage
   parameterName=Custom Damage Normal
   samplerType=SAMPLERTYPE_Normal
   texture=/Game/Global/Textures/Default/Tileable_Patterns/CarDamageNormal
  }
  n161 = Multiply(
   A=@n4
   B=ScalarParameter() {defaultValue=10, group=CustomDamage, parameterName=Custom Damage Normal Map Height Map Final Reveal Mask Multiplier}
  ) {constB=10}
  n162 = LinearInterpolate(
   A=@n159.V3
   B=MaterialFunctionCall(BaseNormal=@n159.V3, AdditionalNormal=@n160.RGB) {materialFunction=&BlendAngleCorrectedNormals}.Result
   Alpha=Multiply(A=@n5.Lerp Alpha No Contrast, B=Clamp(None=@n161))
  )
  n163 = MaterialFunctionCall!(
   Use Transylvania Moss=StaticBoolParameter() {group=Zone Specific, parameterName=Use Transylvania Moss}
   Transition Phase
   Base Color=@n146.d
   SpecularMasks=StaticSwitchParameter(True=StaticSwitchParameter(True=@n154, False=@n153) {group=CustomDamage, parameterName=Use Custom Damage}, False=Constant()) {desc=Specular, group=Specular, parameterName=Use Spec}
   Emissive=StaticSwitchParameter(
    True=MaterialFunctionCall!(AverageHandPosition, SuperChargedFlickerBias=@n155, SuperChargedEffect=@n156, UseDistanceCalc=@n157, Emissive=@n146.Emissive) {materialFunction=/Game/Characters/Fortress_CharacterShader/SuperChargedGlow}.Result
    False=@n146.Emissive
   ) {desc=emissive, group=Special Cases, parameterName=Is a Melee Weapon}
   Normal=StaticSwitchParameter(
    True=StaticSwitchParameter(True=@n162, False=@n159.V3) {desc=N, group=CustomDamage, parameterName=Use Custom Damage Normal Map}
    False=@n159.V3
   ) {desc=Normal, group=CustomDamage, parameterName=Use Custom Damage}
   Use World Space Normals=StaticBool()
  ) {materialFunction=/Game/Packages/Fortress_SharedMaterials/Base_Material_Functions/Moss}
  n164 = MaterialFunctionCall(DiffuseColor=@n163.Base Color, SpecularColor=@n163.Base Color) {materialFunction=&ConvertFromDiffSpec}
  n165 = ComponentMask(None=@n69) {a=true}
  n166 = TextureCoordinate()
  n167 = VectorParameter() {group=Mask, parameterName=Opacity Mask Offset RG Tiling BA, defaultValue=<rgba 0 0 1 1>}
  n168 = Add(A=@n166, B=ComponentMask(None=@n167.RGB) {g=true, r=true})
  n169 = AppendVector(A=@n167.B, B=@n167.B)
  n170 = TextureSampleParameter2D(UVs=StaticSwitchParameter(True=Multiply(A=@n168, B=@n169), False=@n166) {group=Mask, parameterName=Manipulate Opacity Mask Uvs}) {group=Mask, parameterName=OpacityMask, texture=/Game/Global/Textures/Default/Masks_and_Gradients/CheckerboardTexture}
  n171 = StaticComponentMaskParameter(None=AppendVector(A=@n170.RGB, B=@n170.RGB)) {defaultR=true, group=Mask, parameterName=Opacity Mask Channel}
  n172 = Multiply(A=@n2, B=ScalarParameter() {defaultValue=5, group=CustomDamage, parameterName=Custom Damage Morph Target Speed})
  n173 = Transform(None=Subtract(A=VertexColor()) {constB=0.5}) {transformSourceType=TRANSFORMSOURCE_Local}
  n174 = Constant()
  n175 = StaticSwitchParameter(True=Multiply(A=Clamp(None=@n172), B=Multiply(A=@n173) {constB=32, desc=Conversion Factor}), False=@n174) {group=CustomDamage, parameterName=Use a Morph Target With Custom Damage}
  n176 = Add(
   A=MaterialFunctionCall!(
    WSDeformationCenter=VectorParameter() {group=z_code, parameterName=WSDeformationCenter}.RGB
    Code_DestructionAmount=ScalarParameter() {group=z_code, parameterName=Code_DestructionAmount}
    Code_UseUvForDamage
    Code_UseHitEffects=StaticBoolParameter() {defaultValue=true, group=z_code, parameterName=Code_UseHitEffects}
    Code_Emissive
    WSDamageRadius=ScalarParameter() {group=z_code, parameterName=WSDamageRadius}
    CrackDamageShader
    Diffuse
    ScalarModDamageCracks
    ScalarOffsetforDamageCracks
    ExposeLerpFunctionForInvertingTheHitDistanceCalc
    InvertDamageFalloff
    CrackWorldPosition
    WorldPosition
    DistanceCullFadeAnimationOffset=ScalarParameter() {group=WorldPositionOffset, parameterName=DistanceCullFadeAnimationOffset}
   ) {materialFunction=/Game/Packages/jl_tests/Fortnite_HitEffects}.Hit And Shrink WPO
   B=MaterialFunctionCall!(Use Tree Vertex Shader=StaticBoolParameter() {group=Special Cases, parameterName=Use Tree Vertex Shader}) {materialFunction=/Game/Packages/Fortress_SharedMaterials/Base_Material_Functions/FortniteTreeAnimation_VS}.Result
  )
  n177 = MaterialFunctionCall() {materialFunction=&StaticMeshMorphTargets}
  n178 = ScalarParameter() {group=MorphTargets, parameterName=MorphTargetStrength}
  n179 = StaticSwitchParameter(True=Add(A=@n176, B=Multiply(A=@n177.Morph Target 1 WPO, B=@n178)) {desc=7 instructions}, False=@n176) {group=MorphTargets, parameterName=UseMorphTargets}
  n180 = VectorParameter() {group=WorldPositionOffset, parameterName=WorldPositionOffset_ModulatedByVertexColor, defaultValue=<rgba 0 0 0 1>}
  n181 = VertexColor()
  n182 = AppendVector(A=@n181, B=@n181)
  n183 = Add(
   A=@n179
   B=Multiply(
    A=@n180.RGB
    B=StaticComponentMaskParameter(None=@n182) {defaultA=true, group=WorldPositionOffset, parameterName=WorldPositionOffsetVertexColorModulationChannel}
   )
  )
  n184 = SetMaterialAttributes!(
   MaterialAttributes
   Base Color=@n164.BaseColor
   Metallic=@n164.BaseColor
   Specular=@n164.BaseColor
   Emissive Color=@n163.Emissive
   Opacity Mask=StaticSwitchParameter(
    True=StaticSwitchParameter(True=@n165, False=@n171) {group=Mask, parameterName="Use Diffuse Texture Alpha As Opacity Mask (Masked Instances Only)"}
    False=Constant() {r=1}
   ) {desc=Mask, group=Mask, parameterName=Use Mask}
   Normal=@n163.Emissive
   World Position Offset=Add(
    A=StaticSwitchParameter(True=@n175, False=@n174) {desc=WPO, group=CustomDamage, parameterName=Use Custom Damage}
    B=StaticSwitchParameter(True=@n183, False=@n179) {group=WorldPositionOffset, parameterName=UseWorldPositionOffsetModulatedByVertexColors}
   ) {desc=WPO}
  ) {attributeSetTypes [
    #1
    #2
    #3
    #4
    #5
    #6
    #7
   ]}
  MaterialFunctionCall!(
   In=MaterialFunctionCall(MA In=@n184) {materialFunction=&MF_CreativeMoveTool}.MA Out
   DestructionAmount
  ) {materialFunction=/Game/Packages/jl_tests/CrackMaterial/MF_Crack}
  n185 = TextureCoordinate()
  n186 = VectorParameter() {
   group=Specular
   parameterName="TilingTransitioningSpec_UVs  RG=Tiling Settings for Near and BA=Tiling Settings for far"
   defaultValue=<rgba 2 2 4 4>
  }
  n187 = TextureSampleParameter2D(UVs=Multiply(A=@n185, B=ComponentMask(None=@n186.RGB) {g=true, r=true})) {group=Specular, parameterName=TilingTransitioningSpec_Near}
  n188 = TextureSampleParameter2D(UVs=Multiply(A=@n185, B=AppendVector(A=@n186.B, B=@n186.B))) {group=Specular, parameterName=TilingTransitioningSpec_Near}
  n189 = ScalarParameter() {defaultValue=128, desc=world space distance, group=Specular, parameterName=TilingTransitioningSpec_TransitionStartDistance}
  n190 = Multiply(
   A=Subtract(A=PixelDepth().R, B=@n189)
   B=ScalarParameter() {defaultValue=0.002, desc=1/distance in world space, group=Specular, parameterName=TilingTransitioningSpec_FadeDistance}
  )
  Multiply(
   A=LinearInterpolate(
    A=StaticComponentMaskParameter(None=@n187.RGB) {defaultR=true, group=Specular, parameterName=TilingTransitioningSpec_TextureChannel}
    B=StaticComponentMaskParameter(None=@n188.RGB) {defaultR=true, parameterName=TilingTransitioningSpec_TextureChannel}
    Alpha=Clamp(None=@n190)
   )
   B=ScalarParameter() {defaultValue=5, group=Specular, parameterName=TilingTransitioningSpec_Intensity}
  )
  PixelNormalWS()
  ScalarParameter() {group=z_code, parameterName=Code_HitAnimation}
  n191 = VertexNormalWS()
  n192 = PixelNormalWS()
  n193 = ScalarParameter() {defaultValue=1, group=Specular, parameterName=Spec Fresnel Smoothness}
  n194 = Fresnel(
   ExponentIn=ScalarParameter() {defaultValue=3, group=Specular, parameterName=Spec Fresnel Exponent}
   BaseReflectFractionIn=ScalarParameter() {group=Specular, parameterName=Spec Fresnel Bias}
   Normal=StaticSwitchParameter(True=LinearInterpolate(A=@n191, B=@n192, Alpha=@n193), False=@n192) {group=Specular, parameterName=Smooth Spec Fresnel}
  )
  StaticSwitchParameter(True=LinearInterpolate(
    A=ScalarParameter() {defaultValue=5, group=Specular, parameterName=SpecRoughnessFacing}
    B=ScalarParameter() {defaultValue=16, group=Specular, parameterName=SpecRoughnessFresnel}
    Alpha=@n194
   )) {group=Specular, parameterName=Use Spec Power Fresnel}
  StaticSwitchParameter(True=Multiply(A=@n147, B=@n194), False=@n147) {group=Specular, parameterName=Use Spec Fresnel}
  n195 = ScalarParameter() {defaultValue=1, group=Specular, parameterName=SpecRoughness}
  n196 = TextureSampleParameter2D() {group=Specular, parameterName=SpecRoughnessMaskTexture}
  n197 = StaticComponentMaskParameter(None=AppendVector(A=@n196.RGB, B=@n196.RGB)) {defaultR=true, group=Specular, parameterName=SpecRoughnessMaskChannel}
  n198 = StaticSwitchParameter(
   True=StaticSwitchParameter(True=Multiply(A=@n195, B=@n197), False=@n195) {desc=UseSpecPowerMask, group=Specular, parameterName=UseSpecRoughnessMask}
   False=Constant()
  ) {desc=Specular Power, group=Specular, parameterName=Use Spec}
  StaticSwitchParameter(
   True=LinearInterpolate(A=@n198, B=ScalarParameter() {defaultValue=10, group=CustomDamage, parameterName=Custom Damage Spec Power}, Alpha=@n145)
   False=@n198
  ) {desc=Spec Roughness, group=CustomDamage, parameterName=Use Custom Damage}
  VectorParameter() {group=z_code, parameterName=Code_DeformationVector, defaultValue=<rgba 0 0 1 0>}
 ]
}
