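# Fortress shared base material for skeletal meshes, a Default Lit parent whose features are toggled by static switches.
# Up to four texture sets blend into base color, normals and spec/metal/roughness masks, with optional world-aligned projection and color-chart tinting picked by actor position.
# Two optional animated emissive layers, a fresnel function, opacity masking and optional cloth motion are layered on.
# Hit-damage deformation drives world position offset, and the final attributes pass through MF_CreativeMoveTool.
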Material /Game/Packages/Fortress_SharedMaterials/Base_Material/Fortress_BaseMaterial_SkeletalMeshes
graph {
 properties {
  bAutomaticallySetUsageInEditor=false
  bUseMaterialAttributes=true
  bUsedWithMeshDeformer=true
  bUsedWithMorphTargets=true
  bUsedWithSkeletalMesh=true
  pixelDepthOffsetMode=PDOM_Legacy
  usedShadingModels=Default Lit
 }
 nodes [
  n1 = VectorParameter() {group=Emissive, parameterName="EmissiveUVPositioning (RG)UpperLeft (BA)LowerRight", defaultValue=<rgba 0 0 1 1>}
  n2 = TextureCoordinate() {coordinateIndex=1}
  n3 = ScalarParameter() {
   defaultValue=0.1
   desc="1/number of colors. ie 1/10 = .1"
   group=World Space Color Picker
   parameterName="Number Of Colors in Color Chart (1/Number of colors)"
  }
  n4 = Divide(A=ActorPositionWS(), B=ScalarParameter() {defaultValue=16.71, group=World Space Color Picker, parameterName="Color Change Frequency (WS)"})
  n5 = MaterialFunctionCall!(f2=ComponentMask(None=@n4) {g=true, r=true}, f3=@n4, f4) {materialFunction=/Game/Packages/Fortnite_UniqueMaterials_1/VectorComponentOps}
  n6 = StaticSwitchParameter(
   True=ScalarParameter() {group=World Space Color Picker, parameterName=World Position Coloring Override Value}
   False=StaticSwitchParameter(True=@n5.f3 Add, False=@n5.f3 Add) {defaultValue=true, group=World Space Color Picker, parameterName="Use XYZ(T) or XY(F) to Pick World Position Coloring Colors"}
  ) {group=World Space Color Picker, parameterName=Override World Position Coloring Color Picker with a Float Value}
  n7 = Add(A=Multiply(A=ComponentMask(None=@n2) {g=true}, B=@n3), B=@n6)
  n8 = MaterialFunctionCall!(Use UV 2=StaticBoolParameter() {group=Diffuse, parameterName=Mat 1 Use UV2}) {materialFunction=/Game/Packages/Fortress_SharedMaterials/Base_Material_Functions/UseUV1or2}
  n9 = VectorParameter() {group=Diffuse, parameterName=Base Color Offset RG Tiling BA, defaultValue=<rgba 0 0 1 1>}
  n10 = ComponentMask(None=@n9.RGB) {g=true, r=true}
  n11 = StaticSwitchParameter(True=Multiply(A=Add(A=@n8.Result, B=@n10), B=AppendVector(A=@n9.B, B=@n9.B)), False=@n8.Result) {group=Diffuse, parameterName=Manipulate Base Color and Spec UVS}
  n12 = TextureSampleParameter2D(UVs=@n11) {
   group=Specular
   parameterName=SpecularMasks
   samplerType=SAMPLERTYPE_Masks
   texture=&T_Fortnite_Default_S
  }
  n13 = VectorParameter() {group=Diffuse, parameterName=Base Color World Position Coordinate Scale, defaultValue=<rgba 256 256 256 1>}
  n14 = MaterialFunctionCall(
   TextureObject=TextureObjectParameter() {desc=Use for exterior walls and floors, group=Diffuse, parameterName=Diffuse}
   TextureSize=@n13.RGB
   WorldPosition=WorldPosition() {worldPositionShaderOffset=WPT_ExcludeAllShaderOffsets}.XYZ
   Export Float 4=StaticBool() {value=true}
  ) {materialFunction=&WorldAlignedTexture}
  n15 = StaticSwitchParameter(True=@n14.XY Texture, False=@n14.XY Texture) {group=Diffuse, parameterName=Project Base Color On XY or Z if false}
  n16 = TextureSampleParameter2D(UVs=@n11) {desc=Use for exterior walls and floors, group=Diffuse, parameterName=Diffuse}
  n17 = VectorParameter() {group=Diffuse, parameterName=Base Color, defaultValue=<rgba 0.5 0.5 0.5 1>}
  n18 = StaticSwitchParameter(
   True=StaticSwitchParameter(
    True=StaticSwitchParameter(True=@n14.XYZ Texture, False=@n15) {group=Diffuse, parameterName=Project Base Color on XYZ}
    False=AppendVector(A=@n16.RGB, B=@n16.RGB)
   ) {group=Diffuse, parameterName=Project Base Color Via World Position}
   False=AppendVector(A=@n17.RGB, B=@n17.RGB)
  ) {defaultValue=true, group=Diffuse, parameterName="Use a Base Color Texture "}
  n19 = MaterialFunctionCall(Use UV 2=StaticBoolParameter() {group=Diffuse, parameterName=Mat 1 Use UV2}) {materialFunction=&UseUV1or2}
  n20 = VectorParameter() {group=Normal, parameterName=Normal Offset RG Tiling BA, defaultValue=<rgba 0 0 1 1>}
  n21 = Multiply(A=Add(A=@n19.Result, B=ComponentMask(None=@n20.RGB) {g=true, r=true}), B=AppendVector(A=@n20.B, B=@n20.B))
  n22 = MaterialFunctionCall(Use UV 2=StaticBoolParameter() {group=Multiple_Material_Textures, parameterName=Mat 2 Use UV2}) {materialFunction=&UseUV1or2}
  n23 = TextureSampleParameter2D(UVs=@n22.Result) {
   desc=Use for interior walls and ceilings
   group=Multiple_Material_Textures_Normal
   parameterName=Normals_Texture_2
   samplerType=SAMPLERTYPE_Normal
   texture=/Engine/EngineMaterials/FlatNormal
  }
  n24 = TextureSampleParameter2D(UVs=@n22.Result) {
   desc=Use for interior walls and ceilings
   group=Multiple_Material_Textures_Diffuse
   parameterName=Diffuse_Texture_2
   texture=/Engine/EngineMaterials/DefaultDiffuse
  }
  n25 = TextureSampleParameter2D(UVs=@n22.Result) {
   desc=Use for interior walls and ceilings
   group=Multiple_Material_Textures_SpecAndRoughness
   parameterName=SpecularMasks_2
   samplerType=SAMPLERTYPE_Masks
   texture=&T_Fortnite_Default_S
  }
  n26 = MaterialFunctionCall(Use UV 2=StaticBoolParameter() {group=Multiple_Material_Textures, parameterName=Mat 3 Use UV2}) {materialFunction=&UseUV1or2}
  n27 = TextureSampleParameter2D(UVs=@n26.Result) {
   desc=Use for trims and accents
   group=Multiple_Material_Textures_Diffuse
   parameterName=Diffuse_Texture_3
   texture=&DefaultDiffuse
  }
  n28 = TextureSampleParameter2D(UVs=@n26.Result) {
   desc=Use for trims and accents
   group=Multiple_Material_Textures_SpecAndRoughness
   parameterName=SpecularMasks_3
   samplerType=SAMPLERTYPE_Masks
   texture=&T_Fortnite_Default_S
  }
  n29 = TextureSampleParameter2D(UVs=@n26.Result) {
   desc=Use for trims and accents
   group=Multiple_Material_Textures_Normal
   parameterName=Normals_Texture_3
   samplerType=SAMPLERTYPE_Normal
   texture=&FlatNormal
  }
  n30 = MaterialFunctionCall(Use UV 2=StaticBoolParameter() {group=Multiple_Material_Textures, parameterName=Mat 4 Use UV2}) {materialFunction=&UseUV1or2}
  n31 = TextureSampleParameter2D(UVs=@n30.Result) {
   desc=Use for doors
   group=Multiple_Material_Textures_Diffuse
   parameterName=Diffuse_Texture_4
   texture=&DefaultDiffuse
  }
  n32 = TextureSampleParameter2D(UVs=@n30.Result) {
   desc=Use for doors
   group=Multiple_Material_Textures_SpecAndRoughness
   parameterName=SpecularMasks_4
   samplerType=SAMPLERTYPE_Masks
   texture=&T_Fortnite_Default_S
  }
  n33 = TextureSampleParameter2D(UVs=@n30.Result) {
   desc=Use for doors
   group=Multiple_Material_Textures_Normal
   parameterName=Normals_Texture_4
   samplerType=SAMPLERTYPE_Normal
   texture=&FlatNormal
  }
  n34 = VectorParameter() {group=Diffuse, parameterName=Dot Base Color Channel, defaultValue=<rgba 1 0 0 0>}
  n35 = AppendVector(A=@n34.RGB, B=@n34.RGB)
  n36 = StaticBoolParameter() {group=Diffuse, parameterName="Dot Texture 1 (Base Color)"}
  n37 = MaterialFunctionCall!(
   Spec1=@n12.RGB
   Texture1=@n18
   Normal1=TextureSampleParameter2D(UVs=StaticSwitchParameter(True=@n21, False=@n19.Result) {group=Normal, parameterName=Manipulate Normal UVS}) {group=Normal, parameterName=Normals, samplerType=SAMPLERTYPE_Normal, texture=&FlatNormal}.RGB
   Normal2=@n23.RGB
   Texture2=AppendVector(A=@n24.RGB, B=@n24.RGB)
   Spec2=@n25.RGB
   Texture3=AppendVector(A=@n27.RGB, B=@n27.RGB)
   Spec3=@n28.RGB
   Normal3=@n29.RGB
   Texture4=AppendVector(A=@n31.RGB, B=@n31.RGB)
   Spec4=@n32.RGB
   Normal4=@n33.RGB
   Texture5
   Normal5
   Spec5
   Texture6
   Spec6
   Normal6
   Texture7
   Texture8
   Texture9
   Use_Dot_Texture9
   Texture10
   Use 2 Diffuse Textures=StaticBoolParameter() {group=Multiple_Material_Textures, parameterName=Use 2 Materials}
   --------------
   Use 3 Diffuse Textures=StaticBoolParameter() {group=Multiple_Material_Textures, parameterName=Use 3 Materials}
   Use 4 Diffuse Textures=StaticBoolParameter() {group=Multiple_Material_Textures, parameterName=Use 4 Materials}
   Use 5 Diffuse Textures
   Use 6 Diffuse Textures
   Use 7 Diffuse Textures
   Use 8 Diffuse Textures
   Use 9 Diffuse Textures
   Use 10 Diffuse Textures
   "-------------- "
   Dot_Texture1=@n35
   TransitionInfo
   Dot_Texture2
   Dot_Texture3
   Dot_Texture4
   Dot_Texture5
   Dot_Texture6
   Dot_Texture7
   Dot_Texture8
   Dot_Texture9
   Dot_Texture10
   Use_Dot_Texture1=@n36
   Use_Dot_Texture2
   Use_Dot_Texture3
   Use_Dot_Texture4
   Use_Dot_Texture5
   Use_Dot_Texture6
   Use_Dot_Texture7
   Use_Dot_Texture8
   Use_Dot_Texture10
  ) {materialFunction=/Game/Packages/Fortress_SharedMaterials/Base_Material_Functions/Fortnite_TextureSetLerp}
  n38 = DotProduct(A=@n18, B=@n35)
  n39 = Constant4Vector() {constant=<rgba 1 1 1 1>}
  n40 = StaticSwitchParameter(True=@n37.Diffuse, False=StaticSwitch(True=Multiply(A=@n38, B=@n39.RGBA), False=@n18, Value=@n36)) {group=Multiple_Material_Textures, parameterName=Use_Multiple_Material_Textures}
  n41 = ComponentMask(None=@n40) {b=true, g=true, r=true}
  n42 = Multiply(
   A=TextureSampleParameter2D(UVs=StaticSwitchParameter(True=@n7, False=@n6) {group=World Space Color Picker, parameterName=Offset Color Index with UV Channel 2 G}) {group=World Space Color Picker, parameterName=Color Chart, texture=/Game/Global/Textures/Default/ColorSets/CarColors_Interiors}.RGB
   B=@n41
  )
  n43 = TextureSampleParameter2D(UVs=@n6) {group=World Space Color Picker, parameterName=Color Chart 2, texture=/Game/Global/Textures/Default/ColorSets/CarColors}
  n44 = TextureSampleParameter2D() {
   group=World Space Color Picker
   parameterName=World Space Color Picker Mask
   texture=/Game/Vehicles/FORT_Vehicles_01/Materials/M_FORT_Truck_01_M_New
  }
  n45 = AppendVector(A=@n44.RGB, B=@n44.RGB)
  n46 = StaticComponentMaskParameter(None=@n45) {defaultB=true, group=World Space Color Picker, parameterName=World Space Color Picker Mask Channel}
  n47 = LinearInterpolate(
   A=LinearInterpolate(A=@n42, B=Multiply(A=@n43.RGB, B=@n41), Alpha=@n46)
   B=@n41
   Alpha=StaticComponentMaskParameter(None=@n45) {defaultG=true, group=World Space Color Picker, parameterName=World Space Color Picker 2 Mask Channel}
  )
  n48 = LinearInterpolate(A=@n42, B=@n41, Alpha=@n46)
  n49 = StaticSwitchParameter(
   True=StaticSwitchParameter(
    True=StaticSwitchParameter(True=@n47, False=@n48) {group=World Space Color Picker, parameterName=Use 2 World Position Coloring Colors}
    False=@n42
   ) {group=World Space Color Picker, parameterName=Mask World Position Color Picker}
   False=@n41
  ) {group=World Space Color Picker, parameterName=Use World Position Coloring}
  n50 = StaticSwitchParameter(True=Multiply(A=@n49, B=VectorParameter() {group=Diffuse, parameterName=Base Color Multiplier, defaultValue=<rgba 1 1 1 1>}.RGB), False=@n49) {group=Diffuse, parameterName=Modulate Base Color}
  n51 = MaterialFunctionCall!(
   Manipulate Texture UVs=StaticBoolParameter() {group=Emissive, parameterName=Manipulate Emissive Uvs}
   ActorPositionAdd3
   Mask Cropped Emissive=StaticBoolParameter() {defaultValue=true, group=Emissive, parameterName="Mask Cropped Emissive "}
   Texture=TextureObjectParameter() {group=Emissive, parameterName=EmissiveTexture}
   Animated emissive black out time=ScalarParameter() {defaultValue=0.5, desc=should be a value from 0-1, group=Emissive, parameterName=Animated Emissive Blackout Ratio}
   Texture Multiplier=VectorParameter() {group=Emissive, parameterName=Emissive Multiplier, defaultValue=<rgba 1 1 1 1>}.RGB
   Modulate Texture=StaticBoolParameter() {group=Emissive, parameterName=Modulate Emissive}
   UVPositioning LowerRight=AppendVector(A=@n1.B, B=@n1.B)
   UVPositioning UpperLeft=ComponentMask(None=@n1.RGB) {g=true, r=true}
   "------------------- "
   "Animate Using Texture Based Animation(T) or a Flicker(F)"=StaticBoolParameter() {group=Emissive, parameterName=AnimateEmissiveUsingAGradient}
   Animate=StaticBoolParameter() {group=Emissive, parameterName=AnimateEmissive}
   -------------------
   "World Position Animation Offset Multiplier (WS)"=ScalarParameter() {defaultValue=0.01, group=Emissive, parameterName=Animated emissive world position animation offset amount}
   Use World Position to offset the animation=StaticBoolParameter() {group=Emissive, parameterName=Use World position to offset the emissive animation}
   Animated Radius=ScalarParameter() {defaultValue=0.1, group=Emissive, parameterName=Emissive Timeline Animation Falloff Radius}
   Animated Hardness=ScalarParameter() {defaultValue=0.1, group=Emissive, parameterName=Emissive Timeline Animation Falloff Hardness}
   "Loop(T) or Ping  Pong(F) Animation"=StaticBoolParameter() {group=Emissive, parameterName="Animate Emissive Using a loop(T) or Ping Pong(F)"}
   Modulate with diffuse=StaticBoolParameter() {group=Emissive, parameterName=Modulate Emissive with Diffuse}
   Diffuse=@n50
   ---------------------
   ControlTimeManually=StaticBoolParameter() {group=Emissive, parameterName=ControlEmissiveAnimationTimeManually}
   ManualTime=ScalarParameter() {defaultValue=1, group=Emissive, parameterName=EmissiveAnimationManualTime}
   Sequence animation speed=ScalarParameter() {defaultValue=0.25, group=Emissive, parameterName=EmissiveAnimationSpeed}
  ) {materialFunction=/Game/Packages/Fortress_SharedMaterials/Base_Material_Functions/SequenceAnimation}
  n52 = VectorParameter() {group=Emissive2, parameterName="Emissive 2 UV Positioning (RG)UpperLeft (BA)LowerRight", defaultValue=<rgba 0 0 1 1>}
  n53 = Add(
   A=@n51.Result
   B=MaterialFunctionCall(
    Manipulate Texture UVs=StaticBoolParameter() {group=Emissive2, parameterName=Manipulate Emissive 2 Uvs}
    Mask Cropped Emissive=StaticBoolParameter() {defaultValue=true, group=Emissive2, parameterName=Mask Cropped Emissive 2}
    Texture=TextureObjectParameter() {group=Emissive2, parameterName=Emissive Texture 2}
    Animated emissive black out time=ScalarParameter() {defaultValue=0.5, desc=should be a value from 0-1, group=Emissive2, parameterName=Animated Emissive Blackout Ratio 2}
    Texture Multiplier=VectorParameter() {group=Emissive2, parameterName=Emissive Multiplier 2, defaultValue=<rgba 1 1 1 1>}.RGB
    Modulate Texture=StaticBoolParameter() {group=Emissive2, parameterName=Modulate Emissive 2}
    UVPositioning LowerRight=AppendVector(A=@n52.B, B=@n52.B)
    UVPositioning UpperLeft=ComponentMask(None=@n52.RGB) {g=true, r=true}
    "Animate Using Texture Based Animation(T) or a Flicker(F)"=StaticBoolParameter() {group=Emissive2, parameterName=Animate Emissive Using the emissive alpha as a timeline 2}
    Animate=StaticBoolParameter() {group=Emissive2, parameterName=Animate Emissive 2}
    "World Position Animation Offset Multiplier (WS)"=ScalarParameter() {defaultValue=0.01, group=Emissive2, parameterName=Animated emissive world position animation offset amount 2}
    Use World Position to offset the animation=StaticBoolParameter() {group=Emissive2, parameterName=Use World position to offset the emissive animation 2}
    Animated Radius=ScalarParameter() {defaultValue=0.1, group=Emissive2, parameterName=Emissive Timeline Animation Falloff Radius 2}
    Animated Hardness=ScalarParameter() {defaultValue=0.1, group=Emissive2, parameterName=Emissive Timeline Animation Falloff Hardness 2}
    "Loop(T) or Ping  Pong(F) Animation"=StaticBoolParameter() {group=Emissive2, parameterName="Animate Emissive Using a loop(T) or Ping Pong(F) 2"}
    Modulate with diffuse=StaticBoolParameter() {group=Emissive2, parameterName=Modulate Emissive with Diffuse 2}
    Diffuse=@n50
    ControlTimeManually=StaticBoolParameter() {group=Emissive2, parameterName=ControlEmissiveAnimationTimeManually 2}
    ManualTime=ScalarParameter() {defaultValue=1, group=Emissive2, parameterName=EmissiveAnimationManualTime 2}
    Sequence animation speed=ScalarParameter() {defaultValue=0.25, group=Emissive2, parameterName=Emissive Animation Speed 2}
   ) {materialFunction=&SequenceAnimation}.Result
  )
  n54 = MaterialFunctionCall!(
   DiffuseFresnelAmount=Constant() {r=0.2}
   Emissive=StaticSwitchParameter(True=StaticSwitchParameter(True=@n53, False=@n51.Result) {group=Emissive2, parameterName=Use Emissive 2}, False=Constant()) {group=Emissive, parameterName=Use Emissive}
   Base Color=@n50
  ) {desc=diffuse, materialFunction=/Game/Packages/Fortress_SharedMaterials/Base_Material_Functions/FresnelFunction}
  n55 = ScalarParameter() {defaultValue=0.5, group=Specular, parameterName=SpecularMult}
  n56 = ComponentMask(None=@n37.SpecularMasks) {b=true, g=true, r=true}
  n57 = ComponentMask(None=@n56) {desc=spec, r=true}
  n58 = MaterialFunctionCall(Float3=MaterialFunctionCall(X=Multiply(A=@n55, B=@n57), Y=ComponentMask(None=@n56) {desc=metalness, g=true}, Z=ComponentMask(None=@n56) {b=true, desc=roughness}) {desc=Specular Masks, materialFunction=&MakeFloat3}.Result) {materialFunction=&BreakOutFloat3Components}
  n59 = Constant()
  n60 = Constant()
  n61 = LinearInterpolate(
   A=ScalarParameter() {defaultValue=0.1, group=Specular, parameterName=SpecRoughnessMin}
   B=ScalarParameter() {defaultValue=0.9, group=Specular, parameterName=SpecRoughnessMax}
   Alpha=@n58.B
  )
  n62 = Constant() {r=1}
  n63 = ComponentMask(None=@n40) {a=true}
  n64 = TextureCoordinate()
  n65 = VectorParameter() {group=Mask, parameterName=Opacity Mask Offset RG Tiling BA, defaultValue=<rgba 0 0 1 1>}
  n66 = Add(A=@n64, B=ComponentMask(None=@n65.RGB) {g=true, r=true})
  n67 = AppendVector(A=@n65.B, B=@n65.B)
  n68 = TextureSampleParameter2D(UVs=StaticSwitchParameter(True=Multiply(A=@n66, B=@n67), False=@n64) {group=Mask, parameterName=Manipulate Opacity Mask Uvs}) {group=Mask, parameterName=OpacityMask, texture=&T_WhitesRGB}
  n69 = StaticComponentMaskParameter(None=AppendVector(A=@n68.RGB, B=@n68.RGB)) {defaultR=true, group=Mask, parameterName=Opacity Mask Channel}
  n70 = StaticSwitchParameter(
   True=MaterialFunctionCall(Transition Phase=ScalarParameter() {defaultValue=1, group=Mask, parameterName=Opacity}, Height Texture=@n69) {materialFunction=&HeightLerp}.Lerp Alpha No Contrast
   False=@n69
  ) {group=Mask, parameterName=Use a Height Lerp to control masked opacity}
  n71 = StaticSwitchParameter(True=@n37.Normals, False=Constant3Vector() {constant=<rgba 0 0 1 0>}) {group=Normal, parameterName=UseNormalTexture}
  n72 = MaterialFunctionCall!(
   UVs
   Pan Speed
   WPO_ClothOffsetDirection
   WPO_ClothNormalIntensity
   WPO_ClothOffsetIntensity
   "Motion Mask (Defaults to the Red Vertex Color)"
   WPO_ClothWPOWindTextureOffset
   Pan Direction
  ) {materialFunction=/Game/Packages/Fortress_SharedMaterials/Base_Material_Functions/Cloth}
  n73 = MaterialFunctionCall(BaseNormal=@n71, AdditionalNormal=@n72.Normal) {materialFunction=&BlendAngleCorrectedNormals}
  n74 = MaterialFunctionCall(
   UseDestructibleDamage=StaticBoolParameter() {defaultValue=true, group=z_code, parameterName=Code_UseHitEffects}
   UseCrackedDamageTextures=StaticBool()
   DestructibleDamageAmount=ScalarParameter() {group=z_GameplayParameters_DestructibleDamage, parameterName=Code_DestructionAmount}
   WSDeformationCenter=VectorParameter() {group=z_GameplayParameters_DestructibleDamage, parameterName=WSDeformationCenter}.RGB
   WSDamageRadius=ScalarParameter() {group=z_GameplayParameters_DestructibleDamage, parameterName=WSDamageRadius}
  ) {materialFunction=&MF_FN_DestructibleDamage}
  n75 = Add(A=@n74.DamageAnimWPO, B=@n72.Additive WPO)
  MaterialFunctionCall(MA In=MakeMaterialAttributes!(
    BaseColor=@n54.d
    Metallic=StaticSwitchParameter(True=@n58.G, False=@n59) {defaultValue=true, desc=Specular, group=Specular, parameterName=Use Metallic}
    Specular=StaticSwitchParameter(True=@n58.R, False=@n60) {defaultValue=true, desc=Specular, group=Specular, parameterName=Use Spec}
    Roughness=StaticSwitchParameter(True=@n61, False=@n62) {defaultValue=true, desc=Specular, group=Specular, parameterName=Use Roughness}
    Anisotropy
    EmissiveColor=@n54.d
    Opacity
    OpacityMask=StaticSwitchParameter(True=@n63, False=@n70) {group=Mask, parameterName="Use Diffuse Texture Alpha As Opacity Mask (Masked Instances Only)"}
    Normal=StaticSwitchParameter(True=@n73.Result, False=@n71) {parameterName=Use Cloth}
    Tangent
    WorldPositionOffset=StaticSwitchParameter(True=@n75, False=@n74.DamageAnimWPO) {parameterName=Use Cloth}
    SubsurfaceColor
    ClearCoat
    ClearCoatRoughness
    AmbientOcclusion
    Refraction
    CustomizedUVs_0
    CustomizedUVs_1
    CustomizedUVs_2
    CustomizedUVs_3
    CustomizedUVs_4
    CustomizedUVs_5
    CustomizedUVs_6
    CustomizedUVs_7
    PixelDepthOffset
    ShadingModel
    Displacement
   )) {materialFunction=&MF_CreativeMoveTool}
 ]
}
