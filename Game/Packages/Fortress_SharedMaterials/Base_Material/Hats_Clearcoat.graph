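# Masked clear coat base material for character hats, using material attributes and supporting skeletal meshes, with Female Outlander hat textures as defaults.
# Packed S and m masks supply specular, metallic, roughness, AO and clear coat, with scalar parameters scaling roughness and both coat channels.
# Emissive sums SuperChargedGlow and a HitGlow color, while CharacterResInEffect drives the opacity mask.
# All attributes pass through MF_TechArtOverrides before output.
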Material /Game/Packages/Fortress_SharedMaterials/Base_Material/Hats_Clearcoat
graph {
 properties {
  bUseMaterialAttributes=true
  bUsedWithMeshDeformer=true
  bUsedWithSkeletalMesh=true
  blendMode=BLEND_Masked
  pixelDepthOffsetMode=PDOM_Legacy
  shadingModel=MSM_ClearCoat
  usedShadingModels=Clear Coat
 }
 nodes [
  ComponentMask() {g=true, r=true}
  ComponentMask() {g=true, r=true}
  n1 = TextureSampleParameter2D() {parameterName=Base Color, texture=/Game/Accessories/Hats/Textures/Female_Outlander_02/Female_Outlander_02_d}
  n2 = TextureSampleParameter2D() {parameterName=S, samplerType=SAMPLERTYPE_Masks, texture=/Game/Accessories/Hats/Textures/Female_Outlander_02/Female_Outlander_02_s}
  n3 = Multiply(A=@n2.B, B=ScalarParameter() {parameterName=Roughness})
  n4 = Add(
   A=MaterialFunctionCall!(
    AverageHandPosition=VectorParameter() {group=Code, parameterName=AverageHandPosition, defaultValue=<rgba 0 0 0 1>}.RGB
    SuperChargedFlickerBias=ScalarParameter() {defaultValue=1, group=Flicker, parameterName=SuperChargedFlickerBias}
    SuperChargedEffect=ScalarParameter() {group=Code, parameterName=SuperChargedEffect}
    UseDistanceCalc
    Emissive
   ) {materialFunction=/Game/Characters/Fortress_CharacterShader/SuperChargedGlow}.Result
   B=VectorParameter() {group=Emissive, parameterName=HitGlow}.RGB
  )
  n5 = MaterialFunctionCall!(Animation Gradient, Rez In Percentage) {materialFunction=/Game/Packages/Fortress_SharedMaterials/Base_Material_Functions/CharacterResInEffect}
  n6 = TextureSampleParameter2D() {parameterName=Normal, samplerType=SAMPLERTYPE_Normal, texture=/Game/Accessories/Hats/Textures/Female_Outlander_02/Female_Outlander_02_n}
  n7 = TextureSampleParameter2D() {parameterName=m, samplerType=SAMPLERTYPE_Masks, texture=/Game/Accessories/Hats/Textures/Female_Outlander_02/Female_Outlander_02_m}
  n8 = Multiply(A=@n7.G, B=ScalarParameter() {parameterName=Clear Coat})
  n9 = Multiply(A=@n7.B, B=ScalarParameter() {parameterName=Clear Coat Roughness})
  MaterialFunctionCall(In=MakeMaterialAttributes!(
    BaseColor=NamedRerouteDeclaration(None=@n1.RGB) {name=BaseColor, nodeColor=<rgba 0.1953 0 1 1>}
    Metallic=NamedRerouteDeclaration(None=@n2.G) {name=Metallic, nodeColor=<rgba 0 1 0.3438 1>}
    Specular=NamedRerouteDeclaration(None=@n2.R) {name=Specular, nodeColor=<rgba 0 1 0.8359 1>}
    Roughness=NamedRerouteDeclaration(None=@n3) {name=Roughness, nodeColor=<rgba 0.0078 1 0 1>}
    Anisotropy
    EmissiveColor=NamedRerouteDeclaration(None=@n4) {name=EmissiveColor, nodeColor=<rgba 0 0.4141 1 1>}
    Opacity
    OpacityMask=NamedRerouteDeclaration(None=@n5.Modulative Opacity) {name=OpacityMask, nodeColor=<rgba 0 0.6953 1 1>}
    Normal=NamedRerouteDeclaration(None=@n6.RGB) {name=Normal, nodeColor=<rgba 1 0.5391 0 1>}
    Tangent
    WorldPositionOffset
    SubsurfaceColor
    ClearCoat=NamedRerouteDeclaration(None=@n8) {name=ClearCoat, nodeColor=<rgba 0 0.9062 1 1>}
    ClearCoatRoughness=NamedRerouteDeclaration(None=@n9) {name=ClearCoatRoughness, nodeColor=<rgba 0.5234 1 0 1>}
    AmbientOcclusion=NamedRerouteDeclaration(None=@n7.R) {name=AmbientOcclusion, nodeColor=<rgba 0 1 0.6953 1>}
    Refraction
    CustomizedUVs_0
    CustomizedUVs_1
    CustomizedUVs_2
    CustomizedUVs_3
    CustomizedUVs_4
    CustomizedUVs_5
    CustomizedUVs_6
    CustomizedUVs_7
    PixelDepthOffset
    ShadingModel
    Displacement
   )) {materialFunction=&MF_TechArtOverrides}
 ]
}
