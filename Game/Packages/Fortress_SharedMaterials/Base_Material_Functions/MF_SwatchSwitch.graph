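# Material function in the shared Fortress base functions that switches a material's color and alpha between a plain texture and a swatch-blended result.
# A static bool, Use Color Swatch, drives both outputs, D for RGB and A for alpha.
# When on, D lerps from BG Texture color to that color multiplied by Texture, masked by BG alpha, which also becomes A.
# When off, both outputs pass Texture through unchanged.
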MaterialFunction /Game/Packages/Fortress_SharedMaterials/Base_Material_Functions/MF_SwatchSwitch
graph {nodes [
  n1 = FunctionInput() {bUsePreviewValueAsDefault=true, inputName=BG Texture, inputType=FunctionInput_Vector4}
  n2 = ComponentMask(None=@n1) {a=true}
  n3 = FunctionInput() {bUsePreviewValueAsDefault=true, inputName=Texture, inputType=FunctionInput_Vector4}
  n4 = FunctionInput() {inputName=Use Color Swatch, inputType=FunctionInput_StaticBool}
  FunctionOutput(None=StaticSwitch(True=@n2, False=ComponentMask(None=@n3) {a=true}, Value=@n4)) {outputName=A}
  n5 = ComponentMask(None=@n1) {b=true, g=true, r=true}
  n6 = ComponentMask(None=@n3) {b=true, g=true, r=true}
  n7 = Multiply(A=@n5, B=@n6)
  FunctionOutput(None=StaticSwitch(True=LinearInterpolate(A=@n5, B=@n7, Alpha=@n2), False=@n6, Value=@n4)) {outputName=D}
 ]}
