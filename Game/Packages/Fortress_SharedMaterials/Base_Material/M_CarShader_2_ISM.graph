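# Clear-coat master material for vehicles, set up for instanced static meshes and Nanite.
# Up to three layers blend by UV1, with masked paint colors picked from chart textures using actor position.
# An "Old Shader" switch toggles this against a swatch, flake-paint and dust path.
# Damage, time-of-day emissive, optional fake interior light and WPO effects (floating, vertex-color hiding, morphs) follow, ending in shared hit-effect, Curie fallback and Creative move-tool functions.
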Material /Game/Packages/Fortress_SharedMaterials/Base_Material/M_CarShader_2_ISM
graph {
 properties {
  bAutomaticallySetUsageInEditor=false
  bUseMaterialAttributes=true
  bUsedWithInstancedStaticMeshes=true
  bUsedWithNanite=true
  pixelDepthOffsetMode=PDOM_Legacy
  shadingModel=MSM_ClearCoat
  usedShadingModels=Clear Coat
 }
 nodes [
  n1 = TextureSampleParameter2D() {
   desc=Use for interior walls and ceilings
   group=Material_3
   parameterName=Diffuse_Texture_3
   samplerSource=SSM_Wrap_WorldGroupSettings
   texture=/Engine/ArtTools/RenderToTexture/Textures/T_EV_BlankWhite_01
  }
  n2 = TextureSampleParameter2D() {
   desc=Use for interior walls and ceilings
   group=Material_3
   parameterName=SpecularMasks_3
   samplerSource=SSM_Wrap_WorldGroupSettings
   samplerType=SAMPLERTYPE_Masks
   texture=&T_Fortnite_Default_S
  }
  n3 = VectorParameter() {group=Material_3, parameterName=EmissiveColor_Mat3, defaultValue=<rgba 1 1 1 1>}
  n4 = TextureSampleParameter2D() {
   group=Material_3
   parameterName=Emissive_Texture_3
   samplerSource=SSM_Wrap_WorldGroupSettings
   texture=&T_BlacksRGB
  }
  n5 = Power(Base=@n4.RGB, Exp=ScalarParameter() {defaultValue=1, group=Material_3, parameterName=EmissivePower_Mat3})
  n6 = TextureSampleParameter2D() {
   desc=Use for interior walls and ceilings
   group=Material_3
   parameterName=Normals_Texture_3
   samplerSource=SSM_Wrap_WorldGroupSettings
   samplerType=SAMPLERTYPE_Normal
   texture=/Engine/EngineMaterials/FlatNormal
  }
  n7 = TextureSampleParameter2D() {
   group=World Space Color Picker
   parameterName="ColorMask(R=Mat1,G=Mat2,B=Mat3)"
   texture=/Game/Vehicles/FORT_Vehicles_01/Materials/M_FORT_Truck_01_M_New
  }
  n8 = ChannelMaskParameter(None=@n7.RGBA) {group=World Space Color Picker, maskChannel=Blue, parameterName=RandomColorMask_Mat3, defaultValue=<rgba 0 0 1 0>}
  n9 = MakeMaterialAttributes!(
   BaseColor=@n1.RGB
   Metallic=Multiply(A=@n2.G, B=ScalarParameter() {defaultValue=1, group=Material_1, parameterName=Metallic_Mat_3})
   Specular=Multiply(A=@n2.R, B=ScalarParameter() {defaultValue=1, group=Material_1, parameterName=Specular_Mat_3})
   Roughness=Multiply(A=@n2.B, B=ScalarParameter() {defaultValue=1, group=Material_1, parameterName=Roughness_Mat_3})
   Anisotropy
   EmissiveColor=StaticSwitchParameter(True=Multiply(A=@n3.RGB, B=@n5), False=Constant()) {group=Material_3, parameterName=Emissive_Mat3}
   Opacity
   OpacityMask
   Normal=@n6.RGB
   Tangent
   WorldPositionOffset
   SubsurfaceColor
   ClearCoat=Multiply(
    A=StaticSwitchParameter(True=@n8.R, False=@n2.A) {defaultValue=true, group=Material_3, parameterName=UseMaskForClearCoat_Mat3}
    B=ScalarParameter() {defaultValue=1, group=Material_1, parameterName=ClearCoat_Mat_3}
   )
   ClearCoatRoughness=ScalarParameter() {group=Material_3, parameterName=ClearCoat_Roughness_Mat_3}
   AmbientOcclusion
   Refraction
   CustomizedUVs_0
   CustomizedUVs_1
   CustomizedUVs_2
   CustomizedUVs_3
   CustomizedUVs_4
   CustomizedUVs_5
   CustomizedUVs_6
   CustomizedUVs_7
   PixelDepthOffset
   ShadingModel
   Displacement
  )
  n10 = ComponentMask(None=TextureCoordinate() {coordinateIndex=1}) {r=true}
  n11 = Saturate(None=Floor(None=Subtract(A=@n10)))
  n12 = TextureSampleParameter2D() {
   desc=Use for interior walls and ceilings
   group=Material_2
   parameterName=Diffuse_Texture_2
   samplerSource=SSM_Wrap_WorldGroupSettings
   texture=&T_EV_BlankWhite_01
  }
  n13 = TextureSampleParameter2D() {
   desc=Use for interior walls and ceilings
   group=Material_2
   parameterName=SpecularMasks_2
   samplerSource=SSM_Wrap_WorldGroupSettings
   samplerType=SAMPLERTYPE_Masks
   texture=&T_Fortnite_Default_S
  }
  n14 = Multiply(A=@n13.G, B=ScalarParameter() {defaultValue=1, group=Material_1, parameterName=Metallic_Mat_2})
  n15 = Multiply(A=@n13.R, B=ScalarParameter() {defaultValue=1, group=Material_1, parameterName=Specular_Mat_2})
  n16 = Multiply(A=@n13.B, B=ScalarParameter() {defaultValue=1, group=Material_1, parameterName=Roughness_Mat_2})
  n17 = VectorParameter() {group=Material_2, parameterName=EmissiveColor_Mat2, defaultValue=<rgba 1 1 1 1>}
  n18 = TextureSampleParameter2D() {
   group=Material_2
   parameterName=Emissive_Texture_2
   samplerSource=SSM_Wrap_WorldGroupSettings
   texture=&T_BlacksRGB
  }
  n19 = Power(Base=@n18.RGB, Exp=ScalarParameter() {defaultValue=1, group=Material_2, parameterName=EmissivePower_Mat2})
  n20 = StaticSwitchParameter(
   True=Multiply(A=Multiply(A=@n17.RGB, B=@n19), B=VectorParameter() {parameterName=Emissive Multiplier 2, defaultValue=<rgba 1 1 1 1>}.RGB)
   False=Constant()
  ) {group=Material_2, parameterName=Emissive_Mat2}
  n21 = TextureSampleParameter2D() {
   desc=Use for interior walls and ceilings
   group=Material_2
   parameterName=Normals_Texture_2
   samplerSource=SSM_Wrap_WorldGroupSettings
   samplerType=SAMPLERTYPE_Normal
   texture=&FlatNormal
  }
  n22 = ChannelMaskParameter(None=@n7.RGBA) {group=World Space Color Picker, maskChannel=Green, parameterName=RandomColorMask_Mat2, defaultValue=<rgba 0 1 0 0>}
  n23 = Multiply(
   A=StaticSwitchParameter(True=@n22.R, False=@n13.A) {defaultValue=true, group=Material_2, parameterName=UseMaskForClearCoat_Mat2}
   B=ScalarParameter() {defaultValue=1, group=Material_2, parameterName=ClearCoat_Mat_2}
  )
  n24 = ScalarParameter() {group=Material_2, parameterName=ClearCoat_Roughness_Mat_2}
  n25 = Saturate(None=Floor(None=@n10))
  n26 = TextureSampleParameter2D() {group=Material_1, parameterName=Diffuse_Texture_1, samplerSource=SSM_Wrap_WorldGroupSettings, texture=/Engine/EngineMaterials/DefaultDiffuse}
  n27 = TextureSampleParameter2D() {
   group=Material_1
   parameterName=SpecularMasks_1
   samplerSource=SSM_Wrap_WorldGroupSettings
   samplerType=SAMPLERTYPE_Masks
   texture=&T_Fortnite_Default_S
  }
  n28 = TextureSampleParameter2D() {
   group=Material_1
   parameterName=Emissive_Texture_1
   samplerSource=SSM_Wrap_WorldGroupSettings
   texture=&T_BlacksRGB
  }
  n29 = Multiply(
   A=VectorParameter() {group=Material_1, parameterName=EmissiveColor_Mat1, defaultValue=<rgba 1 1 1 1>}.RGB
   B=Power(Base=@n28.RGB, Exp=ScalarParameter() {defaultValue=1, group=Material_1, parameterName=EmissivePower_Mat1})
  )
  n30 = VectorParameter() {parameterName=Emissive Multiplier, defaultValue=<rgba 1 1 1 1>}
  n31 = TextureSampleParameter2D() {
   group=Material_1
   parameterName=Normals_Texture_1
   samplerSource=SSM_Wrap_WorldGroupSettings
   samplerType=SAMPLERTYPE_Normal
   texture=&FlatNormal
  }
  n32 = ChannelMaskParameter(None=@n7.RGBA) {group=World Space Color Picker, parameterName=RandomColorMask_Mat1}
  n33 = MakeMaterialAttributes(
   BaseColor=@n26.RGB
   Metallic=Multiply(A=@n27.G, B=ScalarParameter() {defaultValue=1, desc=Material_1, group=Material_1, parameterName=Metallic_Mat_1})
   Specular=Multiply(A=@n27.R, B=ScalarParameter() {defaultValue=1, desc=Material_1, group=Material_1, parameterName=Specular_Mat_1})
   Roughness=Multiply(A=@n27.B, B=ScalarParameter() {defaultValue=1, desc=Material_1, group=Material_1, parameterName=Roughness_Mat_1})
   EmissiveColor=StaticSwitchParameter(True=Multiply(A=@n29, B=@n30.RGB), False=Constant()) {group=Material_1, parameterName=Emissive_Mat1}
   Normal=@n31.RGB
   ClearCoat=Multiply(
    A=StaticSwitchParameter(True=@n32.R, False=@n27.A) {defaultValue=true, group=Material_1, parameterName=UseMaskForClearCoat_Mat1}
    B=ScalarParameter() {defaultValue=1, desc=Material_1, group=Material_1, parameterName=ClearCoat_Mat_1}
   )
   ClearCoatRoughness=ScalarParameter() {defaultValue=0.02, desc=Material_1, group=Material_1, parameterName=ClearCoat_Roughness_Mat_1}
  )
  n34 = StaticSwitchParameter(
   True=MaterialFunctionCall!(
    Base Material=@n33
    Top Material=MakeMaterialAttributes(
     BaseColor=@n12.RGB
     Metallic=@n14
     Specular=@n15
     Roughness=@n16
     EmissiveColor=@n20
     Normal=@n21.RGB
     ClearCoat=@n23
     ClearCoatRoughness=@n24
    )
    Alpha=@n25
   ) {materialFunction=/Engine/Functions/MaterialLayerFunctions/MatLayerBlend_Standard}.Blended Material
   False=@n33
  ) {group=Material_2, parameterName=Use2ndMaterial}
  n35 = GetMaterialAttributes(None=StaticSwitchParameter(
    True=MaterialFunctionCall(Base Material=@n34, Top Material=@n9, Alpha=@n11) {materialFunction=&MatLayerBlend_Standard}.Blended Material
    False=@n34
   ) {group=Material_3, parameterName=Use3rdMaterial}) {attributeGetTypes [#1, #2]}
  n36 = ActorPositionWS()
  n37 = Add(A=Add(A=ComponentMask(None=@n36) {r=true}, B=ComponentMask(None=@n36) {g=true}), B=ComponentMask(None=@n36) {b=true})
  n38 = ScalarParameter() {defaultValue=16.71, group=World Space Color Picker, parameterName="Color Change Frequency (WS)"}
  n39 = ComponentMask(None=VertexInterpolator(VS=Divide(A=@n37, B=@n38)).PS) {r=true}
  n40 = StaticSwitchParameter(
   True=VectorParameter() {group=World Space Color Picker, parameterName=Color_Mat1, defaultValue=<rgba 0 0 0 1>}.RGB
   False=TextureSampleParameter2D(UVs=@n39) {group=World Space Color Picker, parameterName=Color Chart, texture=/Game/Global/Textures/Default/ColorSets/CarColors}.RGB
  ) {group=World Space Color Picker, parameterName=UseManualColorOverride}
  n41 = LinearInterpolate(A=@n35.Base Color, B=Multiply(A=@n40, B=@n35.Base Color), Alpha=Round(None=@n35.Clear Coat))
  n42 = Constant3Vector() {constant=<rgba 1 1 1 0>}
  n43 = LinearInterpolate(A=@n42, B=@n40, Alpha=@n32.R)
  n44 = TextureSampleParameter2D(UVs=@n39) {group=World Space Color Picker, parameterName=Color Chart2, texture=/Game/Global/Textures/Default/ColorSets/CarColors_Interiors}
  n45 = StaticSwitchParameter(True=VectorParameter() {group=World Space Color Picker, parameterName=Color_Mat2, defaultValue=<rgba 0 0 0 1>}.RGB, False=@n44.RGB) {group=World Space Color Picker, parameterName=UseManualColorOverride}
  n46 = StaticSwitchParameter(True=LinearInterpolate(A=@n43, B=LinearInterpolate(A=@n42, B=@n45, Alpha=@n22.R), Alpha=@n25), False=@n43) {parameterName=Use2ndMaterial}
  n47 = LinearInterpolate(
   A=@n42
   B=StaticSwitchParameter(
    True=VectorParameter() {group=World Space Color Picker, parameterName=Color_Mat2, defaultValue=<rgba 0 0 0 1>}.RGB
    False=TextureSampleParameter2D(UVs=@n39) {group=World Space Color Picker, parameterName=Color Chart4, texture=&CarColors}.RGB
   ) {group=World Space Color Picker, parameterName=UseManualColorOverride}
   Alpha=@n8.R
  )
  n48 = Multiply(A=StaticSwitchParameter(True=LinearInterpolate(A=@n46, B=@n47, Alpha=@n11), False=@n46) {parameterName=Use3rdMaterial}, B=@n35.Base Color)
  n49 = TextureSampleParameter2D() {
   desc=Use for interior walls and ceilings
   group=Material_3
   parameterName=Diffuse_Texture_3
   samplerSource=SSM_Wrap_WorldGroupSettings
   texture=&T_EV_BlankWhite_01
  }
  n50 = TextureSampleParameter2D() {
   desc=Use for interior walls and ceilings
   group=Material_3
   parameterName=SpecularMasks_3
   samplerSource=SSM_Wrap_WorldGroupSettings
   samplerType=SAMPLERTYPE_Masks
   texture=&T_Fortnite_Default_S
  }
  n51 = TextureSampleParameter2D() {
   desc=Use for interior walls and ceilings
   group=Material_3
   parameterName=Normals_Texture_3
   samplerSource=SSM_Wrap_WorldGroupSettings
   samplerType=SAMPLERTYPE_Normal
   texture=&FlatNormal
  }
  n52 = MakeMaterialAttributes(BaseColor=@n49.RGB, Metallic=@n50.G, Specular=@n50.G, Roughness=@n50.G, Normal=@n51.RGB)
  n53 = ComponentMask(None=TextureCoordinate() {coordinateIndex=1}) {r=true}
  n54 = Saturate(None=Floor(None=Subtract(A=@n53)))
  n55 = ActorPositionWS()
  n56 = ComponentMask(None=@n55) {r=true}
  n57 = ComponentMask(None=@n55) {g=true}
  n58 = Divide(
   A=Add(A=Add(A=@n56, B=@n57), B=ComponentMask(None=@n55) {b=true})
   B=ScalarParameter() {defaultValue=16.71, group=World Space Color Picker, parameterName="Color Change Frequency (WS)", sortPriority=5}
  )
  n59 = TextureCoordinate()
  n60 = StaticSwitchParameter(True=ComponentMask(None=VertexInterpolator(VS=@n58).PS) {r=true}, False=@n59) {group=Material_2, parameterName=UseRandomColor_2, sortPriority=4}
  n61 = TextureSampleParameter2D() {
   desc=Use for interior walls and ceilings
   group=Material_2
   parameterName=BG_Diffuse_Texture_2
   samplerSource=SSM_Wrap_WorldGroupSettings
   sortPriority=1
   texture=&DefaultDiffuse
  }
  n62 = MaterialFunctionCall!(
   Texture=TextureSampleParameter2D(UVs=ShadingPathSwitch(Default=@n60, Deferred=@n60, Forward=@n59, Mobile=@n59)) {group=Material_2, parameterName=Diffuse_Texture_2, texture=&T_EV_BlankWhite_01}.RGBA
   BG Texture=@n61.RGBA
   Use Color Swatch=StaticBoolParameter() {group=Material_2, parameterName=Use Color Swatch 2, sortPriority=1}
  ) {materialFunction=/Game/Packages/Fortress_SharedMaterials/Base_Material_Functions/MF_SwatchSwitch}
  n63 = TextureSampleParameter2D() {
   desc=Use for interior walls and ceilings
   group=Material_2
   parameterName=SpecularMasks_2
   samplerSource=SSM_Wrap_WorldGroupSettings
   samplerType=SAMPLERTYPE_Masks
   sortPriority=2
   texture=&T_Fortnite_Default_S
  }
  n64 = Multiply(A=Multiply(A=@n62.D, B=ScalarParameter() {defaultValue=1, group=Material_2, parameterName=Emissive Val 2, sortPriority=4}), B=@n63.A)
  n65 = TextureSampleParameter2D() {
   desc=Use for interior walls and ceilings
   group=Material_2
   parameterName=Normals_Texture_2
   samplerSource=SSM_Wrap_WorldGroupSettings
   samplerType=SAMPLERTYPE_Normal
   sortPriority=3
   texture=&FlatNormal
  }
  n66 = Floor(None=@n53)
  n67 = TextureSampleParameter2D() {
   group=Material_1
   parameterName=Specular Masks
   samplerSource=SSM_Wrap_WorldGroupSettings
   samplerType=SAMPLERTYPE_Masks
   sortPriority=2
   texture=&T_Fortnite_Default_S
  }
  n68 = TextureSampleParameter2D!(UVs, Level=ScalarParameter() {group=Material_1, parameterName=Normal Mip Bias}, Apply View MipBias) {
   group=Material_1
   mipValueMode=TMVM_MipLevel
   parameterName=Normals
   samplerSource=SSM_Wrap_WorldGroupSettings
   samplerType=SAMPLERTYPE_Normal
   texture=&FlatNormal
  }
  n69 = ScalarParameter() {group=Material_1, parameterName=Normal_Flatness}
  n70 = DotProduct(
   A=CameraVectorWS()
   B=Transform(None=MaterialFunctionCall(Normal=@n68.RGB, Flatness=@n69) {materialFunction=&FlattenNormal}.Result)
  )
  n71 = LinearInterpolate(Alpha=@n70) {constB=0.5}
  n72 = DotProduct(A=AtmosphericLightVector(), B=ReflectionVectorWS())
  n73 = ScalarParameter() {defaultValue=4, group=Material_1, parameterName=Light Vector Power, sortPriority=6}
  n74 = LinearInterpolate(A=@n71, Alpha=Saturate(None=Power(Base=@n72, Exp=@n73))) {constA=0.25}
  n75 = TextureCoordinate()
  n76 = StaticBoolParameter() {group=Material_1, parameterName=Metallic Carpaint, sortPriority=5}
  n77 = TextureSampleParameter2D!(UVs, Apply View MipBias) {
   group=Material_1
   parameterName=BG Diffuse Texture
   samplerSource=SSM_Wrap_WorldGroupSettings
   sortPriority=1
   texture=&DefaultDiffuse
  }
  n78 = MaterialFunctionCall(
   Texture=TextureSampleParameter2D(UVs=StaticSwitch(True=@n74, False=@n75, Value=@n76)) {group=Material_1, parameterName=Diffuse, texture=&T_EV_BlankWhite_01}.RGBA
   BG Texture=@n77.RGBA
   Use Color Swatch=StaticBoolParameter() {group=Material_1, parameterName=Use Color Swatch, sortPriority=1}
  ) {materialFunction=&MF_SwatchSwitch}
  n79 = Multiply(A=@n78.D, B=LinearInterpolate(Alpha=@n70) {constA=0.5})
  n80 = Power(Base=@n72, Exp=Multiply(A=@n73) {constB=7})
  n81 = TextureSampleParameter2D!(
   UVs=Multiply(A=TextureCoordinate(), B=ScalarParameter() {defaultValue=20, group=Material_1, parameterName=FlakeTiling, sortPriority=7})
   Bias=ScalarParameter() {group=Material_1, parameterName=BiasFlakes, sortPriority=8}
   Apply View MipBias
  ) {
   mipValueMode=TMVM_MipBias
   parameterName=FlakeNormal
   samplerType=SAMPLERTYPE_Normal
   texture=/Game/Environments/Sets/Neo/Vehicles/Textures/T_Flakes_N
  }
  n82 = Multiply(A=Multiply(A=Saturate(None=@n80), B=Max(A=@n81.R, B=@n81.R)), B=Constant() {r=7})
  n83 = Add(A=@n79, B=Multiply(A=@n79, B=Saturate(None=@n82)))
  n84 = Multiply(A=LinearInterpolate(Alpha=@n74) {constA=0.4}, B=@n78.D)
  n85 = TextureSampleParameter2D!(UVs, Level=ScalarParameter() {group=Material_1, parameterName=Normal Mip Bias}, Apply View MipBias) {
   group=Material_1
   mipValueMode=TMVM_MipLevel
   parameterName=Normals
   samplerSource=SSM_Wrap_WorldGroupSettings
   samplerType=SAMPLERTYPE_Normal
   sortPriority=3
   texture=&FlatNormal
  }
  n86 = MaterialFunctionCall(Normal=@n85.RGB, Flatness=ScalarParameter() {group=Material_1, parameterName=Normal_Flatness}) {materialFunction=&FlattenNormal}
  n87 = MaterialFunctionCall(Normal=@n81.RGB, Flatness=Constant() {r=0.9}) {materialFunction=&FlattenNormal}
  n88 = Constant3Vector() {constant=<rgba 1 1 0 0>}
  n89 = Add(A=@n86.Result, B=Multiply(A=@n71, B=Multiply(A=@n87.Result, B=@n88)))
  n90 = ScalarParameter() {defaultValue=1, group=Material_1, parameterName="Emissive Val ", sortPriority=4}
  n91 = SetMaterialAttributes!(
   MaterialAttributes
   Base Color=LinearInterpolate(A=@n78.D, B=StaticSwitch(True=@n83, False=@n84, Value=@n76), Alpha=@n78.D)
   Emissive Color=Multiply(A=Multiply(A=@n78.D, B=@n90), B=@n67.A)
   Metallic=@n67.G
   Specular=@n67.R
   Roughness=LinearInterpolate(A=@n67.B, B=LinearInterpolate(A=@n67.B) {constB=0.5}, Alpha=@n78.A)
   Normal=StaticSwitch(True=LinearInterpolate(A=@n86.Result, B=@n89, Alpha=@n78.A), False=@n86.Result, Value=@n76)
   Clear Coat=Multiply(A=@n77.A, B=ScalarParameter() {defaultValue=1, group=Material_1, parameterName=Clearcoat Level, sliderMax=1, sortPriority=5})
   Clear Coat Roughness=ScalarParameter() {defaultValue=0.02, desc=Material_1, group=Material_1, parameterName=ClearCoat_Roughness, sortPriority=4}
  ) {attributeSetTypes [
    #1
    #3
    #4
    #5
    #6
    #7
    #2
    #8
   ]}
  n92 = Constant()
  n93 = SetMaterialAttributes!(
   MaterialAttributes
   Base Color=VectorParameter() {group=Material_1, parameterName=DustColor, defaultValue=<rgba 0.7135 0.4582 0.2212 0>}.RGB
   Metallic=@n92
   Specular=Constant() {r=0.3}
   Roughness=Constant() {r=1}
   Clear Coat=@n92
  ) {attributeSetTypes [
    #1
    #4
    #5
    #6
    #2
   ]}
  n94 = ComponentMask(None=VertexNormalWS()) {b=true}
  n95 = Subtract(A=WorldPosition().XYZ, B=ActorPositionWS())
  n96 = Divide(
   A=ComponentMask(None=Transform(None=@n95) {transformSourceType=TRANSFORMSOURCE_World, transformType=TRANSFORM_Local}) {b=true}
   B=ScalarParameter() {defaultValue=64, group=Material_1, parameterName=Dust Fade Height}
  )
  n97 = Multiply(
   A=Multiply(A=OneMinus(None=@n94), B=ScalarParameter() {defaultValue=0.5, group=Material_1, parameterName=Dust Level})
   B=Saturate(None=OneMinus(None=@n96))
  )
  n98 = StaticSwitchParameter(True=BlendMaterialAttributes(A=@n91, B=@n93, Alpha=Saturate(None=@n97)), False=@n91) {group=Material_1, parameterName=Use Dust}
  n99 = StaticSwitchParameter(
   True=MaterialFunctionCall(
    Base Material=@n98
    Top Material=MakeMaterialAttributes(BaseColor=@n62.D, Metallic=@n63.G, Specular=@n63.G, Roughness=@n63.G, EmissiveColor=@n64, Normal=@n65.RGB)
    Alpha=Saturate(None=@n66)
   ) {materialFunction=&MatLayerBlend_Standard}.Blended Material
   False=@n98
  ) {group=Material_2, parameterName=Use2ndMaterial}
  n100 = StaticSwitchParameter(
   True=SetMaterialAttributes!(
    MaterialAttributes=@n35.MaterialAttributes
    Base Color=StaticSwitchParameter(True=@n41, False=@n48) {parameterName=UseClearCoatForRandomCarPaint}
   ) {attributeSetTypes [#1]}
   False=StaticSwitchParameter(
    True=MaterialFunctionCall(Base Material=@n99, Top Material=@n52, Alpha=@n54) {materialFunction=&MatLayerBlend_Standard}.Blended Material
    False=@n99
   ) {group=Material_3, parameterName=Use3rdMaterial}
  ) {defaultValue=true, group=Old Material, parameterName=Old Shader}
  n101 = GetMaterialAttributes(None=@n100) {attributeGetTypes [#7, #1, #2]}
  n102 = TextureSampleParameter2D!(
   UVs=Multiply(A=TextureCoordinate(), B=ScalarParameter() {defaultValue=1, group=Damage, parameterName=DamageTiling, sortPriority=5})
   Apply View MipBias
  ) {
   group=Damage
   parameterName=DamageNormal
   samplerType=SAMPLERTYPE_Normal
   sortPriority=6
   texture=/Game/Global/Textures/Default/Tileable_Patterns/CarDamageNormal
  }
  n103 = ScalarParameter() {defaultValue=0.2, group=Damage, parameterName=DamageAmount}
  n104 = ScalarParameter() {defaultValue=1, group=Damage, parameterName=DamageAmountClearCoat, sortPriority=1}
  n105 = ComponentMask(None=@n101.Clear Coat) {r=true}
  n106 = Constant() {r=0.3333}
  n107 = Multiply(
   A=Saturate(None=LinearInterpolate(A=@n103, B=@n104, Alpha=@n105))
   B=StaticSwitchParameter(
    True=MaterialFunctionCall(Speed=@n106) {materialFunction=&TimeWithSpeedVariable}.Frac Time
    False=ScalarParameter() {group=Code, parameterName=Code_CustomDamagePercentage, sortPriority=4}
   ) {group=Damage, parameterName=Debug custom damage animation, sortPriority=7}
  )
  n108 = LinearInterpolate(
   A=@n101.Normal
   B=MaterialFunctionCall(BaseNormal=@n101.Normal, AdditionalNormal=@n102.RGB) {materialFunction=&BlendAngleCorrectedNormals}.Result
   Alpha=@n107
  )
  n109 = LinearInterpolate(
   A=ScalarParameter() {defaultValue=-5, group=Damage, parameterName=DamageMaskLow, sortPriority=2}
   B=ScalarParameter() {defaultValue=12.6508, group=Damage, parameterName=DamageMaskHigh, sortPriority=3}
   Alpha=Desaturation(None=@n102.RGB, Fraction=Constant() {r=1})
  )
  n110 = LinearInterpolate(
   A=@n101.Base Color
   B=LinearInterpolate(A=@n101.Base Color, B=Constant3Vector() {constant=<rgba 0.5833 0.5833 0.5833 0>}, Alpha=Saturate(None=@n109))
   Alpha=@n107
  )
  n111 = GetMaterialAttributes(None=ShadingPathSwitch(
    Default=SetMaterialAttributes!(MaterialAttributes=@n101.MaterialAttributes, Normal=@n108, Base Color=@n110) {attributeSetTypes [#7, #1]}
    Mobile=@n100
   )) {attributeGetTypes [#3]}
  n112 = MaterialFunctionCall(
   Invert Fresnel=StaticBool() {value=true}
   Power=ScalarParameter() {defaultValue=2, group=Emissve, parameterName=EmissveFresnelExponent, sortPriority=8}
  ) {materialFunction=&Fresnel_Function}
  n113 = GetMaterialAttributes(None=SetMaterialAttributes!(MaterialAttributes=@n111.MaterialAttributes, Emissive Color=Multiply(A=@n112.Result, B=@n111.Emissive Color)) {attributeSetTypes [#3]}) {attributeGetTypes [#3]}
  n114 = VectorParameter() {group=Emissve, parameterName=Emissive multiplier over day phases, sortPriority=8, defaultValue=<rgba 0 0 1 1>}
  n115 = ScalarParameter() {
   desc=Passed in from Car_Default blueprint when time of day changes so the headlights line up visually with spotlights
   parameterName=PassedInTimeOfDay
  }
  n116 = SetMaterialAttributes(
   MaterialAttributes=@n113.MaterialAttributes
   Emissive Color=Multiply(
    A=@n113.Emissive Color
    B=MaterialFunctionCall!(1=@n114.R, 2=@n114.R, 3=@n114.R, 4=@n114.R, A=@n115) {materialFunction=/Engine/Functions/Engine_MaterialFunctions02/Lerp_Multiple_Float}.Lerp 4 Inputs
   )
  ) {attributeSetTypes [#3]}
  n117 = GetMaterialAttributes(None=@n116) {attributeGetTypes [#3]}
  n118 = TextureSampleParameter2D() {
   group=InteriorLight
   parameterName="ColorMask(R=Mat1,G=Mat2,B=Mat3)"
   texture=&M_FORT_Truck_01_M_New
  }
  n119 = VectorParameter() {group=InteriorLight, parameterName=InteriorLight, defaultValue=<rgba 0.1302 0.0554 0.0305 1>}
  n120 = Constant() {r=1}
  n121 = ScalarParameter() {defaultValue=300, group=InteriorLight, parameterName=InteriorLightRadius}
  n122 = CameraPositionWS()
  n123 = Normalize(VectorInput=Subtract(A=WorldPosition().XYZ, B=@n122))
  n124 = Add(A=VectorParameter() {group=InteriorLight, parameterName=InteriorLightPosition, defaultValue=<rgba 0 0 50 1>}.RGB, B=ObjectPositionWS())
  n125 = Power(
   Base=OneMinus(None=MaterialFunctionCall!(Sphere Radius=@n120, Sphere Falloff=@n121, Unit Ray Direction=@n123, Sphere Center=@n124, Ray Origin=@n122) {materialFunction=/Engine/Functions/Engine_MaterialFunctions02/Gradients/RayTraceSphereFalloff}.Unclamped Results)
   Exp=ScalarParameter() {defaultValue=3, group=InteriorLight, parameterName=InteriorLightFalloff}
  )
  n126 = Add(
   A=@n117.Emissive Color
   B=Multiply(
    A=StaticComponentMaskParameter(None=@n118.RGBA) {defaultA=true, group=InteriorLight, parameterName=Interior Light Mask}
    B=Multiply(A=@n119.RGB, B=@n125)
   )
  )
  n127 = GetMaterialAttributes(None=StaticSwitchParameter(
    True=SetMaterialAttributes(MaterialAttributes=@n117.MaterialAttributes, Emissive Color=@n126) {attributeSetTypes [#3]}
    False=@n117.MaterialAttributes
   ) {group=InteriorLight, parameterName=EmissiveWindowInterior}) {attributeGetTypes [#9]}
  n128 = ActorPositionWS()
  n129 = Add(A=ComponentMask(None=@n128) {r=true}, B=ComponentMask(None=@n128) {g=true})
  n130 = ComponentMask(None=@n128) {b=true}
  n131 = Multiply(A=Frac(None=Add(A=@n129, B=@n130)), B=ScalarParameter() {defaultValue=225.4355, group=FloatingWPO, parameterName=RandomTime})
  n132 = Time()
  n133 = Sine(None=Multiply(A=Add(A=@n131, B=@n132), B=ScalarParameter() {defaultValue=0.2, group=FloatingWPO, parameterName=FloatingSpeed}))
  n134 = ScalarParameter() {defaultValue=4, group=FloatingWPO, parameterName=FloatingOffset}
  n135 = Add(A=@n127.World Position Offset, B=AppendVector(A=Constant2Vector(), B=Multiply(A=@n133, B=@n134)))
  n136 = ShadingPathSwitch(
   Default=StaticSwitchParameter(
    True=SetMaterialAttributes!(MaterialAttributes=@n127.MaterialAttributes, World Position Offset=@n135) {attributeSetTypes [#9]}
    False=@n127.MaterialAttributes
   ) {group=FloatingWPO, parameterName="Floating "}
   Forward=@n116
   Mobile=@n116
  )
  n137 = GetMaterialAttributes(None=@n136) {attributeGetTypes [#9]}
  n138 = WorldPosition()
  n139 = VertexColor()
  n140 = VectorParameter() {group=Hide Elements by VertexColor, parameterName=Elements to Hide by color, defaultValue=<rgba 1 0 0 0>}
  n141 = Add(A=Multiply(A=@n139.R, B=@n140.R), B=Multiply(A=@n139.G, B=@n140.G))
  n142 = Multiply(A=@n139.B, B=@n140.B)
  n143 = Multiply(
   A=Subtract(A=ObjectPositionWS(), B=Add(A=@n137.World Position Offset, B=@n138.XYZ))
   B=Add(A=Add(A=@n141, B=@n142), B=Multiply(A=@n139.A, B=@n140.A))
  )
  n144 = StaticSwitchParameter(
   True=SetMaterialAttributes(MaterialAttributes=@n137.MaterialAttributes, World Position Offset=Add(A=@n137.World Position Offset, B=@n143)) {attributeSetTypes [#9]}
   False=@n136
  ) {group=Hide Elements by VertexColor, parameterName=HideElement}
  n145 = GetMaterialAttributes(None=@n144) {attributeGetTypes [#9, #1]}
  n146 = MaterialFunctionCall(
   Material_In=@n145.MaterialAttributes
   UseDestructibleDamage=StaticBoolParameter() {defaultValue=true, group=z_code, parameterName=Code_UseHitEffects}
   UseCrackedDamageTextures=StaticBool() {value=true}
  ) {materialFunction=&MF_FN_DestructibleDamage}
  n147 = MaterialFunctionCall() {materialFunction=/Engine/Functions/Engine_MaterialFunctions02/WorldPositionOffset/StaticMeshMorphTargets}
  n148 = ScalarParameter() {group=MorphTargets, parameterName=MorphTargetStrength}
  n149 = StaticBoolParameter() {group=Special Cases, parameterName=Use Tree Vertex Shader}
  n150 = VectorParameter() {group=WorldPositionOffset, parameterName=WorldPositionOffset_ModulatedByVertexColor, defaultValue=<rgba 0 0 0 1>}
  n151 = Add(
   A=Add(
    A=@n146.DamageAnimWPO
    B=MaterialFunctionCall!(Use Tree Vertex Shader=@n149) {materialFunction=/Game/Packages/Fortress_SharedMaterials/Base_Material_Functions/FortniteTreeAnimation_VS}.Result
   )
   B=Multiply(A=@n150.RGB, B=VertexColor().A)
  )
  n152 = StaticSwitchParameter(True=Add(A=Multiply(A=@n147.Morph Target 1 WPO, B=@n148), B=@n151) {desc=7 instructions}, False=@n151) {group=MorphTargets, parameterName=UseMorphTargets}
  n153 = MaterialProxyReplace(
   Realtime=SetMaterialAttributes(MaterialAttributes=@n146.Material_Out, World Position Offset=Add(A=@n145.World Position Offset, B=@n152)) {attributeSetTypes [#9]}
   MaterialProxy=@n144
  )
  MaterialFunctionCall(MA In=MaterialFunctionCall!(In=ShadingPathSwitch(Default=@n153, Mobile=@n146.Material_Out), UseSphericalMappingOverride) {materialFunction=/Game/Packages/Fortress_SharedMaterials/Base_Material_Functions/MF_CurieFallbackEffect}.Result) {materialFunction=&MF_CreativeMoveTool}
 ]
}
