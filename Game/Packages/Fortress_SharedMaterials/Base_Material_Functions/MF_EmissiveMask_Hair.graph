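# Material function that adds a hit-effect glow to the emissive of incoming material attributes, named as the hair variant of the emissive mask.
# A Fresnel term on the transformed material normal, tuned by exponent and brightness parameters, blends between HitGlowOuter and HitGlow colors, both black by default.
# HitGlowUsesPulse blends the glow multiplier from steady to a time-driven sine pulse.
# The result is added to the existing EmissiveColor.
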MaterialFunction /Game/Packages/Fortress_SharedMaterials/Base_Material_Functions/MF_EmissiveMask_Hair
graph {nodes [
  n1 = GetMaterialAttributes(None=FunctionInput() {inputName="Material (MA)", inputType=FunctionInput_MaterialAttributes}) {attributeGetTypes [#1, #2]}
  n2 = ScalarParameter() {defaultValue=1.4, group=Hit Effects, parameterName=HitGlowFresnelBrightness}
  n3 = Fresnel(ExponentIn=ScalarParameter() {defaultValue=1, group=Hit Effects, parameterName=HitGlowFresnelExponent}, Normal=Transform(None=@n1.Normal)) {baseReflectFraction=0}
  n4 = Sine(None=Time() {bOverride_Period=true, period=0.5}) {period=0.5}
  n5 = Multiply(
   A=LinearInterpolate(
    A=VectorParameter() {group=Hit Effects, parameterName=HitGlowOuter, defaultValue=<rgba 0 0 0 1>}.RGB
    B=VectorParameter() {group=Hit Effects, parameterName=HitGlow, defaultValue=<rgba 0 0 0 1>}.RGB
    Alpha=Multiply(A=@n2, B=@n3)
   )
   B=LinearInterpolate(
    B=ConstantBiasScale(None=@n4) {bias=1.5, scale=0.4}
    Alpha=ScalarParameter() {defaultValue=1, group=Hit Effects, parameterName=HitGlowUsesPulse}
   ) {constA=1}
  )
  FunctionOutput(None=SetMaterialAttributes!(MaterialAttributes=@n1.MaterialAttributes, Normal=@n1.MaterialAttributes, EmissiveColor=Add(A=@n1.EmissiveColor, B=@n5)) {attributeSetTypes [#1, #2]}) {outputName="Result (MA)"}
 ]}
