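# Material function that adds animated ashes-hair vertex displacement to a weapon skin material by overriding World Position Offset, behind a static switch.
# Two panning turbulence samples, projected from pre-skinned position and blended by normal, drive displacement masked by an adjustable vertex-color gradient.
# An optional switch adds pivot-painter-style sine rotation about a noise-derived axis, with vertex color remapped between bounding-box limits as pivot and vertex alpha scaling the angle.
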MaterialFunction /Game/Weapons/WeaponSkins/MaterialFunction/MF_AshesHair
graph {nodes [
  n1 = GetMaterialAttributes(None=FunctionInput() {inputName=Material, inputType=FunctionInput_MaterialAttributes})
  n2 = VertexColor()
  n3 = Multiply(
   A=Add(
    A=ChannelMaskParameter(None=@n2.A) {group=AshesHair, maskChannel=Blue, parameterName=ASH_GradientChannel, defaultValue=<rgba 0 0 1 0>}.R
    B=ScalarParameter() {defaultValue=-0.05, group=AshesHair, parameterName=ASH_GradientOffset}
   )
   B=ScalarParameter() {defaultValue=1.25, group=AshesHair, parameterName=ASH_GradientMultiplier}
  ) {constB=2}
  n4 = PreSkinnedPosition()
  n5 = Divide(
   A=ShaderStageSwitch(PixelShader=VertexInterpolator(VS=@n4).PS, VertexShader=@n4)
   B=VectorParameter() {group=AshesHair, parameterName=ASH_WPONoiseTiling, defaultValue=<rgba 128 128 128 1>}.RGB
  )
  n6 = Add(A=ComponentMask(None=@n5) {g=true, r=true}, B=ComponentMask(None=@n5) {b=true})
  n7 = VectorParameter() {group=AshesHair, parameterName=ASH_WPONoisePanSpeedAndDirection, defaultValue=<rgba 0 0.1 0.1 1>}
  n8 = ComponentMask(None=@n7.RGB) {g=true, r=true}
  n9 = TextureObjectParameter() {
   group=AshesHair
   parameterName=WindTexture
   samplerType=SAMPLERTYPE_LinearColor
   texture=/Game/Effects/Fort_Effects/Textures/Plasma/T_Turbulence_AdditiveWarp_Linear
  }
  n10 = Add(A=ComponentMask(None=@n5) {b=true, r=true}, B=ComponentMask(None=@n5) {g=true})
  n11 = ComponentMask(None=@n7.RGB) {b=true, r=true}
  n12 = PreSkinnedNormal()
  n13 = ShaderStageSwitch(PixelShader=VertexInterpolator(VS=@n12).PS, VertexShader=@n12)
  n14 = LinearInterpolate(
   A=TextureSample(UVs=Panner(Coordinate=@n6, Speed=@n8), Tex=@n9) {samplerSource=SSM_Wrap_WorldGroupSettings}.RGB
   B=TextureSample(UVs=Panner(Coordinate=@n10, Speed=@n11), Tex=@n9) {samplerSource=SSM_Wrap_WorldGroupSettings}.RGB
   Alpha=ComponentMask(None=Abs(None=@n13)) {g=true}
  )
  n15 = Multiply(
   A=Clamp(None=Power(Base=@n3))
   B=Multiply(
    A=ConstantBiasScale(None=@n14) {bias=-0.5, scale=2}
    B=VectorParameter() {group=AshesHair, parameterName=ASH_NoiseDisplacement_Amt, defaultValue=<rgba 1 1 1 1>}.RGB
   )
  )
  n16 = StaticSwitchParameter(True=Add(A=@n15, B=NamedRerouteUsage()), False=@n15) {desc="Uses vertex color as pivot location, similar to pivot painter", parameterName=Enable Vertex Rotation}
  FunctionOutput(None=StaticSwitchParameter(
    True=SetMaterialAttributes!(MaterialAttributes=@n1.MaterialAttributes, WorldPositionOffset=@n16) {attributeSetTypes [#1]}
    False=@n1.MaterialAttributes
   ) {parameterName=UseAshesHair})
  n17 = Divide(A=Time())
  n18 = Multiply(A=Divide(A=Sine(None=@n17)) {constB=16}, B=ScalarParameter() {defaultValue=1, parameterName=Hair Rotation Amp})
  n19 = VertexColor()
  n20 = PreSkinnedPosition()
  n21 = ShaderStageSwitch(PixelShader=VertexInterpolator(VS=@n20).PS, VertexShader=@n20)
  n22 = VectorParameter() {desc=Remaps pivot location from 0-1 to the proper scale, parameterName=BoundboxMin, defaultValue=<rgba -15.787 -16.2923 0 1>}
  n23 = VectorParameter() {desc=Remaps pivot location from 0-1 to the proper scale, parameterName=BoundboxMax, defaultValue=<rgba 15.9435 7.5782 169.593 1>}
  NamedRerouteDeclaration(None=RotateAboutAxis(
    NormalizedRotationAxis=NamedRerouteUsage()
    RotationAngle=Multiply(A=@n19.A, B=@n18)
    PivotPoint=LinearInterpolate(A=@n22.RGB, B=@n23.RGB, Alpha=@n19)
    Position=@n21
   )) {name=Rotation Animtaion, nodeColor=<rgba 1 0 0.2109 1>}
  n24 = Time()
  n25 = Add(
   A=@n21
   B=MaterialFunctionCall(Y=Multiply(A=@n24) {constB=10}) {materialFunction=&MakeFloat3}.Result
  )
  n26 = TextureObjectParameter() {
   group=AshesHair
   parameterName=WindTexture
   samplerType=SAMPLERTYPE_LinearColor
   texture=&T_Turbulence_AdditiveWarp_Linear
  }
  n27 = TextureSample(UVs=Divide(A=ComponentMask(None=@n25) {g=true, r=true}) {constB=200}, Tex=@n26) {
   samplerSource=SSM_Wrap_WorldGroupSettings
   samplerType=SAMPLERTYPE_LinearColor
   texture=&T_Turbulence_AdditiveWarp_Linear
  }
  NamedRerouteDeclaration(None=Normalize(VectorInput=Subtract(A=@n27.RGB) {constB=0.5})) {name=Rotation Axis, nodeColor=<rgba 1 0.2812 0 1>}
 ]}
