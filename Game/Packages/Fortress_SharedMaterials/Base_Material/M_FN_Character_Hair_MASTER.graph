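# Master hair material for Fortress characters, built from material functions passing material attributes, with usage enabled for skeletal meshes, morph targets, mesh deformers and Nanite.
# A static switch, UseSimpleMaterial, picks either a simple material fed by the super-levelling mask or FN_Hair layered with Ashes and animated hair effects.
# The result passes through tech-art overrides, a Fresnel-fed de-rez transition, character opacity and a hair emissive mask.
# A team outline function branches separately off FN_Hair.
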Material /Game/Packages/Fortress_SharedMaterials/Base_Material/M_FN_Character_Hair_MASTER
graph {
 properties {
  bUseMaterialAttributes=true
  bUsedWithMeshDeformer=true
  bUsedWithMorphTargets=true
  bUsedWithNanite=true
  bUsedWithSkeletalMesh=true
  pixelDepthOffsetMode=PDOM_Legacy
  shadingModel=MSM_FromMaterialExpression
  usedShadingModels=Default Lit
 }
 nodes [
  n1 = MaterialFunctionCall() {materialFunction=/Game/Packages/Fortress_SharedMaterials/Base_Material_Functions/FN_Hair}
  MaterialFunctionCall!(Material=@n1.Result) {materialFunction=/Game/Packages/Fortress_SharedMaterials/Base_Material_Functions/MF_TeamOutline}
  n2 = Fresnel()
  n3 = MaterialFunctionCall!(Local Normal, Local Position, Material) {materialFunction=/Game/Packages/Fortress_SharedMaterials/Base_Material_Functions/MF_SuperLevellingMask}
  n4 = MaterialFunctionCall!(Material=@n1.Result) {materialFunction=/Game/Weapons/WeaponSkins/MaterialFunction/MF_AshesHair}
  n5 = MaterialFunctionCall(In=StaticSwitchParameter(
    True=MaterialFunctionCall!(SuperLevellingMask=@n3.SuperLevelling_Mask, SkinFXMask, Material=@n3.SuperLevelling_Mask) {materialFunction=/Game/Packages/Fortress_SharedMaterials/Base_Material_Functions/MF_SimpleMaterial}.Result
    False=MaterialFunctionCall!(GMap, Material=@n4.Result) {materialFunction=/Game/Weapons/WeaponSkins/MaterialFunction/MF_AnimatedHair}.Result
   ) {parameterName=UseSimpleMaterial}) {materialFunction=&MF_TechArtOverrides}
  MaterialFunctionCall!("Material (MA)"=MaterialFunctionCall!(Material=MaterialFunctionCall!(Fresnel=@n2, Material=@n5.Result) {materialFunction=/Game/Packages/Fortress_SharedMaterials/Base_Material_Functions/CharacterDeRez_Transition}.Result) {materialFunction=/Game/Packages/Fortress_SharedMaterials/Base_Material_Functions/BaseCharacterShaderFunctions/BaseFN_CharacterOpacity}.Result) {materialFunction=/Game/Packages/Fortress_SharedMaterials/Base_Material_Functions/MF_EmissiveMask_Hair}
 ]
}
