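# Master character material for the Fortnite sports Impulse outfit, recoloring scarf, top, gloves, pants, socks and boots.
# Vertex-color sphere masks isolate garments, and SkinFX texture channels place colors within each, with UseCustomColors switching between custom and preset palettes.
# Between BaseFN PreFX and PostFX functions, static switches add cloth fuzz, silk, clearcoat, thin film and metal LUT.
# ShowScarf drives World Position Offset that collapses masked vertices to hide the scarf.
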Material /Game/Packages/Fortress_SharedMaterials/Base_Material/M_FN_Character_SPORTS_Impulse
graph {
 properties {
  bUseMaterialAttributes=true
  bUsedWithClothing=true
  bUsedWithMeshDeformer=true
  bUsedWithMorphTargets=true
  bUsedWithSkeletalMesh=true
  pixelDepthOffsetMode=PDOM_Legacy
  shadingModel=MSM_FromMaterialExpression
  subsurfaceProfile=&SS_HeroSkin_02
  usedShadingModels=Default Lit | Clear Coat | Subsurface Profile | Cloth
 }
 nodes [
  n1 = TextureObjectParameter() {group=Skin FX, parameterName=SkinFX_Mask, samplerType=SAMPLERTYPE_Masks, texture=&T_BlackMask}
  n2 = TextureSample(Tex=@n1) {
   samplerType=SAMPLERTYPE_Masks
   texture=/Game/Characters/Player/Male/Medium/Bodies/M_MED_Soccer/Textures/Layout_01/T_M_MALE_Soccer_Demo_M_Med_Soccer_Body_TC
  }
  ChannelMaskParameter(None=@n2.RGBA) {group=Skin FX, parameterName=Dirt}
  n3 = TextureSampleParameter2D() {
   group=Skin FX
   parameterName=SkinFX2_Mask
   samplerType=SAMPLERTYPE_Masks
   texture=/Game/Characters/Player/Male/Medium/Bodies/M_MED_Impulse/Textures/T_M_MED_Impulse_FX2
  }
  NamedRerouteDeclaration(None=ChannelMaskParameter(None=@n3.RGBA) {group=Skin FX, maskChannel=Blue, parameterName=Silk Channel, defaultValue=<rgba 0 0 1 0>}.R) {name=Silk Channel, nodeColor=<rgba 0.5703 1 0 1>}
  n4 = Constant()
  n5 = SetMaterialAttributes!(MaterialAttributes, Base Color=NamedRerouteUsage(), Metallic=@n4, Specular=@n4, Roughness=@n4, Emissive Color=@n4) {attributeSetTypes [
    #1
    #2
    #3
    #4
    #5
   ]}
  n6 = NamedRerouteUsage()
  n7 = Constant()
  n8 = GetMaterialAttributes(None=NamedRerouteUsage()) {attributeGetTypes [#1, #5]}
  n9 = StaticSwitchParameter(
   True=SetMaterialAttributes!(MaterialAttributes=@n8.MaterialAttributes, Base Color=NamedRerouteUsage(), Emissive Color=@n7) {attributeSetTypes [#1, #5]}
   False=SetMaterialAttributes(MaterialAttributes=@n8.MaterialAttributes, Base Color=NamedRerouteUsage(), Emissive Color=@n7) {attributeSetTypes [#1, #5]}
  ) {group=Impulse, parameterName=UseCustomColors}
  n10 = StaticSwitchParameter(
   True=MaterialFunctionCall!(Alpha=NamedRerouteUsage(), Gmap=NamedRerouteUsage(), Material=@n9) {materialFunction=/Game/Weapons/WeaponSkins/MaterialFunction/ClothFuzz}.Result
   False=@n9
  ) {group=Skin FX, parameterName=Use Cloth fuzz}
  n11 = BlendMaterialAttributes(
   A=@n9
   B=StaticSwitchParameter(
    True=MaterialFunctionCall!(Alpha=@n6, Material=@n10) {materialFunction=/Game/Weapons/WeaponSkins/MaterialFunction/MF_Silk}.Result
    False=@n10
   ) {group=Skin FX, parameterName=Use Silk}
   Alpha=ScalarParameter() {group=Impulse, parameterName=MetallicVariantStrength, sliderMax=1}
  )
  n12 = StaticSwitchParameter(
   True=MaterialFunctionCall!(Alpha=NamedRerouteUsage(), In=@n11) {materialFunction=/Game/Weapons/WeaponSkins/MaterialFunction/Clearcoat}.Result
   False=@n11
  ) {group=Skin FX, parameterName=Use Clear Coat}
  n13 = StaticSwitchParameter(
   True=MaterialFunctionCall!(Alpha=NamedRerouteUsage(), Material Attributes=@n12) {materialFunction=/Game/Weapons/WeaponSkins/MaterialFunction/ThinFilmFunction}.Result
   False=@n12
  ) {group=Skin FX, parameterName=Use Thin Film}
  n14 = StaticSwitchParameter(
   True=MaterialFunctionCall(Mask=NamedRerouteUsage(), Materials=@n13) {materialFunction=&MF_MetalsLUT}.Result
   False=@n13
  ) {group=Skin FX, parameterName=UseMetalLUT}
  n15 = ScalarParameter() {defaultValue=1, group=Impulse, parameterName=ShowScarf}
  n16 = Multiply(A=Multiply(A=NamedRerouteUsage(), B=OneMinus(None=@n15)), B=Subtract(A=ObjectPositionWS(), B=WorldPosition().XYZ))
  NamedRerouteDeclaration(None=StaticSwitchParameter(
    True=@n5
    False=SetMaterialAttributes!(MaterialAttributes=@n14, World Position Offset=@n16, Base Color) {attributeSetTypes [#6, #1]}
   ) {group=Impulse, parameterName=Debug}) {name=MA OUT, nodeColor=<rgba 1 0.4688 0 1>}
  n17 = VertexColor()
  NamedRerouteDeclaration(None=SphereMask(A=@n17.R, B=Constant()) {attenuationRadius=0.1}) {name=ScarfMask, nodeColor=<rgba 0.9453 1 0 1>}
  n18 = NamedRerouteUsage()
  n19 = NamedRerouteUsage()
  n20 = Multiply(A=Add(A=Add(A=@n18, B=@n19), B=NamedRerouteUsage())) {constB=2}
  NamedRerouteDeclaration(None=OneMinus(None=Saturate(None=@n20))) {name=Black Mask, nodeColor=<rgba 1 0 0.6562 1>}
  NamedRerouteDeclaration(None=@n8.Emissive Color) {name=EmissiveColorIn, nodeColor=<rgba 0.3594 0 1 1>}
  NamedRerouteDeclaration(None=@n8.Base Color) {name=BaseColorIn, nodeColor=<rgba 1 0 0.9141 1>}
  n21 = NamedRerouteUsage()
  n22 = NamedRerouteUsage()
  n23 = NamedRerouteUsage()
  n24 = LinearInterpolate(
   A=LinearInterpolate(A=LinearInterpolate(A=@n21, B=@n22, Alpha=@n23), B=NamedRerouteUsage(), Alpha=NamedRerouteUsage())
   B=NamedRerouteUsage()
   Alpha=NamedRerouteUsage()
  )
  n25 = NamedRerouteUsage()
  n26 = NamedRerouteUsage()
  n27 = LinearInterpolate(
   A=LinearInterpolate(A=LinearInterpolate(A=@n24, B=@n25, Alpha=@n26), B=NamedRerouteUsage(), Alpha=NamedRerouteUsage())
   B=NamedRerouteUsage()
   Alpha=NamedRerouteUsage()
  )
  NamedRerouteDeclaration(None=LinearInterpolate(A=Multiply(A=@n21, B=@n27), B=@n21, Alpha=NamedRerouteUsage())) {name=CustomColorOut, nodeColor=<rgba 1 0 0.5156 1>}
  n28 = ScalarParameter() {defaultValue=0.5, group=Impulse, parameterName=Mask SmoothStep Min}
  n29 = ScalarParameter() {defaultValue=0.8, group=Impulse, parameterName=Mask SmoothStep Max}
  NamedRerouteDeclaration(None=SmoothStep(Min=@n28, Max=@n29, Value=@n2.R) {constMax=0.9, constMin=0.6}) {name=SkinFX_R, nodeColor=<rgba 0.6875 0 1 1>}
  NamedRerouteDeclaration(None=SmoothStep(Min=@n28, Max=@n29, Value=@n2.G) {constMax=0.9, constMin=0.6}) {name=SkinFX_G, nodeColor=<rgba 1 0.0234 0 1>}
  NamedRerouteDeclaration(None=SmoothStep(Min=@n28, Max=@n29, Value=@n2.B) {constMin=0.6}) {name=SkinFX_B, nodeColor=<rgba 0 0.1562 1 1>}
  NamedRerouteDeclaration(None=@n2.RGB) {name=SkinFX, nodeColor=<rgba 0 0.3437 1 1>}
  n30 = VectorParameter() {group=Impulse, parameterName=CustomColor Scarf1, defaultValue=<rgba 1 0 0 1>}
  n31 = VectorParameter() {group=Impulse, parameterName=CustomColor Scarf2, defaultValue=<rgba 0 1 0 1>}
  n32 = NamedRerouteUsage()
  NamedRerouteDeclaration(None=LinearInterpolate(
    A=LinearInterpolate(A=@n30.RGB, B=@n31.RGB, Alpha=@n32)
    B=VectorParameter() {group=Impulse, parameterName=CustomColor Scarf3, defaultValue=<rgba 0 0 1 1>}.RGB
    Alpha=NamedRerouteUsage()
   )) {name=CustomColor Scarf, nodeColor=<rgba 1 0.5156 0 1>}
  n33 = LinearInterpolate(
   A=VectorParameter() {group=Impulse, parameterName=CustomColor Top4, defaultValue=<rgba 0.3 0.3 0.3 1>}.RGB
   B=VectorParameter() {group=Impulse, parameterName=CustomColor Top1, defaultValue=<rgba 1 0 0 1>}.RGB
   Alpha=NamedRerouteUsage()
  )
  n34 = VectorParameter() {group=Impulse, parameterName=CustomColor Top2, defaultValue=<rgba 0 1 0 1>}
  n35 = NamedRerouteUsage()
  NamedRerouteDeclaration(None=LinearInterpolate(
    A=LinearInterpolate(A=@n33, B=@n34.RGB, Alpha=@n35)
    B=VectorParameter() {group=Impulse, parameterName=CustomColor Top3, defaultValue=<rgba 0 0 1 1>}.RGB
    Alpha=NamedRerouteUsage()
   )) {name=CustomColor Top, nodeColor=<rgba 0.1719 1 0 1>}
  n36 = VertexColor()
  NamedRerouteDeclaration(None=SphereMask(A=@n36.B, B=Constant() {r=0.6}) {attenuationRadius=0.1}) {name=Top Mask, nodeColor=<rgba 0 0.1562 1 1>}
  n37 = VectorParameter() {group=Impulse, parameterName=CustomColor Gloves3, defaultValue=<rgba 0.2 0.2 0.2 1>}
  n38 = VectorParameter() {group=Impulse, parameterName=CustomColor Gloves1, defaultValue=<rgba 1 0 0 1>}
  n39 = NamedRerouteUsage()
  NamedRerouteDeclaration(None=LinearInterpolate(
    A=LinearInterpolate(A=@n37.RGB, B=@n38.RGB, Alpha=@n39)
    B=VectorParameter() {group=Impulse, parameterName=CustomColor Gloves2, defaultValue=<rgba 0 0 1 1>}.RGB
    Alpha=NamedRerouteUsage()
   )) {name=CustomColor Gloves, nodeColor=<rgba 1 0 0.1641 1>}
  NamedRerouteDeclaration(None=SphereMask(A=@n36.B, B=Constant() {r=0.8}) {attenuationRadius=0.1}) {name=Glove Mask, nodeColor=<rgba 0.3594 1 0 1>}
  n40 = LinearInterpolate(
   A=VectorParameter() {group=Impulse, parameterName=CustomColor Boots4, defaultValue=<rgba 0.3 0.3 0.3 1>}.RGB
   B=VectorParameter() {group=Impulse, parameterName=CustomColor Boots1, defaultValue=<rgba 1 0 0 1>}.RGB
   Alpha=NamedRerouteUsage()
  )
  n41 = VectorParameter() {group=Impulse, parameterName=CustomColor Boots2, defaultValue=<rgba 0 1 0 1>}
  n42 = NamedRerouteUsage()
  NamedRerouteDeclaration(None=LinearInterpolate(
    A=LinearInterpolate(A=@n40, B=@n41.RGB, Alpha=@n42)
    B=VectorParameter() {group=Impulse, parameterName=CustomColor Boots3, defaultValue=<rgba 0 0 1 1>}.RGB
    Alpha=NamedRerouteUsage()
   )) {name=CustomColor Boots, nodeColor=<rgba 0 1 0.3203 1>}
  n43 = LinearInterpolate(
   A=VectorParameter() {group=Impulse, parameterName=CustomColor Pants4, defaultValue=<rgba 0.3 0.3 0.3 1>}.RGB
   B=VectorParameter() {group=Impulse, parameterName=CustomColor Pants1, defaultValue=<rgba 1 0 0 1>}.RGB
   Alpha=NamedRerouteUsage()
  )
  n44 = VectorParameter() {group=Impulse, parameterName=CustomColor Pants2, defaultValue=<rgba 0 1 0 1>}
  n45 = NamedRerouteUsage()
  NamedRerouteDeclaration(None=LinearInterpolate(
    A=LinearInterpolate(A=@n43, B=@n44.RGB, Alpha=@n45)
    B=VectorParameter() {group=Impulse, parameterName=CustomColor Pants3, defaultValue=<rgba 0 0 1 1>}.RGB
    Alpha=NamedRerouteUsage()
   )) {name=CustomColor Pants, nodeColor=<rgba 0 1 0.3203 1>}
  NamedRerouteDeclaration(None=SphereMask(A=@n36.B, B=Constant()) {attenuationRadius=0.1}) {name=Pants Mask, nodeColor=<rgba 1 0 0.6797 1>}
  NamedRerouteDeclaration(None=SphereMask(A=@n36.B, B=Constant() {r=0.2}) {attenuationRadius=0.1}) {name=Boots Mask, nodeColor=<rgba 0 1 0.6484 1>}
  NamedRerouteDeclaration(None=LinearInterpolate(
    A=VectorParameter() {group=Impulse, parameterName=CustomColor Socks1, defaultValue=<rgba 1 0 0 1>}.RGB
    B=VectorParameter() {group=Impulse, parameterName=CustomColor Socks2, defaultValue=<rgba 0 1 0 1>}.RGB
    Alpha=NamedRerouteUsage()
   )) {name=CustomColor Socks, nodeColor=<rgba 0 0.8594 1 1>}
  NamedRerouteDeclaration(None=ChannelMaskParameter(None=@n3.RGBA) {group=Skin FX, parameterName=Cloth Channel}.R) {name=Cloth Channel, nodeColor=<rgba 0 1 0.2969 1>}
  NamedRerouteDeclaration(None=SphereMask(A=@n36.B, B=Constant() {r=0.4}) {attenuationRadius=0.1}) {name=SocksMask, nodeColor=<rgba 0.1953 0 1 1>}
  NamedRerouteDeclaration(None=@n2.R) {name="SkinFX_R (raw)", nodeColor=<rgba 0 0.25 1 1>}
  NamedRerouteDeclaration(None=@n2.G) {name="SkinFX_G (raw)", nodeColor=<rgba 0 1 0.0859 1>}
  NamedRerouteDeclaration(None=@n2.B) {name="SkinFX_B (raw)", nodeColor=<rgba 0 1 0.7891 1>}
  n46 = VectorParameter() {group=Impulse, parameterName=PresetColor Scarf1, defaultValue=<rgba 1 0 0 1>}
  n47 = VectorParameter() {group=Impulse, parameterName=PresetColor Scarf2, defaultValue=<rgba 0 1 0 1>}
  n48 = NamedRerouteUsage()
  NamedRerouteDeclaration(None=LinearInterpolate(
    A=LinearInterpolate(A=@n46.RGB, B=@n47.RGB, Alpha=@n48)
    B=VectorParameter() {group=Impulse, parameterName=PresetColor Scarf3, defaultValue=<rgba 0 0 1 1>}.RGB
    Alpha=NamedRerouteUsage()
   )) {name=PresetColor Scarf, nodeColor=<rgba 1 0.5156 0 1>}
  n49 = LinearInterpolate(
   A=VectorParameter() {group=Impulse, parameterName=PresetColor Top4, defaultValue=<rgba 0.3 0.3 0.3 1>}.RGB
   B=VectorParameter() {group=Impulse, parameterName=PresetColor Top1, defaultValue=<rgba 1 0 0 1>}.RGB
   Alpha=NamedRerouteUsage()
  )
  n50 = VectorParameter() {group=Impulse, parameterName=PresetColor Top2, defaultValue=<rgba 0 1 0 1>}
  n51 = NamedRerouteUsage()
  NamedRerouteDeclaration(None=LinearInterpolate(
    A=LinearInterpolate(A=@n49, B=@n50.RGB, Alpha=@n51)
    B=VectorParameter() {group=Impulse, parameterName=PresetColor Top3, defaultValue=<rgba 0 0 1 1>}.RGB
    Alpha=NamedRerouteUsage()
   )) {name=PresetColor Top, nodeColor=<rgba 0.1719 1 0 1>}
  n52 = VectorParameter() {group=Impulse, parameterName=PresetColor Gloves3, defaultValue=<rgba 0.2 0.2 0.2 1>}
  n53 = VectorParameter() {group=Impulse, parameterName=PresetColor Gloves1, defaultValue=<rgba 1 0 0 1>}
  n54 = NamedRerouteUsage()
  NamedRerouteDeclaration(None=LinearInterpolate(
    A=LinearInterpolate(A=@n52.RGB, B=@n53.RGB, Alpha=@n54)
    B=VectorParameter() {group=Impulse, parameterName=PresetColor Gloves2, defaultValue=<rgba 0 0 1 1>}.RGB
    Alpha=NamedRerouteUsage()
   )) {name=PresetColor Gloves, nodeColor=<rgba 1 0 0.1641 1>}
  n55 = LinearInterpolate(
   A=VectorParameter() {group=Impulse, parameterName=PresetColor Pants4, defaultValue=<rgba 0.3 0.3 0.3 1>}.RGB
   B=VectorParameter() {group=Impulse, parameterName=PresetColor Pants1, defaultValue=<rgba 1 0 0 1>}.RGB
   Alpha=NamedRerouteUsage()
  )
  n56 = VectorParameter() {group=Impulse, parameterName=PresetColor Pants2, defaultValue=<rgba 0 1 0 1>}
  n57 = NamedRerouteUsage()
  NamedRerouteDeclaration(None=LinearInterpolate(
    A=LinearInterpolate(A=@n55, B=@n56.RGB, Alpha=@n57)
    B=VectorParameter() {group=Impulse, parameterName=PresetColor Pants3, defaultValue=<rgba 0 0 1 1>}.RGB
    Alpha=NamedRerouteUsage()
   )) {name=PresetColor Pants, nodeColor=<rgba 0 1 0.3203 1>}
  n58 = LinearInterpolate(
   A=VectorParameter() {group=Impulse, parameterName=PresetColor Boots4, defaultValue=<rgba 0.3 0.3 0.3 1>}.RGB
   B=VectorParameter() {group=Impulse, parameterName=PresetColor Boots1, defaultValue=<rgba 1 0 0 1>}.RGB
   Alpha=NamedRerouteUsage()
  )
  n59 = VectorParameter() {group=Impulse, parameterName=PresetColor Boots2, defaultValue=<rgba 0 1 0 1>}
  n60 = NamedRerouteUsage()
  NamedRerouteDeclaration(None=LinearInterpolate(
    A=LinearInterpolate(A=@n58, B=@n59.RGB, Alpha=@n60)
    B=VectorParameter() {group=Impulse, parameterName=PresetColor Boots3, defaultValue=<rgba 0 0 1 1>}.RGB
    Alpha=NamedRerouteUsage()
   )) {name=PresetColor Boots, nodeColor=<rgba 0 1 0.3203 1>}
  NamedRerouteDeclaration(None=LinearInterpolate(
    A=VectorParameter() {group=Impulse, parameterName=PresetColor Socks1, defaultValue=<rgba 1 0 0 1>}.RGB
    B=VectorParameter() {group=Impulse, parameterName=PresetColor Socks2, defaultValue=<rgba 0 1 0 1>}.RGB
    Alpha=NamedRerouteUsage()
   )) {name=PresetColor Socks, nodeColor=<rgba 0 0.8594 1 1>}
  n61 = MaterialFunctionCall() {materialFunction=&BaseFN_CharacterFunction_PreFX}
  NamedRerouteDeclaration(None=@n61.Gmap) {name=Gmap, nodeColor=<rgba 0.2891 0 1 1>}
  NamedRerouteDeclaration(None=@n3.RGB) {name=MaterialFunctionMasks, nodeColor=<rgba 1 0 0.4453 1>}
  n62 = NamedRerouteUsage()
  n63 = NamedRerouteUsage()
  n64 = NamedRerouteUsage()
  n65 = LinearInterpolate(
   A=LinearInterpolate(A=LinearInterpolate(A=@n62, B=@n63, Alpha=@n64), B=NamedRerouteUsage(), Alpha=NamedRerouteUsage())
   B=NamedRerouteUsage()
   Alpha=NamedRerouteUsage()
  )
  n66 = NamedRerouteUsage()
  n67 = NamedRerouteUsage()
  n68 = LinearInterpolate(
   A=LinearInterpolate(A=LinearInterpolate(A=@n65, B=@n66, Alpha=@n67), B=NamedRerouteUsage(), Alpha=NamedRerouteUsage())
   B=NamedRerouteUsage()
   Alpha=NamedRerouteUsage()
  )
  NamedRerouteDeclaration(None=LinearInterpolate(A=Multiply(A=@n62, B=@n68), B=@n62, Alpha=NamedRerouteUsage())) {name=ExpandedPresetColorOut, nodeColor=<rgba 0.8281 0 1 1>}
  NamedRerouteDeclaration(None=ChannelMaskParameter(None=@n2.RGBA) {group=Skin FX, maskChannel=Green, parameterName=Clearcoat Channel, defaultValue=<rgba 0 1 0 0>}.R) {name=ClearCoat Channel, nodeColor=<rgba 1 0 0.75 1>}
  NamedRerouteDeclaration(None=@n61.Result) {name=MA IN, nodeColor=<rgba 0 1 0.9063 1>}
  n69 = Constant() {r=1}
  NamedRerouteDeclaration(None=OneMinus(None=SphereMask(A=@n17.G, B=@n69) {attenuationRadius=0.1})) {name=SkinMask, nodeColor=<rgba 1 0.5625 0 1>}
  NamedRerouteDeclaration(None=ChannelMaskParameter(None=@n2.RGBA) {group=Skin FX, maskChannel=Green, parameterName=ThinFilm Channel, defaultValue=<rgba 0 1 0 0>}.R) {name=ThinFilm Channel, nodeColor=<rgba 0 0.0859 1 1>}
  NamedRerouteDeclaration(None=ChannelMaskParameter(None=@n3.RGBA) {group=Skin FX, maskChannel=Green, parameterName=Metal LUT Channel, defaultValue=<rgba 0 1 0 0>}.R) {name=Metal LUT Channel, nodeColor=<rgba 0.7109 1 0 1>}
  NamedRerouteUsage()
  NamedRerouteUsage()
  NamedRerouteUsage()
  NamedRerouteUsage()
  NamedRerouteUsage()
  n70 = MaterialFunctionCall(AO=@n61.AO, Distance Blend=@n61.AO, Fresnel=@n61.AO, Gmap=@n61.AO, Pre Skinned Local Position=@n61.AO, Material=NamedRerouteUsage()) {materialFunction=&BaseFN_CharacterFunction_PostFX}
  n71 = Constant()
  SetMaterialAttributes!(MaterialAttributes=@n70.Result, Base Color=NamedRerouteUsage(), Metallic=@n71, Specular=@n71, Roughness=@n71, Emissive Color) {attributeSetTypes [
    #1
    #2
    #3
    #4
    #5
   ]}
  VertexColor()
 ]
}
