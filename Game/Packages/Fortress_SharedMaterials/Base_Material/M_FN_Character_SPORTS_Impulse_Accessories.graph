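# Master character material for Impulse sports-skin accessories on skeletal meshes, supporting default lit, clear coat and subsurface profile shading.
# Shared PreFX and PostFX functions bracket a chain of static switches for custom colors, silk, clear coat, thin film and metal LUT.
# The custom color path blends two color parameters into base color and overrides emissive with a constant.
# A SkinFX mask texture supplies selectable effect channels, and vertex color blue yields a custom-color mask.
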Material /Game/Packages/Fortress_SharedMaterials/Base_Material/M_FN_Character_SPORTS_Impulse_Accessories
graph {
 properties {
  bUseMaterialAttributes=true
  bUsedWithMeshDeformer=true
  bUsedWithSkeletalMesh=true
  pixelDepthOffsetMode=PDOM_Legacy
  shadingModel=MSM_FromMaterialExpression
  subsurfaceProfile=&SS_HeroSkin_02
  usedShadingModels=Default Lit | Clear Coat | Subsurface Profile
 }
 nodes [
  n1 = TextureObjectParameter() {group=Skin FX, parameterName=SkinFX_Mask, samplerType=SAMPLERTYPE_Masks, texture=&T_BlackMask}
  n2 = TextureSample(Tex=@n1) {
   samplerType=SAMPLERTYPE_Masks
   texture=/Game/Characters/Player/Male/Medium/Bodies/M_MED_Soccer/Textures/Layout_01/T_M_MALE_Soccer_Demo_M_Med_Soccer_Body_TC
  }
  ChannelMaskParameter(None=@n2.RGBA) {group=Skin FX, parameterName=Dirt}
  n3 = MaterialFunctionCall() {materialFunction=&BaseFN_CharacterFunction_PreFX}
  MaterialFunctionCall(AO=@n3.AO, Distance Blend=@n3.AO, Fresnel=@n3.AO, Gmap=@n3.AO, Pre Skinned Local Position=@n3.AO, Material=NamedRerouteUsage()) {materialFunction=&BaseFN_CharacterFunction_PostFX}
  n4 = NamedRerouteUsage()
  n5 = GetMaterialAttributes(None=NamedRerouteUsage()) {attributeGetTypes [#1, #2]}
  n6 = StaticSwitchParameter(
   True=SetMaterialAttributes!(MaterialAttributes=@n5.MaterialAttributes, Base Color=NamedRerouteUsage(), Emissive Color=Constant()) {attributeSetTypes [#1, #2]}
   False=SetMaterialAttributes(
    MaterialAttributes=@n5.MaterialAttributes
    Base Color=NamedRerouteDeclaration(None=@n5.Base Color) {name=BaseColorIn, nodeColor=<rgba 1 0 0.9141 1>}
    Emissive Color=NamedRerouteDeclaration(None=@n5.Emissive Color) {name=EmissiveColorIn, nodeColor=<rgba 0.3594 0 1 1>}
   ) {attributeSetTypes [#1, #2]}
  ) {group=Impulse, parameterName=UseCustomColors}
  n7 = StaticSwitchParameter(
   True=MaterialFunctionCall!(Alpha=NamedRerouteUsage(), Material=@n6) {materialFunction=/Game/Weapons/WeaponSkins/MaterialFunction/MF_Silk}.Result
   False=@n6
  ) {group=Skin FX, parameterName=Use Silk}
  n8 = StaticSwitchParameter(
   True=MaterialFunctionCall!(Alpha=NamedRerouteUsage(), In=@n7) {materialFunction=/Game/Weapons/WeaponSkins/MaterialFunction/Clearcoat}.Result
   False=@n7
  ) {group=Skin FX, parameterName=Use Clear Coat}
  n9 = StaticSwitchParameter(
   True=MaterialFunctionCall!(Alpha=NamedRerouteUsage(), Material Attributes=@n8) {materialFunction=/Game/Weapons/WeaponSkins/MaterialFunction/ThinFilmFunction}.Result
   False=@n8
  ) {group=Skin FX, parameterName=Use Thin Film}
  NamedRerouteDeclaration(None=StaticSwitchParameter(
    True=MaterialFunctionCall(Mask=@n4, Materials=@n9) {materialFunction=&MF_MetalsLUT}.Result
    False=@n9
   ) {group=Skin FX, parameterName=UseMetalLUT}) {name=MA OUT, nodeColor=<rgba 1 0.4688 0 1>}
  n10 = NamedRerouteUsage()
  n11 = NamedRerouteUsage()
  NamedRerouteDeclaration(None=LinearInterpolate(A=@n10, B=Multiply(A=@n10, B=@n11), Alpha=NamedRerouteUsage())) {name=CustomColorOut, nodeColor=<rgba 1 0 0.5156 1>}
  n12 = ScalarParameter() {defaultValue=0.5, group=Impulse, parameterName=Mask SmoothStep Min}
  n13 = ScalarParameter() {defaultValue=0.8, group=Impulse, parameterName=Mask SmoothStep Max}
  NamedRerouteDeclaration(None=SmoothStep(Min=@n12, Max=@n13, Value=@n2.R) {constMax=0.9, constMin=0.6}) {name=SkinFX_R, nodeColor=<rgba 0.6875 0 1 1>}
  NamedRerouteDeclaration(None=SmoothStep(Min=@n12, Max=@n13, Value=@n2.G) {constMax=0.9, constMin=0.6}) {name=SkinFX_G, nodeColor=<rgba 1 0.0234 0 1>}
  NamedRerouteDeclaration(None=SmoothStep(Min=@n12, Max=@n13, Value=@n2.B) {constMin=0.6}) {name=SkinFX_B, nodeColor=<rgba 0 0.1562 1 1>}
  NamedRerouteDeclaration(None=@n2.RGB) {name=SkinFX, nodeColor=<rgba 0 0.3437 1 1>}
  NamedRerouteDeclaration(None=LinearInterpolate(
    A=VectorParameter() {group=Impulse, parameterName=CustomColor 1, defaultValue=<rgba 1 0 0 1>}.RGB
    B=VectorParameter() {group=Impulse, parameterName=CustomColor 2, defaultValue=<rgba 0 1 0 1>}.RGB
    Alpha=NamedRerouteUsage()
   )) {name=CustomColor Accessories, nodeColor=<rgba 0 0.8594 1 1>}
  NamedRerouteDeclaration(None=@n2.R) {name="SkinFX_R (raw)", nodeColor=<rgba 0 0.25 1 1>}
  NamedRerouteDeclaration(None=@n2.G) {name="SkinFX_G (raw)", nodeColor=<rgba 0 1 0.0859 1>}
  NamedRerouteDeclaration(None=@n2.B) {name="SkinFX_B (raw)", nodeColor=<rgba 0 1 0.7891 1>}
  n14 = NamedRerouteUsage()
  n15 = NamedRerouteUsage()
  n16 = Multiply(A=Add(A=Add(A=@n14, B=@n15), B=NamedRerouteUsage())) {constB=2}
  NamedRerouteDeclaration(None=OneMinus(None=Saturate(None=@n16))) {name=Black Mask, nodeColor=<rgba 1 0 0.6562 1>}
  NamedRerouteDeclaration(None=SphereMask(A=VertexColor().B, B=Constant()) {attenuationRadius=0.1}) {name=CustomColors Mask, nodeColor=<rgba 1 0 0.6797 1>}
  NamedRerouteDeclaration(None=ChannelMaskParameter(None=@n2.RGBA) {group=Skin FX, maskChannel=Blue, parameterName=Silk Channel, defaultValue=<rgba 0 0 1 0>}.R) {name=Silk Channel, nodeColor=<rgba 0.5703 1 0 1>}
  NamedRerouteDeclaration(None=ChannelMaskParameter(None=@n2.RGBA) {group=Skin FX, maskChannel=Green, parameterName=Clearcoat Channel, defaultValue=<rgba 0 1 0 0>}.R) {name=ClearCoat Channel, nodeColor=<rgba 1 0 0.75 1>}
  NamedRerouteDeclaration(None=@n3.Result) {name=MA IN, nodeColor=<rgba 0 1 0.9063 1>}
  NamedRerouteDeclaration(None=ChannelMaskParameter(None=@n2.RGBA) {group=Skin FX, maskChannel=Green, parameterName=ThinFilm Channel, defaultValue=<rgba 0 1 0 0>}.R) {name=ThinFilm Channel, nodeColor=<rgba 0 0.0859 1 1>}
  NamedRerouteDeclaration(None=ChannelMaskParameter(None=@n2.RGBA) {group=Skin FX, maskChannel=Green, parameterName=Metal LUT Channel, defaultValue=<rgba 0 1 0 0>}.R) {name=Metal LUT Channel, nodeColor=<rgba 0.7109 1 0 1>}
 ]
}
