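# M_FN_Character_Sports is a skeletal-mesh master material for sports character skins with customizable colors, numbers and crests.
# The shared PreFX function feeds a static-switch chain of optional color, number, UV rotation, crest, silk, cloth fuzz and clearcoat layers.
# Selectable channels of a SkinFX mask drive most layers and a tinted dirt blend on Base Color.
# PostFX finishes the result; a default-bypassed switch can force Subsurface Profile shading.
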Material /Game/Packages/Fortress_SharedMaterials/Base_Material/M_FN_Character_Sports.M_FN_Character_SPORTS
graph {
 properties {
  bUseMaterialAttributes=true
  bUsedWithMeshDeformer=true
  bUsedWithMorphTargets=true
  bUsedWithSkeletalMesh=true
  pixelDepthOffsetMode=PDOM_Legacy
  refractionMethod=RM_IndexOfRefraction
  shadingModel=MSM_FromMaterialExpression
  usedShadingModels=Default Lit | Clear Coat | Subsurface Profile | Cloth
 }
 nodes [
  n1 = MaterialFunctionCall() {materialFunction=&BaseFN_CharacterFunction_PreFX}
  n2 = TextureObjectParameter() {group=Skin FX, parameterName=SkinFX_Mask, samplerType=SAMPLERTYPE_Masks, texture=&T_BlackMask}
  n3 = TextureSample(Tex=@n2) {
   samplerType=SAMPLERTYPE_Masks
   texture=/Game/Characters/Player/Male/Medium/Bodies/M_MED_Soccer/Textures/Layout_01/T_M_MALE_Soccer_Demo_M_Med_Soccer_Body_TC
  }
  n4 = ChannelMaskParameter(None=@n3.RGBA) {group=Skin FX, maskChannel=Green, parameterName=Clearcoat Channel, defaultValue=<rgba 0 1 0 0>}
  n5 = ChannelMaskParameter(None=@n3.RGBA) {group=Skin FX, parameterName=Custom Numbers Channel, defaultValue=<rgba 0 1 0 0>}
  n6 = MaterialFunctionCall!(GMap=@n1.Gmap, Material=@n1.Gmap) {materialFunction=/Game/Weapons/WeaponSkins/MaterialFunction/CustomizeableColors}
  n7 = StaticSwitchParameter(True=@n6.Result, False=@n1.Result) {group=Skin FX, parameterName=UseCustomColors}
  n8 = StaticSwitchParameter(
   True=MaterialFunctionCall!(Alpha=@n5.R, Material=@n7) {materialFunction=/Game/Weapons/WeaponSkins/MaterialFunction/CustomNumbers_DoNotUse}.Result
   False=@n7
  ) {group=Skin FX, parameterName="Use OLD Custom Numbers (Do Not Use)"}
  n9 = StaticSwitchParameter(
   True=MaterialFunctionCall!(Alpha=@n5.R, Material=@n8) {materialFunction=/Game/Weapons/WeaponSkins/MaterialFunction/CustomNumbers_Optimized}.Result
   False=@n8
  ) {group=Skin FX, parameterName=Use Custom Numbers Optimized}
  n10 = StaticSwitchParameter(
   True=MaterialFunctionCall!(Alpha=ChannelMaskParameter(None=@n3.RGBA) {group=Skin FX, parameterName=UV Rotation Channel}.R, Material=@n9) {materialFunction=/Game/Weapons/WeaponSkins/MaterialFunction/UV_Rotator}.Result
   False=@n9
  ) {group=Skin FX, parameterName=Use Rotator}
  n11 = StaticSwitchParameter(True=MaterialFunctionCall!(Materials=@n10) {materialFunction=/Game/Weapons/WeaponSkins/MaterialFunction/CustomCrests}.Result, False=@n10) {group=Skin FX, parameterName=Use Custom Crest}
  n12 = StaticSwitchParameter(
   True=MaterialFunctionCall!(
    Alpha=ChannelMaskParameter(None=@n3.RGBA) {group=Skin FX, maskChannel=Green, parameterName=Silk Channel, defaultValue=<rgba 0 1 0 0>}.R
    Material=@n11
   ) {materialFunction=/Game/Weapons/WeaponSkins/MaterialFunction/MF_Silk}.Result
   False=@n11
  ) {group=Skin FX, parameterName=Use Silk}
  n13 = StaticSwitchParameter(
   True=MaterialFunctionCall!(
    Alpha=ChannelMaskParameter(None=@n3.RGBA) {group=Skin FX, maskChannel=Green, parameterName=Cloth Channel, defaultValue=<rgba 0 1 0 0>}.R
    Gmap=@n1.Gmap
    Material=@n12
   ) {materialFunction=/Game/Weapons/WeaponSkins/MaterialFunction/ClothFuzz}.Result
   False=@n12
  ) {group=Skin FX, parameterName=Use Cloth}
  n14 = GetMaterialAttributes(None=StaticSwitchParameter(True=MaterialFunctionCall!(Alpha=@n4.R, In=@n13) {materialFunction=/Game/Weapons/WeaponSkins/MaterialFunction/Clearcoat}.Result, False=@n13) {group=Skin FX, parameterName=Use Clearcoat}) {attributeGetTypes [#1]}
  n15 = Multiply(A=ChannelMaskParameter(None=@n3.RGBA) {group=Skin FX, parameterName=Dirt}.R, B=@n6.All Colorized)
  n16 = LinearInterpolate(
   A=@n14.Base Color
   B=LinearInterpolate(
    A=@n14.Base Color
    B=VectorParameter() {group=CustomColors, parameterName=CC_DirtColor, defaultValue=<rgba 0.0312 0.0261 0.0208 1>}.RGB
    Alpha=Saturate(None=@n15)
   )
   Alpha=ScalarParameter() {group=CustomColors, parameterName=Custom Dirt Intensity}
  )
  n17 = MaterialFunctionCall(
   AO=@n1.AO
   Distance Blend=@n1.AO
   Fresnel=@n1.AO
   Gmap=@n1.AO
   Pre Skinned Local Position=@n1.AO
   Material=SetMaterialAttributes!(MaterialAttributes=@n14.MaterialAttributes, Base Color=Saturate(None=@n16)) {attributeSetTypes [#1]}
  ) {materialFunction=&BaseFN_CharacterFunction_PostFX}
  StaticSwitch(
   True=@n17.Result
   False=SetMaterialAttributes!(MaterialAttributes=@n17.Result, Shading Model=ShadingModel() {shadingModel=MSM_SubsurfaceProfile}) {attributeSetTypes [#2]}
  ) {defaultValue=true}
 ]
}
