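# Master material for skeletal-mesh characters, backpacks and pickaxes that applies cosmetic wraps over the base character look.
# Character PreFX output feeds MF_Base_Customization, is blended with the wrapped result by a Vehicle_MASTER mask output, then runs through PostFX and accessory overrides.
# Object-bound, safe-zone and virtual-pivot parameters position the wrap, with an MF_WrapBounds front/back mask computed alongside.
# Zero-alpha lerps on Base Color reference wrap parameters without changing the colour.
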Material /Game/Packages/Fortress_SharedMaterials/Base_Material/M_FN_Character_Wrap_MASTER
graph {
 properties {
  bAutomaticallySetUsageInEditor=false
  bHasPixelAnimation=true
  bUseMaterialAttributes=true
  bUsedWithClothing=true
  bUsedWithMorphTargets=true
  bUsedWithSkeletalMesh=true
  bUsedWithStaticMesh=false
  pixelDepthOffsetMode=PDOM_Legacy
  refractionMethod=RM_IndexOfRefraction
  usedShadingModels=Default Lit
 }
 nodes [
  Constant() {r=1}
  LinearInterpolate()
  n1 = MaterialFunctionCall() {materialFunction=&BaseFN_CharacterFunction_PreFX}
  n2 = GetMaterialAttributes(None=@n1.Result) {attributeGetTypes [#1, #2]}
  n3 = VectorParameter() {group=Weapon Specific Wrap Adjustments, parameterName=DO NOT OVERRIDE-OBMAX, defaultValue=<rgba 0 0 0 1>}
  n4 = VectorParameter() {group=Weapon Specific Wrap Adjustments, parameterName=DO NOT OVERRIDE-OBMIN, defaultValue=<rgba 0 0 0 1>}
  n5 = VectorParameter() {group=Weapon Specific Wrap Adjustments, parameterName=DO NOT OVERRIDE-OBCENTER, defaultValue=<rgba 0 0 0 1>}
  n6 = Add(A=Add(A=@n3.RGB, B=@n4.RGB), B=@n5.RGB)
  n7 = VectorParameter() {group=Weapon Specific Wrap Adjustments, parameterName=DO NOT OVERRIDE-SZ-Back_Frnt, sortPriority=104, defaultValue=<rgba 0.3 0 0 0.2>}
  n8 = VectorParameter() {group=Weapon Specific Wrap Adjustments, parameterName=DO NOT OVERRIDE-SZ-Middle_Frnt, sortPriority=105, defaultValue=<rgba 0 0 0 0.2>}
  n9 = VectorParameter() {group=Weapon Specific Wrap Adjustments, parameterName=DO NOT OVERRIDE-SZ-Front_Frnt, sortPriority=106, defaultValue=<rgba -0.3 0 0 0.2>}
  n10 = Add(A=Add(A=@n7.RGB, B=@n8.RGB), B=@n9.RGB)
  n11 = VectorParameter() {group=Weapon Specific Wrap Adjustments, parameterName=DO NOT OVERRIDE-SZ-Back, sortPriority=100, defaultValue=<rgba 0.3 0 0 0.2>}
  n12 = VectorParameter() {group=Weapon Specific Wrap Adjustments, parameterName=DO NOT OVERRIDE-SZ-Middle, sortPriority=101, defaultValue=<rgba 0 0 0 0.2>}
  n13 = VectorParameter() {group=Weapon Specific Wrap Adjustments, parameterName=DO NOT OVERRIDE-SZ-Front, sortPriority=102, defaultValue=<rgba -0.3 0 0 0.2>}
  n14 = Add(A=Add(A=Add(A=@n10, B=@n11.RGB), B=@n12.RGB), B=@n13.RGB)
  n15 = VectorParameter() {group=Weapon Specific Wrap Adjustments, parameterName=DO NOT OVERRIDE-Virtual-Pivot, sortPriority=110, defaultValue=<rgba 0 0 0 0.5>}
  n16 = ScalarParameter() {defaultValue=1, group=Weapon Specific Wrap Adjustments, parameterName="DO NOT OVERRIDE   0=Disable full weapon wrap", sortPriority=0}
  n17 = Add(A=@n15.RGB, B=@n16)
  n18 = VectorParameter() {
   desc=for duel pick axe
   group=Weapon Specific Wrap Adjustments
   parameterName=DO NOT OVERRIDE-Virtual-Pivot InGame
   sortPriority=110
   defaultValue=<rgba 0 0 0 0.5>
  }
  n19 = VectorParameter() {parameterName=Vehicle Direction, defaultValue=<rgba 0 0 1 1>}
  n20 = Add(A=Add(A=Add(A=@n6, B=@n14), B=Add(A=@n17, B=@n18.RGB)), B=@n19.RGB)
  n21 = ScalarParameter() {group=Weapon Specific Wrap Adjustments, parameterName=Is Backpack, sortPriority=0}
  n22 = ScalarParameter() {
   desc=only use if accent area is not a texture streching area
   group=Weapon Specific Wrap Adjustments
   parameterName=Is Has_Char_Accent
   sortPriority=0
  }
  n23 = ScalarParameter() {parameterName=IsPickaxe}
  n24 = LinearInterpolate(A=LinearInterpolate(A=LinearInterpolate(A=@n2.Base Color, B=@n20) {constAlpha=0}, B=Add(A=@n21, B=@n22)) {constAlpha=0}, B=@n23) {constAlpha=0}
  LinearInterpolate(
   A=@n24
   B=Add(A=ScalarParameter() {parameterName=isMech}, B=ScalarParameter() {group=Weapon Specific Wrap Adjustments, parameterName=SIde Project Vehicle})
  ) {constAlpha=0}
  n25 = SetMaterialAttributes!(MaterialAttributes=@n2.MaterialAttributes, Base Color=@n24) {attributeSetTypes [#1]}
  n26 = ScalarParameter() {defaultValue=1, parameterName=Character Wrap Scale Multiplier}
  n27 = ScalarParameter() {parameterName=Vehicle X Axis Rotation, sliderMin=0.25}
  n28 = MaterialFunctionCall() {materialFunction=/Game/Packages/Fortress_SharedMaterials/Base_Material_Functions/BaseCustomizationFunctions/MF_FN_Vehicle_Master.MF_FN_Vehicle_MASTER}
  n29 = ScalarParameter() {defaultValue=0.25, parameterName=Vehicle Z axis Rotation}
  n30 = Constant() {r=1}
  n31 = Constant()
  n32 = Subtract(A=@n30, B=@n23) {desc="Special Override to indicate IsCharacter = 0 when Pickaxe = 1"}
  n33 = VectorParameter() {group=Weapon Specific Wrap Adjustments, parameterName=DO NOT OVERRIDE-Virtual-Pivot-Ch, sortPriority=110, defaultValue=<rgba 0 0 0 0.5>}
  n34 = ScalarParameter() {group=Weapon Specific Wrap Adjustments, parameterName=DO NOT OVERRIDE-Character-Position-BP-Switch, sortPriority=111}
  n35 = LinearInterpolate(
   A=LinearInterpolate(
    A=AppendVector(A=@n15.RGB, B=@n15.RGB)
    B=AppendVector(A=@n18.RGB, B=@n18.RGB)
    Alpha=ScalarParameter() {group=Weapon Specific Wrap Adjustments, parameterName=DO NOT OVERRIDE-Pickaxe-Position-BP-Switch, sortPriority=111}
   )
   B=AppendVector(A=@n33.RGB, B=@n15.A)
   Alpha=@n34
  )
  n36 = LinearInterpolate(A=AppendVector(A=@n7.RGB, B=@n7.RGB), B=AppendVector(A=@n11.RGB, B=@n11.RGB), Alpha=NamedRerouteUsage())
  n37 = LinearInterpolate(A=AppendVector(A=@n9.RGB, B=@n9.RGB), B=AppendVector(A=@n13.RGB, B=@n13.RGB), Alpha=NamedRerouteUsage())
  n38 = LinearInterpolate(A=AppendVector(A=@n8.RGB, B=@n8.RGB), B=AppendVector(A=@n12.RGB, B=@n12.RGB), Alpha=NamedRerouteUsage())
  n39 = MaterialFunctionCall(
   AO=@n1.AO
   Distance Blend=@n1.AO
   Fresnel=@n1.AO
   Gmap=@n1.AO
   Pre Skinned Local Position=@n1.AO
   Material=BlendMaterialAttributes(
    A=@n25
    B=MaterialFunctionCall!(
     Material=@n25
     ------------------
     Object Bounds Center Adjustment=@n5.RGB
     Object Bounds Min Adjustment=@n4.RGB
     Wrap ScaleMultiplier=@n26
     Rotation X Axis=@n27
     "Customization Mask(_CM A)"=@n28.Customization Gmap
     Forward Axis=@n19.RGB
     Object Bounds Max Adjustment=@n3.RGB
     Rotation Z Axis=@n29
     Normal Map=@n2.Normal
     Cropped Emissive Mask
     "Customization Crevasse (_CM G)"=@n28.Customization Gmap
     "Customization Mask UV Islands (_CM B Bottom Right)"=@n28.Customization Gmap
     -------------------
     "Customization Mask 1 (_CM B Top Left)"=@n28.Customization Gmap
     "Customization Mask 2 (_CM B Top Right)"=@n28.Customization Gmap
     "Customization Mask 3 (_CM B Bottom Left)"=@n28.Customization Gmap
     "Customization Mask (_CM RGBA)"=@n28.Customization Gmap
     --------------------
     Pre-Skinned Local Position=@n1.Pre skinned Local Position
     Pre-Skinned Bounds=@n28.Customization Gmap
     Pre-Skinned Local Normal=@n28.Customization Gmap
     Per_ObjectMaskOverride=@n16
     isVehicle=@n30
     Side Project Vehicle=@n31
     isMech=@n30
     isCharacter=@n32
     isCharAccent=@n22
     isPickaxe=@n23
     isBackpack=@n21
     Object-SafeZone-Back=@n36
     Object-SafeZone-Middle=@n38
     Object-SafeZone-Front=@n37
     Object-SafeZone-VP_Charcter_Preview_Switch=@n34
     Object-SafeZone-Virtual-Pivot=@n35
     Wrap Holdout
    ) {materialFunction=/Game/Packages/Fortress_SharedMaterials/Base_Material_Functions/BaseCustomizationFunctions/MF_Base_Customization}.Result
    Alpha=@n28."Customization Mask UV Islands (_CM B Bottom Right)"
   )
  ) {materialFunction=&BaseFN_CharacterFunction_PostFX}
  MaterialFunctionCall(In=@n39.Result) {materialFunction=&MF_TechArtOverrides_accessories}
  n40 = Constant() {r=1}
  n41 = NamedRerouteUsage()
  n42 = NamedRerouteUsage()
  n43 = NamedRerouteUsage()
  n44 = Constant()
  n45 = ComponentMask(None=Add(A=MaterialFunctionCall!(
     is Vehicle=@n40
     Object Bounds Max Adjustment=@n41
     Object Bounds Min Adjustment=@n42
     Object Bounds Center Adjustment=@n43
     isMech=@n40
     Side Project Vehicle=@n44
     Local Position=@n1.Pre skinned Local Position
    ) {materialFunction=/Game/Weapons/WeaponSkins/Wraps/MaterialFunctions/MF_WrapBounds}.Result) {constB=-0.5}) {g=true}
  n46 = Add(A=Saturate(None=Divide(A=@n45) {constB=-0.001}), B=NamedRerouteUsage())
  Multiply(A=NamedRerouteDeclaration(None=Saturate(None=@n46)) {name=MAsking_Frnt_Back, nodeColor=<rgba 0 1 0.2969 1>})
  NamedRerouteDeclaration(None=@n3.RGB) {name=Do_Not_Max, nodeColor=<rgba 1 0.375 0 1>}
  NamedRerouteDeclaration(None=@n4.RGB) {name=Do_Not_Min, nodeColor=<rgba 1 0.375 0 1>}
  NamedRerouteDeclaration(None=@n5.RGB) {name=Do_Not_Cen, nodeColor=<rgba 1 0.375 0 1>}
  NamedRerouteDeclaration(None=Max(A=@n21, B=@n23)) {name="is back\\pick", nodeColor=<rgba 1 0 0.5859 1>}
 ]
}
