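# Base material function for charm accessories that builds material attributes from Diffuse, SpecularMasks and Normals texture parameters.
# Diffuse RGB drives both base color and emissive, while the mask's G channel feeds both metallic and specular, leaving R unused.
# Mask B lerps between roughness min and max parameters.
# On ES2 and ES3_1, a static switch can substitute mobile-specific roughness bounds.
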MaterialFunction /Game/Accessories/Charms/MaterialFunctions/BaseFunctions/MF_CharmBaseFunction
graph {nodes [
  n1 = TextureSampleParameter2D() {group=Diffuse, parameterName=Diffuse, texture=/Engine/EngineMaterials/DefaultDiffuse}
  n2 = TextureSampleParameter2D() {
   desc="R = SpecMask G = Metallic B = Roughness"
   group=Specular
   parameterName=SpecularMasks
   samplerType=SAMPLERTYPE_Masks
   texture=&T_Fortnite_Default_S
  }
  n3 = LinearInterpolate(
   A=ScalarParameter() {group=Roughness, parameterName=RoughnessMin}
   B=ScalarParameter() {defaultValue=1, group=Roughness, parameterName=RoughnessMax}
   Alpha=@n2.B
  )
  n4 = StaticSwitchParameter(
   True=LinearInterpolate(
    A=ScalarParameter() {group=Roughness, parameterName=MobileRoughnessMin}
    B=ScalarParameter() {defaultValue=1, group=Roughness, parameterName=MobileRoughnessMax}
    Alpha=@n2.B
   )
   False=@n3
  ) {group=Roughness, parameterName=OverrideRoughnessOnMobile}
  n5 = TextureSampleParameter2D() {group=Normal, parameterName=Normals, samplerType=SAMPLERTYPE_Normal, texture=&FlatNormalBlue}
  FunctionOutput(None=MakeMaterialAttributes!(
    BaseColor=@n1.RGB
    Metallic=@n2.G
    Specular=@n2.G
    Roughness=FeatureLevelSwitch(Default=@n3, ES2=@n4, ES3_1=@n4, SM4_REMOVED=@n3)
    Anisotropy
    EmissiveColor=@n1.RGB
    Opacity
    OpacityMask
    Normal=@n5.RGB
    Tangent
    WorldPositionOffset
    SubsurfaceColor
    ClearCoat
    ClearCoatRoughness
    AmbientOcclusion
    Refraction
    CustomizedUVs_0
    CustomizedUVs_1
    CustomizedUVs_2
    CustomizedUVs_3
    CustomizedUVs_4
    CustomizedUVs_5
    CustomizedUVs_6
    CustomizedUVs_7
    PixelDepthOffset
    ShadingModel
    Displacement
   ))
 ]}
