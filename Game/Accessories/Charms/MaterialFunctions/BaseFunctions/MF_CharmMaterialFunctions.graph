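# Material function that layers optional emissive effects and vertex-color element hiding onto a charm's material attributes.
# It replaces incoming emissive with a parameter color, then static switches chain charm emissive texture, panning emissive and disco lights.
# The original incoming emissive serves as the panning and disco alpha, blended from one by UseFX-DiffuseAlphaMask.
# An optional World Position Offset step hides elements by vertex color, and emissive fresnel always runs last.
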MaterialFunction /Game/Accessories/Charms/MaterialFunctions/BaseFunctions/MF_CharmMaterialFunctions
graph {nodes [
  n1 = FunctionInput() {bUsePreviewValueAsDefault=true, inputName=MaterialAttributes, inputType=FunctionInput_MaterialAttributes}
  n2 = SetMaterialAttributes!(MaterialAttributes=@n1, EmissiveColor=VectorParameter() {group=CharmFX, parameterName=BasicEmissiveColor, defaultValue=<rgba 0 0 0 1>}.RGB) {attributeSetTypes [#1]}
  n3 = GetMaterialAttributes(None=@n2) {attributeGetTypes [#1]}
  n4 = MaterialFunctionCall() {materialFunction=/Game/Accessories/Charms/MaterialFunctions/MF_Charm_Emissive}
  n5 = StaticSwitchParameter(
   True=SetMaterialAttributes(MaterialAttributes=@n3.MaterialAttributes, EmissiveColor=Add(A=@n3.EmissiveColor, B=@n4.Result)) {attributeSetTypes [#1]}
   False=@n2
  ) {group=CharmFX, parameterName=UseCharmEmissiveTexture}
  n6 = LinearInterpolate(
   A=Constant() {r=1}
   B=GetMaterialAttributes(None=@n1) {attributeGetTypes [#1]}.EmissiveColor
   Alpha=ScalarParameter() {group=CharmFX, parameterName=UseFX-DiffuseAlphaMask}
  )
  n7 = StaticSwitchParameter(
   True=MaterialFunctionCall!(Alpha=@n6, isVehicle, Gmap, UVOffset, Local Normal, Local Position, EmissiveColorOverride, Material=@n5) {materialFunction=/Game/Weapons/WeaponSkins/MaterialFunction/PanningEmissive}.Result
   False=@n5
  ) {group=CharmFX, parameterName=UsePanningEmissive}
  n8 = StaticSwitchParameter(
   True=MaterialFunctionCall!(Material=@n7, Alpha=@n6) {materialFunction=/Game/Weapons/WeaponSkins/MaterialFunction/DiscoFloorFunction}.Result
   False=@n7
  ) {group=CharmFX, parameterName=Use Disco Lights}
  n9 = SetMaterialAttributes!(
   MaterialAttributes=GetMaterialAttributes(None=@n8).MaterialAttributes
   WorldPositionOffset=MaterialFunctionCall!(WPO) {materialFunction=/Game/Packages/Fortress_SharedMaterials/Base_Material_Functions/Fortnite_HideElementsWithVertexColor}.Result
  ) {attributeSetTypes [#2]}
  FunctionOutput(None=MaterialFunctionCall!(In=StaticSwitchParameter(True=@n9, False=@n8) {group=CharmFX, parameterName=Use Vertex Colors for Mask}) {materialFunction=/Game/Weapons/WeaponSkins/MaterialFunction/MF_EmissiveFresnel}.Result)
 ]}
