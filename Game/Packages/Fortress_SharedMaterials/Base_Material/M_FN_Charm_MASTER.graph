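# Master material for Fortress charm accessories, a Default Lit surface enabled for skeletal meshes and mesh deformers.
# It uses material attributes and keeps all of its logic in three chained material functions.
# MF_CharmBaseFunction builds the attributes and passes them through MF_CharmMaterialFunctions.
# CharacterDeRez_Transition wraps that result last, adding the shared character de-rez transition.
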Material /Game/Packages/Fortress_SharedMaterials/Base_Material/M_FN_Charm_MASTER
graph {
 properties {
  bAutomaticallySetUsageInEditor=false
  bUseMaterialAttributes=true
  bUsedWithMeshDeformer=true
  bUsedWithSkeletalMesh=true
  pixelDepthOffsetMode=PDOM_Legacy
  usedShadingModels=Default Lit
 }
 nodes [MaterialFunctionCall!(
   Fresnel
   Material=MaterialFunctionCall!(MaterialAttributes=MaterialFunctionCall() {materialFunction=/Game/Accessories/Charms/MaterialFunctions/BaseFunctions/MF_CharmBaseFunction}.Result) {materialFunction=/Game/Accessories/Charms/MaterialFunctions/BaseFunctions/MF_CharmMaterialFunctions}.Result
  ) {materialFunction=/Game/Packages/Fortress_SharedMaterials/Base_Material_Functions/CharacterDeRez_Transition}]
}
