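# Master material for Fortnite cosmetic customization, applying wraps through shared material functions using material attributes.
# A dummy master function supplies the base material, normal, customization masks and pre-skinned position, bounds and normal to MF_Base_Customization.
# Primary Triplaner parameters drive wrap scale and rotation, with vehicle and global Z rotations summed.
# The customization result feeds an accessories tech-art override function and the Deferred input of a shading path switch, whose default is constant.
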Material /Game/Packages/Fortress_SharedMaterials/Base_Material/M_FN_Customization_MASTER
graph {
 properties {
  bAutomaticallySetUsageInEditor=false
  bHasPixelAnimation=true
  bUseMaterialAttributes=true
  bUsedWithStaticMesh=false
  pixelDepthOffsetMode=PDOM_Legacy
  refractionMethod=RM_IndexOfRefraction
  usedShadingModels=Default Lit
 }
 nodes [
  n1 = MaterialFunctionCall() {materialFunction=/Game/Packages/Fortress_SharedMaterials/Base_Material_Functions/BaseCustomizationFunctions/MF_FN_Dummy_MASTER}
  n2 = MaterialFunctionCall!(
   Material=@n1.Material Result
   ------------------
   Object Bounds Center Adjustment
   Object Bounds Min Adjustment
   Wrap ScaleMultiplier=ScalarParameter() {defaultValue=1, group=Primary Triplaner, parameterName=Vehicle Wrap Scale Multiplier, sortPriority=-49}
   Rotation X Axis=ScalarParameter() {group=Primary Triplaner, parameterName=Global X Rot, sortPriority=-39}
   "Customization Mask(_CM A)"=@n1.Material Result
   Forward Axis
   Object Bounds Max Adjustment
   Rotation Z Axis=Add(
    A=ScalarParameter() {group=Primary Triplaner, parameterName=Vehicle Z axis Rotation, sortPriority=-32}
    B=ScalarParameter() {group=Primary Triplaner, parameterName=Global Z Rot}
   )
   Normal Map=@n1.Material Result
   Cropped Emissive Mask=@n1.Material Result
   "Customization Crevasse (_CM G)"=@n1.Material Result
   "Customization Mask UV Islands (_CM B Bottom Right)"=@n1.Material Result
   -------------------
   "Customization Mask 1 (_CM B Top Left)"=@n1.Material Result
   "Customization Mask 2 (_CM B Top Right)"=@n1.Material Result
   "Customization Mask 3 (_CM B Bottom Left)"=@n1.Material Result
   "Customization Mask (_CM RGBA)"
   --------------------
   Pre-Skinned Local Position=@n1.Material Result
   Pre-Skinned Bounds=@n1.Material Result
   Pre-Skinned Local Normal=@n1.Material Result
   Per_ObjectMaskOverride
   isVehicle
   Side Project Vehicle
   isMech
   isCharacter
   isCharAccent
   isPickaxe
   isBackpack
   Object-SafeZone-Back
   Object-SafeZone-Middle
   Object-SafeZone-Front
   Object-SafeZone-VP_Charcter_Preview_Switch
   Object-SafeZone-Virtual-Pivot
   Wrap Holdout
  ) {materialFunction=/Game/Packages/Fortress_SharedMaterials/Base_Material_Functions/BaseCustomizationFunctions/MF_Base_Customization}
  MaterialFunctionCall(In=@n2.Result) {materialFunction=&MF_TechArtOverrides_accessories}
  ShadingPathSwitch(Default=Constant(), Deferred=@n2.Result)
 ]
}
