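# Pressure gauge material, defaulting to pneumatic shotgun textures, that animates a gauge needle from a Pressure custom primitive data value.
# Pressure sweeps an angle for RotateAboutAxis around a parameterized pivot and local axis, with sine vibration growing near full pressure or with Overcharge.
# Crossing into the top pressure range triggers a glow that fades to a faint level at full.
# Static switches provide debug pressure and a pivot-visualizing sphere mask.
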Material /Game/Packages/Fortress_SharedMaterials/Base_Material/M_FN_PressureGauge
graph {
 properties {
  bTangentSpaceNormal=false
  bUsedWithMeshDeformer=true
  bUsedWithSkeletalMesh=true
  pixelDepthOffsetMode=PDOM_Legacy
  usedShadingModels=Default Lit
 }
 nodes [
  n1 = TextureSampleParameter2D() {group=Textures, parameterName=Base Color, texture=/Game/Weapons/FORT_Shotguns/Textures/T_Pneumatic_Shotgun_Base_D}
  n2 = TransformPosition(None=VectorParameter() {group=Rotation Setup, parameterName=CenterOfRotation, defaultValue=<rgba -2.57 21.15 12.05 0>}.RGB) {transformType=TRANSFORMPOSSOURCE_World}
  n3 = WorldPosition()
  n4 = ScalarParameter() {defaultValue=0.05, group=DEBUG, parameterName=Radius}
  Add(
   A=@n1.RGB
   B=StaticSwitchParameter(True=SphereMask(A=@n2, B=@n3.XYZ, Radius=@n4) {attenuationRadius=1}, False=@n1.RGB) {desc=Allows for visualization of center of rotation, group=DEBUG, parameterName=DEBUG Center of Rotation}
  )
  Constant()
  n5 = ScalarParameter() {defaultValue=8, parameterName=Charge Flash Size}
  n6 = ScalarParameter() {bUseCustomPrimitiveData=true, group=Custom Primitive Data, parameterName=Pressure, primitiveDataIndex=31, sliderMax=1}
  n7 = StaticSwitchParameter(True=ScalarParameter() {group=DEBUG, parameterName=Debug Pressure, sliderMax=1}, False=@n6) {group=DEBUG, parameterName=DEBUG Pressure}
  n8 = Saturate(None=Subtract(A=Multiply(A=@n7, B=@n5) {constB=10}, B=Subtract(A=@n5)) {constB=9})
  n9 = OneMinus(None=@n8)
  n10 = Multiply(A=@n9, B=@n9)
  Multiply(A=VectorParameter() {parameterName=Glow Color, defaultValue=<rgba 3 0 0 0>}.RGB, B=Multiply(A=Add(A=@n10) {constB=0.05}, B=Ceil(None=@n8)))
  n11 = Frac(None=Multiply(A=ScalarParameter() {defaultValue=0.2, group=Sweep, parameterName=Rotation Speed}, B=Time()))
  StaticSwitchParameter(True=Saturate(None=Multiply(A=@n11) {constB=4}), False=@n7) {group=DEBUG, parameterName=DEBUG Center of Rotation}
  n12 = Transform(None=VectorParameter() {group=Rotation Setup, parameterName=RotationAxis, defaultValue=<rgba 0 1 -0.2 0>}.RGB) {transformSourceType=TRANSFORMSOURCE_Local}
  n13 = LinearInterpolate(
   A=ScalarParameter() {defaultValue=0.2, group=Sweep, parameterName=Angle Start}
   B=ScalarParameter() {defaultValue=-0.2, group=Sweep, parameterName=Angle Final}
   Alpha=@n7
  )
  n14 = ScalarParameter() {defaultValue=5, group=Vibration, parameterName=Vibration Region}
  n15 = Saturate(None=Subtract(A=Multiply(A=@n7, B=@n14) {constB=10}, B=Subtract(A=@n14)) {constB=9})
  n16 = ScalarParameter() {bUseCustomPrimitiveData=true, group=Custom Primitive Data, parameterName=Overcharge, primitiveDataIndex=30, sliderMax=1}
  n17 = Sine(None=Multiply(A=Time() {bOverride_Period=true, period=5000}, B=ScalarParameter() {defaultValue=11, group=Vibration, parameterName=Vibration Speed}) {constB=3})
  n18 = Multiply(
   A=Max(A=Multiply(A=@n15, B=@n15), B=@n16)
   B=Multiply(A=ConstantBiasScale(None=@n17), B=ScalarParameter() {defaultValue=0.05, group=Vibration, parameterName=Vibration})
  )
  StaticSwitchParameter(
   True=Constant()
   False=RotateAboutAxis(NormalizedRotationAxis=Normalize(VectorInput=@n12), RotationAngle=Add(A=@n13, B=@n18), PivotPoint=@n2, Position=@n3.XYZ)
  ) {desc=Allows for visualization of center of rotation, group=DEBUG, parameterName=DEBUG Center of Rotation}
  TextureSampleParameter2D() {
   group=Textures
   parameterName=Normal Map
   samplerType=SAMPLERTYPE_Normal
   texture=/Game/Weapons/FORT_Shotguns/Textures/T_Pneumatic_Shotgun_Base_N
  }
  TextureSampleParameter2D() {
   group=Textures
   parameterName=Specular
   samplerType=SAMPLERTYPE_Masks
   texture=/Game/Weapons/FORT_Shotguns/Textures/T_Pneumatic_Shotgun_Base_S
  }
 ]
}
