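# Master Default Lit material for Fortress skeletal-mesh vehicles, with wrap customization, built on material attributes.
# MF_FN_Vehicle_MASTER supplies the surface and masks to MF_Base_Customization, which takes vehicle direction, bounds, safe zone, mech and side-projection parameters.
# Vehicle emissive, cropped emissive, damage, tech-art overrides, camera dither and Creative move-tool functions follow in series.
# Zero-alpha lerps reference wrap parameters in Base Color without changing it.
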Material /Game/Packages/Fortress_SharedMaterials/Base_Material/M_FN_Vehicle_MASTER
graph {
 properties {
  bAutomaticallySetUsageInEditor=false
  bHasPixelAnimation=true
  bUseMaterialAttributes=true
  bUsedWithSkeletalMesh=true
  bUsedWithStaticMesh=false
  pixelDepthOffsetMode=PDOM_Legacy
  refractionMethod=RM_IndexOfRefraction
  usedShadingModels=Default Lit
 }
 nodes [
  LinearInterpolate()
  n1 = MaterialFunctionCall() {materialFunction=/Game/Packages/Fortress_SharedMaterials/Base_Material_Functions/BaseCustomizationFunctions/MF_FN_Vehicle_Master.MF_FN_Vehicle_MASTER}
  n2 = GetMaterialAttributes(None=@n1.Material Result) {attributeGetTypes [#1]}
  n3 = VectorParameter() {group=Weapon Specific Wrap Adjustments, parameterName=DO NOT OVERRIDE-OBMAX, defaultValue=<rgba 0 0 0 1>}
  n4 = VectorParameter() {group=Weapon Specific Wrap Adjustments, parameterName=DO NOT OVERRIDE-OBMIN, defaultValue=<rgba 0 0 0 1>}
  n5 = VectorParameter() {group=Weapon Specific Wrap Adjustments, parameterName=DO NOT OVERRIDE-OBCENTER, defaultValue=<rgba 0 0 0 1>}
  n6 = Add(A=Add(A=@n3.RGB, B=@n4.RGB), B=@n5.RGB)
  n7 = VectorParameter() {group=Weapon Specific Wrap Adjustments, parameterName=DO NOT OVERRIDE-SZ-Back, sortPriority=100, defaultValue=<rgba 0.3 0 0 0.2>}
  n8 = VectorParameter() {group=Weapon Specific Wrap Adjustments, parameterName=DO NOT OVERRIDE-SZ-Middle, sortPriority=101, defaultValue=<rgba 0 0 0 0.2>}
  n9 = VectorParameter() {group=Weapon Specific Wrap Adjustments, parameterName=DO NOT OVERRIDE-SZ-Front, sortPriority=102, defaultValue=<rgba -0.3 0 0 0.2>}
  n10 = Add(A=Add(A=@n7.RGB, B=@n8.RGB), B=@n9.RGB)
  n11 = VectorParameter() {group=Weapon Specific Wrap Adjustments, parameterName=DO NOT OVERRIDE-Virtual-Pivot, sortPriority=110, defaultValue=<rgba 0 0 0 0.5>}
  n12 = ScalarParameter() {defaultValue=1, group=Weapon Specific Wrap Adjustments, parameterName="DO NOT OVERRIDE   0=Disable full weapon wrap", sortPriority=0}
  n13 = VectorParameter() {parameterName=Vehicle Direction, defaultValue=<rgba 1 0 0 1>}
  n14 = Add(A=Add(A=Add(A=@n6, B=@n10), B=Add(A=@n11.RGB, B=@n12)), B=@n13.RGB)
  n15 = ScalarParameter() {parameterName=isMech}
  n16 = ScalarParameter() {group=Weapon Specific Wrap Adjustments, parameterName=SIde Project Vehicle}
  n17 = LinearInterpolate(A=LinearInterpolate(A=LinearInterpolate(A=@n2.Base Color, B=@n14) {constAlpha=0}, B=@n15) {constAlpha=0}, B=@n16) {constAlpha=0}
  n18 = Constant()
  n19 = MaterialFunctionCall!(
   Material=MaterialFunctionCall!(
    Material=SetMaterialAttributes!(MaterialAttributes=@n2.MaterialAttributes, Base Color=@n17) {attributeSetTypes [#1]}
    ------------------
    Object Bounds Center Adjustment=@n5.RGB
    Object Bounds Min Adjustment=@n4.RGB
    Wrap ScaleMultiplier=ScalarParameter() {defaultValue=3.5, parameterName=Vehicle Wrap Scale Multiplier}
    Rotation X Axis=ScalarParameter() {defaultValue=0.25, parameterName=Vehicle X Axis Rotation}
    "Customization Mask(_CM A)"=@n1.Customization Gmap
    Forward Axis=@n13.RGB
    Object Bounds Max Adjustment=@n3.RGB
    Rotation Z Axis=ScalarParameter() {defaultValue=0.25, parameterName=Vehicle Z axis Rotation}
    Normal Map=@n1.Customization Gmap
    Cropped Emissive Mask
    "Customization Crevasse (_CM G)"=@n1.Customization Gmap
    "Customization Mask UV Islands (_CM B Bottom Right)"=@n1.Customization Gmap
    -------------------
    "Customization Mask 1 (_CM B Top Left)"=@n1.Customization Gmap
    "Customization Mask 2 (_CM B Top Right)"=@n1.Customization Gmap
    "Customization Mask 3 (_CM B Bottom Left)"=@n1.Customization Gmap
    "Customization Mask (_CM RGBA)"=@n1.Customization Gmap
    --------------------
    Pre-Skinned Local Position=@n1.Customization Gmap
    Pre-Skinned Bounds=@n1.Customization Gmap
    Pre-Skinned Local Normal=@n1.Customization Gmap
    Per_ObjectMaskOverride=@n12
    isVehicle=Constant() {r=1}
    Side Project Vehicle=@n16
    isMech=@n15
    isCharacter=@n18
    isCharAccent
    isPickaxe=@n18
    isBackpack=@n18
    Object-SafeZone-Back=AppendVector(A=@n7.RGB, B=@n7.RGB)
    Object-SafeZone-Middle=AppendVector(A=@n8.RGB, B=@n8.RGB)
    Object-SafeZone-Front=AppendVector(A=@n9.RGB, B=@n9.RGB)
    Object-SafeZone-VP_Charcter_Preview_Switch
    Object-SafeZone-Virtual-Pivot=AppendVector(A=@n11.RGB, B=@n11.RGB)
    Wrap Holdout
   ) {materialFunction=/Game/Packages/Fortress_SharedMaterials/Base_Material_Functions/BaseCustomizationFunctions/MF_Base_Customization}.Result
   "Tail Pipe Mask (_VL R)"=@n1."Tail Pipe Mask (_VL R)"
   "Head Light Mask (_VL G)"=@n1."Tail Pipe Mask (_VL R)"
   "Tail Light Mask (_VL B)"=@n1."Tail Pipe Mask (_VL R)"
  ) {materialFunction=/Game/Packages/Fortress_SharedMaterials/Base_Material_Functions/BaseCustomizationFunctions/MF_BaseVehicle_Emissive}
  n20 = MaterialFunctionCall!(Material=@n19.Result, Cropped Emissive Mask=@n19.Result) {materialFunction=/Game/Packages/Fortress_SharedMaterials/Base_Material_Functions/BaseCustomizationFunctions/MF_Base_Customization_CroppedEmissive}
  n21 = MaterialFunctionCall(In=MaterialFunctionCall(Material=@n20.Result, "Clear Coat/Damage Mask (_M B)"=@n1."CC/Damage Mask (_M B)") {materialFunction=&MF_BaseVehicle_Damage}.Result) {materialFunction=&MF_TechArtOverrides_accessories}
  MaterialFunctionCall(MA In=MaterialFunctionCall!(In=@n21.Result) {materialFunction=/Game/Packages/Fortress_SharedMaterials/Base_Material_Functions/BaseCustomizationFunctions/MF_FN_Vehicle_CameraDither}.Result) {materialFunction=&MF_CreativeMoveTool}
 ]
}
