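# Parent Default Lit material, named for Fortnite weapon mods, usable on skeletal, instanced static and Niagara mesh particles.
# It chains material attribute functions: the weapon master feeds ADS scope and anti-camera-clipping layers, then scope masking, tech art accessory overrides and the Creative move tool.
# Before scope masking, the static switch "Use Vertex Colors for Mask" can replace World Position Offset with a function that hides elements by vertex color.
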Material /Game/Packages/Fortress_SharedMaterials/Base_Material/M_FN_WeaponMods_Parent
graph {
 properties {
  bAutomaticallySetUsageInEditor=false
  bUseMaterialAttributes=true
  bUsedWithInstancedStaticMeshes=true
  bUsedWithMeshDeformer=true
  bUsedWithNiagaraMeshParticles=true
  bUsedWithSkeletalMesh=true
  pixelDepthOffsetMode=PDOM_Legacy
  refractionMethod=RM_IndexOfRefraction
  usedShadingModels=Default Lit
 }
 nodes [
  n1 = MaterialFunctionCall!(In) {materialFunction=/Game/Packages/Fortress_SharedMaterials/Base_Material_Functions/BaseCustomizationFunctions/MF_FN_Weapon_MASTER}
  n2 = GetMaterialAttributes(None=MaterialFunctionCall!(In=MaterialFunctionCall!(Material Attributes=@n1.Material Result) {materialFunction=/Game/Packages/Fortnite_UniqueMaterials_1/Weapons/ADSScopeMaster}.Result) {materialFunction=/Game/Packages/Fortress_SharedMaterials/Base_Material_Functions/BaseCharacterShaderFunctions/MF_AntiCameraClipping}.Result) {attributeGetTypes [#1]}
  n3 = StaticSwitchParameter(
   True=MaterialFunctionCall!(WPO) {materialFunction=/Game/Packages/Fortress_SharedMaterials/Base_Material_Functions/Fortnite_HideElementsWithVertexColor}.Result
   False=@n2.World Position Offset
  ) {group=Opacity Mask, parameterName=Use Vertex Colors for Mask}
  n4 = MaterialFunctionCall(In=MaterialFunctionCall!(In=SetMaterialAttributes!(MaterialAttributes=@n2.MaterialAttributes, World Position Offset=@n3) {attributeSetTypes [#1]}) {materialFunction=/Game/Packages/Fortress_SharedMaterials/Base_Material_Functions/MF_ScopeMasking}.Result) {materialFunction=&MF_TechArtOverrides_accessories}
  MaterialFunctionCall(MA In=@n4.Result) {materialFunction=&MF_CreativeMoveTool}
  VertexInterpolator(VS=PreSkinnedPosition())
 ]
}
