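# Translucent lit parent material for weapon scope lenses, with a static switch for simple Fresnel glass.
# Front/back faces get remapped UVs, procedural curved normals, hue-shifted tints and sky-light reflections.
# A sphere-mask opening shapes opacity, while eye-socket distance and alignment drive fade terms.
# Optional FPS shading adds emissive scene color, magnified and distorted around a projected first-person point and masked with bump-offset parallax; mobile omits scene color.
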Material /Game/Packages/Fortress_SharedMaterials/Base_Material/M_FN_WeaponOptics_Parent
graph {
 properties {
  bUseMaterialAttributes=true
  bUsedWithInstancedStaticMeshes=true
  bUsedWithMeshDeformer=true
  bUsedWithNanite=true
  bUsedWithSkeletalMesh=true
  blendMode=BLEND_Translucent
  pixelDepthOffsetMode=PDOM_Legacy
  usedShadingModels=Default Lit
 }
 nodes [
  n1 = VectorParameter() {parameterName=PositionOffset, defaultValue=<rgba 0 -2.5 2.5 1>}
  n2 = TransformPosition(None=@n1.RGB) {transformType=TRANSFORMPOSSOURCE_TranslatedWorld}
  n3 = Constant() {r=1}
  n4 = MaterialFunctionCall(World Position=TransformPosition(None=MaterialFunctionCall!(CameraRelativePosition=@n2, Scale, FirstPersonLerpAlpha=@n3) {materialFunction=/Game/Global/MaterialFunctions/MF_FirstPersonTransforms}.CameraRealtiveFP) {transformSourceType=TRANSFORMPOSSOURCE_TranslatedWorld, transformType=TRANSFORMPOSSOURCE_World}) {materialFunction=&TransformToClipSpace}
  n5 = ComponentMask(None=@n4.Clip Space XY) {g=true, r=true}
  n6 = NamedRerouteUsage()
  n7 = Frac(None=Multiply(
    A=MaterialFunctionCall(XY=@n6) {materialFunction=&Swizzle}.YX
    B=Constant2Vector() {g=1, r=-1}
   ))
  n8 = ScalarParameter() {defaultValue=0.7875, group=FPSShading, parameterName=ScopeSceneTextureScale}
  n9 = NamedRerouteUsage()
  n10 = Power(Base=SphereMask(
    A=@n9
    B=Constant2Vector() {g=0.5, r=0.5}
    Radius=NamedRerouteUsage()
    Hardness=ScalarParameter() {defaultValue=0.8102, group=FPSShading, parameterName=ScopeEndHardness2}
   ) {hardnessPercent=0})
  n11 = LinearInterpolate(
   A=@n8
   B=Add(A=@n8, B=ScalarParameter() {defaultValue=0.797, desc=0.797027, group=FPSShading, parameterName=ScopeSceneTextureScaleDistort})
   Alpha=@n10
  )
  n12 = Multiply(A=@n4.Clip Space XY, B=Constant2Vector() {g=-0.5, r=0.5}) {constB=-1}
  n13 = Divide(A=Subtract(A=ScreenPosition().ViewportUV, B=@n12), B=@n11)
  n14 = Constant() {r=0.5}
  n15 = SceneColor!(UVs=Subtract(A=Add(A=@n13, B=@n14), B=Divide(A=@n14, B=@n11) {desc=Scale the center offset by the scaling factor}) {desc=Re offset to maintain center}) {constInput=<0.75 0>}
  n16 = Multiply(A=ScalarParameter() {defaultValue=0.112, group=FPSShading, parameterName=SceneBrightness}, B=@n10)
  n17 = MaterialFunctionCall!(
   UV Coord=@n9
   Height=ScalarParameter() {defaultValue=8.867, group=FPSShading, parameterName=ScopeEndOffset}
   ReferencePlane
   HeightRatio
   Normal
  ) {materialFunction=/Game/Global/MaterialFunctions/MF_BumpOffset_Accurate}
  n18 = Constant2Vector() {g=0.5, r=0.5}
  n19 = NamedRerouteUsage()
  n20 = ScalarParameter() {defaultValue=0.6666, group=FPSShading, parameterName=ScopeEndHardness}
  n21 = SphereMask(A=NamedRerouteUsage(), B=Constant2Vector() {g=0.5, r=0.5}, Radius=NamedRerouteUsage()) {hardnessPercent=0}
  n22 = ScalarParameter() {defaultValue=0.01, group=FPSShading, parameterName=ReflectonBrightnessAddPure}
  n23 = EyeAdaptationInverse(LightValueInput=Multiply(A=NamedRerouteUsage(), B=Multiply(A=@n21, B=@n22)))
  n24 = Add(
   A=LinearInterpolate(A=Constant3Vector(), B=Multiply(A=@n15.RGB, B=@n16), Alpha=SphereMask(A=@n17.Result, B=@n18, Radius=@n19, Hardness=@n20))
   B=@n23.EyeAdaptationInverse
  )
  Add(
   A=AppendVector(
    A=MaterialFunctionCall!(Vector2=@n5, MaximumNumberOfDigits, UVs=@n7, DebugTextLocation RG_UpperRight BA_LowerLeft, Component Spacing) {materialFunction=/Engine/Functions/Engine_MaterialFunctions02/Utility/DebugFloat2Values}.ColorCodedOutput
    B=Constant()
   )
   B=@n24
  )
  Add(
   A=@n12
   B=Multiply(A=ComponentMask(None=@n1.RGB) {g=true, r=true}, B=ViewProperty() {property=MEVP_ViewSize}.InvProperty) {desc=0-1 per pixel}
  )
  CameraPositionWS()
  n25 = ComponentMask(None=VectorParameter() {group=FPSShading, parameterName=MaxOffsetScreen, defaultValue=<rgba 0.5 0.5 0 1>}.RGB) {g=true, r=true}
  Clamp(None=@n12, Min=Multiply(A=@n25) {constB=-1}, Max=@n25)
  Constant2Vector() {g=-0.5, r=-0.5}
  Constant2Vector() {g=0.5, r=0.5}
  Constant2Vector() {g=1, r=1}
  Constant()
  n26 = NamedRerouteUsage()
  n27 = Constant() {r=0.5}
  n28 = LinearInterpolate(
   A=AppendVector(A=@n26, B=SphereMask(A=@n26, B=@n27) {attenuationRadius=0.1})
   B=ConstantBiasScale(None=NamedRerouteUsage())
   Alpha=ScalarParameter() {group=Debug, parameterName=PreviewGlassNormals}
  )
  n29 = NamedRerouteUsage()
  n30 = NamedRerouteUsage()
  n31 = ScalarParameter() {defaultValue=3, group=Simple Glass, parameterName=Simple Glass Exponent}
  n32 = ScalarParameter() {defaultValue=0.02, group=Simple Glass, parameterName=Simple Glass Base Reflection}
  n33 = NamedRerouteUsage()
  n34 = NamedRerouteUsage()
  n35 = Fresnel(ExponentIn=ScalarParameter() {defaultValue=1, group=Lens Color, parameterName=LensHueShiftFresnelExponent}, Normal=NamedRerouteUsage()) {baseReflectFraction=0, exponent=2}
  n36 = ScalarParameter() {group=Lens Color, parameterName=LensEyeHueShift}
  n37 = ScalarParameter() {group=Lens Color, parameterName=LensTargetHueShift}
  n38 = NamedRerouteUsage()
  n39 = MaterialFunctionCall(
   Hue Shift Percentage=Multiply(A=@n35, B=LinearInterpolate(A=@n36, B=@n37, Alpha=@n38))
   Texture=LinearInterpolate(
    A=VectorParameter() {group=Lens Color, parameterName=LensEyePieceTint, defaultValue=<rgba 0.1236 0.1521 0.2083 1>}.RGB
    B=VectorParameter() {group=Lens Color, parameterName=LensTargetPieceTint, defaultValue=<rgba 0.1529 1 0.3577 1>}.RGB
    Alpha=NamedRerouteUsage()
   )
  ) {materialFunction=&HueShift}
  n40 = LinearInterpolate(A=@n39.Result, Alpha=Fresnel(Normal=NamedRerouteUsage()))
  n41 = StaticSwitchParameter(True=TextureCoordinate(), False=NamedRerouteUsage()) {group=Scope Mask, parameterName=Use Standard UV Coords for Scope Mask?}
  n42 = Constant2Vector() {g=0.5, r=0.5}
  n43 = LinearInterpolate(A=ScalarParameter() {defaultValue=0.5, group=Scope Mask, parameterName=ScopeOpeningRadius}, Alpha=NamedRerouteUsage()) {constB=0}
  n44 = ScalarParameter() {group=Scope Mask, parameterName=ScopeOpeningHardness}
  n45 = OneMinus(None=Power(
    Base=SphereMask(A=@n41, B=@n42, Radius=@n43, Hardness=@n44)
    Exp=ScalarParameter() {defaultValue=1, group=Scope Mask, parameterName=ScopeOpeningExp}
   ))
  n46 = Multiply(
   A=Multiply(A=@n45, B=@n45)
   B=LinearInterpolate(A=ScalarParameter() {defaultValue=0.985, group=Scope Mask, parameterName=Max Opacity}, Alpha=NamedRerouteUsage()) {constA=0.99}
  )
  n47 = Transform(None=Constant3Vector() {constant=<rgba 0 1 0 0>}) {transformSourceType=TRANSFORMSOURCE_Instance, transformType=TRANSFORM_View}
  n48 = Constant3Vector() {constant=<rgba 0 0 1 0>}
  n49 = Saturate(None=InverseLinearInterpolate(Value=DotProduct(A=@n47, B=@n48)) {constA=0.99})
  n50 = SetMaterialAttributes!(
   MaterialAttributes
   Base Color=LinearInterpolate(B=Multiply(A=@n34, B=@n40), Alpha=NamedRerouteUsage()) {constB=0, desc="Only show reflections in\r\n3rd person"}
   Opacity=Saturate(None=LinearInterpolate(B=@n46, Alpha=@n49) {constA=1})
  ) {attributeSetTypes [#1, #2]}
  n51 = Multiply(A=OneMinus(None=NamedRerouteUsage())) {constB=2}
  n52 = LinearInterpolate(B=NamedRerouteUsage(), Alpha=OneMinus(None=Saturate(None=@n51))) {constB=0, desc="Only show reflections in\r\n3rd person"}
  n53 = StaticSwitchParameter(
   True=SetMaterialAttributes(
    Base Color=LinearInterpolate(B=@n29, Alpha=@n30) {constB=0, desc="Only show reflections in\r\n3rd person"}
    Opacity=Fresnel(ExponentIn=@n31, BaseReflectFractionIn=@n32, Normal=@n33) {exponent=3}
   ) {attributeSetTypes [#1, #2]}
   False=StaticSwitchParameter(
    True=SetMaterialAttributes!(MaterialAttributes=@n50, Emissive Color=@n52) {attributeSetTypes [#3]}
    False=@n50
   ) {group=Simple Glass, parameterName=Use FPS ComplexShading, sortPriority=5}
  ) {group=Simple Glass, parameterName=Use Simple Glass, sortPriority=5}
  n54 = ScalarParameter() {group=Reflection, parameterName=OpaqueOnly - Roughness, sliderMax=1, sortPriority=500}
  n55 = ScalarParameter() {group=Reflection, parameterName=OpaqueOnly - FlipReflectionEye, sliderMax=1, sortPriority=501}
  n56 = ScalarParameter() {group=Reflection, parameterName=OpaqueOnly - FlipReflectionTarget, sliderMax=1, sortPriority=502}
  n57 = NamedRerouteUsage()
  n58 = Multiply(
   A=NamedRerouteUsage()
   B=LinearInterpolate(
    A=Constant3Vector() {constant=<rgba 1 1 1 0>}
    B=Constant3Vector() {constant=<rgba 1 1 1 0>}
    Alpha=LinearInterpolate(A=@n55, B=@n56, Alpha=@n57)
   )
  ) {constB=-1}
  n59 = MaterialFunctionCall(In=StaticSwitchParameter(
    True=SetMaterialAttributes(Emissive Color=@n28) {attributeSetTypes [#3]}
    False=SetMaterialAttributes!(MaterialAttributes=@n53, Roughness=@n54, Normal=@n58) {attributeSetTypes [#4, #5]}
   ) {group=Debug, parameterName=Debug UVs}) {materialFunction=&MF_TechArtOverrides_accessories}
  MaterialFunctionCall(MA In=@n59.Result) {materialFunction=&MF_CreativeMoveTool}
  n60 = Subtract(A=MaterialFunctionCall(UVs=NamedRerouteUsage(), Texture Scale=ScalarParameter() {defaultValue=1, group=Coordinates, parameterName=Lens UV Scale}) {materialFunction=&ScaleUVsByCenter}.UVs) {constB=0.5}
  n61 = Distance(A=Constant(), B=@n60)
  n62 = Multiply(
   A=ScalarParameter() {defaultValue=4, group=Coordinates, parameterName=Lens Curve}
   B=LinearInterpolate(
    A=ScalarParameter() {defaultValue=1, group=Coordinates, parameterName=Lens Curve Multiplier - Eye, sortPriority=501}
    B=ScalarParameter() {defaultValue=1, group=Coordinates, parameterName=Lens Curve Multiplier - Target, sortPriority=502}
    Alpha=NamedRerouteUsage()
   )
  )
  n63 = Multiply(
   A=AppendVector(A=@n60, B=OneMinus(None=@n61))
   B=MaterialFunctionCall(X=@n62, Y=Multiply(A=@n62) {constB=-1}, Z=Constant() {r=1}) {materialFunction=&MakeFloat3}.Result
  )
  n64 = StaticSwitchParameter(True=Add(A=VectorParameter() {group=FPSShading, parameterName=BiasReflection, defaultValue=<rgba -0.2 0 0 1>}.RGB, B=@n63), False=@n63) {group=Simple Glass, parameterName=Use FPS ComplexShading, sortPriority=5}
  n65 = Transform(None=NamedRerouteDeclaration(None=Normalize(VectorInput=@n64)) {name=Glass Normal Tangent, nodeColor=<rgba 1 0 0.7734 1>})
  n66 = SkyLightEnvMapSample(Direction=ReflectionVectorWS(CustomWorldNormal=NamedRerouteDeclaration(None=@n65) {name=Glass Normal World, nodeColor=<rgba 1 0 0.7734 1>}))
  n67 = Multiply(
   A=Power(
    Base=NamedRerouteDeclaration(None=@n66) {name=ReflectionENV, nodeColor=<rgba 0 1 0.0156 1>}
    Exp=ScalarParameter() {defaultValue=0.2, group=Reflection, parameterName=ReflectonBrightnessExponent}
   ) {constExponent=0.2}
   B=LinearInterpolate(
    A=ScalarParameter() {defaultValue=0.05, group=Reflection, parameterName=Reflection Brightness}
    B=ScalarParameter() {defaultValue=0.05, group=Reflection, parameterName=Reflection Brightness Edge}
    Alpha=@n35
   )
  )
  Multiply(A=@n21, B=@n67)
  n68 = Distance(A=Constant3Vector(), B=NamedRerouteUsage())
  NamedRerouteDeclaration(None=@n68) {name=Dist, nodeColor=<rgba 1 0.3281 0 1>}
  n69 = VectorParameter() {
   group=Coordinates
   parameterName=Glass UV Remap - Back Bounds
   channelNames {a=Max Y, b=Max X, g=Min Y, r=Min X}
   defaultValue=<rgba 0 0 1 1>
  }
  n70 = VectorParameter() {
   group=Coordinates
   parameterName=Glass UV Remap - Front Bounds
   channelNames {a=Max Y, b=Max X, g=Min Y, r=Min X}
   defaultValue=<rgba 0 0 1 1>
  }
  n71 = NamedRerouteUsage()
  n72 = LinearInterpolate(
   A=MaterialFunctionCall(X=@n69.Min X, Y=@n69.Min X) {materialFunction=&MakeFloat2}.Result
   B=MaterialFunctionCall(X=@n70.Min X, Y=@n70.Min X) {materialFunction=&MakeFloat2}.Result
   Alpha=@n71
  )
  n73 = LinearInterpolate(
   A=MaterialFunctionCall(X=@n69.Max X, Y=@n69.Max X) {materialFunction=&MakeFloat2}.Result
   B=MaterialFunctionCall(X=@n70.Max X, Y=@n70.Max X) {materialFunction=&MakeFloat2}.Result
   Alpha=@n71
  )
  n74 = TextureCoordinate()
  NamedRerouteDeclaration(None=Frac(None=InverseLinearInterpolate(A=@n72, B=@n73, Value=@n74))) {name=Glass UVs, nodeColor=<rgba 0.4766 1 0 1>}
  n75 = Sign(None=ComponentMask(None=PreSkinnedNormal()) {g=true})
  NamedRerouteDeclaration(None=Saturate(None=VertexInterpolator(VS=@n75).PS)) {name=Glass Front/Back, nodeColor=<rgba 0.4062 0 1 1>}
  NamedRerouteDeclaration(None=@n39.Result) {name=Tint, nodeColor=<rgba 1 0 0.0469 1>}
  NamedRerouteDeclaration(None=ScalarParameter() {defaultValue=0.426, group=FPSShading, parameterName=ScopeEndRadius}) {name=ScopeRadius, nodeColor=<rgba 0 1 0.3438 1>}
  n76 = TransformPosition(None=VectorParameter() {parameterName=EyeSocketLocation}.RGB) {transformSourceType=TRANSFORMPOSSOURCE_Instance, transformType=TRANSFORMPOSSOURCE_Camera}
  n77 = ComponentMask(None=Normalize(VectorInput=NamedRerouteDeclaration(None=@n76) {name=Socket World Position, nodeColor=<rgba 0.1719 0 1 1>})) {r=true}
  NamedRerouteDeclaration(None=Saturate(None=InverseLinearInterpolate(Value=@n77) {constA=0.1, constB=0.2})) {name=Alignment, nodeColor=<rgba 1 0 0.2344 1>}
  NamedRerouteDeclaration(None=StaticSwitchParameter(True=Constant(), False=@n67) {group=Simple Glass, parameterName=Use FPS ComplexShading, sortPriority=5}) {name=Reflections, nodeColor=<rgba 1 0.2578 0 1>}
  n78 = ScalarParameter() {defaultValue=32, desc=Any distance below this number will return 0, parameterName=ZoomTransitionMinDistance}
  n79 = Saturate(None=Divide(
    A=Subtract(A=@n68, B=@n78) {constB=50}
    B=ScalarParameter() {defaultValue=40, desc=Any distance above this number will return 1, parameterName=ZoomTransitionMaxDistance}
   ) {constB=1000})
  NamedRerouteDeclaration(None=Max(A=NamedRerouteDeclaration(None=@n79) {name=ZoomTransition, nodeColor=<rgba 1 0.1641 0 1>}, B=NamedRerouteUsage())) {name=Opacity, nodeColor=<rgba 1 0 0.3516 1>}
  n80 = NamedRerouteUsage()
  NamedRerouteDeclaration(None=Multiply(A=OneMinus(None=@n80), B=ShadingPathSwitch(Default=@n24, Mobile=@n23.EyeAdaptationInverse) {desc=Migrated from FeatureLevelSwitch})) {name=EmissiveReflections, nodeColor=<rgba 0.4766 0 1 1>}
  n81 = MaterialFunctionCall(Float2=Add(A=ScreenPosition().ViewportUV) {constB=-0.5}) {materialFunction=&BreakOutFloat2Components}
  n82 = MaterialFunctionCall(Float2=MaterialFunctionCall() {materialFunction=&ScreenResolution}.Visible Resolution) {materialFunction=&BreakOutFloat2Components}
  n83 = MaterialFunctionCall(X=Multiply(A=@n81.R, B=Divide(A=@n82.R, B=@n82.R)), Y=@n81.G) {materialFunction=&MakeFloat2}
  NamedRerouteDeclaration(None=Add(A=NamedRerouteDeclaration(None=@n83.Result) {name=Screen Position Centered, nodeColor=<rgba 1 0.1 0.1 1>}) {constB=0.5}) {name=Screen Position Ratioed, nodeColor=<rgba 1 0.1 0.1 1>}
  ObjectPositionWS()
  ScalarParameter() {defaultValue=1, desc=0.797027, group=FPSShading, parameterName=TestScale}
  ScalarParameter() {defaultValue=0.1, parameterName=RoughTest}
  SetMaterialAttributes!(MaterialAttributes, Emissive Color=Constant3Vector()) {attributeSetTypes [#3]}
  Subtract(A=@n14, B=Divide(A=@n12) {constB=4})
  TransformPosition(None=@n1.RGB) {transformType=TRANSFORMPOSSOURCE_World}
  VectorParameter() {group=FPSShading, parameterName=OffsetScreenSample, defaultValue=<rgba 0 0 0 1>}
  ViewProperty()
 ]
}
