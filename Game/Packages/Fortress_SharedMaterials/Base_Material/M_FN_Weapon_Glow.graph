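# Unlit base material for Fortress weapon glow, judging by its name and shared-materials location.
# Its graph holds a single BaseColor vector parameter, white by default, with no textures, math or visible link to a material output.
# Usage flags enable it on skeletal meshes and with mesh deformers.
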Material /Game/Packages/Fortress_SharedMaterials/Base_Material/M_FN_Weapon_Glow
graph {
 properties {bUsedWithMeshDeformer=true, bUsedWithSkeletalMesh=true, pixelDepthOffsetMode=PDOM_Legacy, shadingModel=MSM_Unlit, usedShadingModels=Unlit}
 nodes [VectorParameter() {parameterName=BaseColor, defaultValue=<rgba 1 1 1 1>}]
}
