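# Material function for a vehicle res-in (spawn) dissolve, outputting a modulative opacity plus intermediate gradient and debug values.
# A world-position gradient along a directional vector, overridable by input, is noise-distorted and height-lerped against a rez-in percentage from custom primitive data.
# A second height lerp adds a tiled grid-square pattern, indexing an animation texture whose alpha, scaled by FirstSpawnPassComplete, becomes opacity.
# A weapon-local sphere mask is left unconnected.
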MaterialFunction /Game/Packages/Fortress_SharedMaterials/Base_Material_Functions/VehicleResInEffect
graph {nodes [
  n1 = VectorParameter() {parameterName=DissolveSphericalGradientHandLocationAndRadius, defaultValue=<rgba 0 0 0 1>}
  FunctionOutput(None=AppendVector(A=@n1.RGB, B=@n1.RGB)) {outputName=Vector4 debug}
  n2 = ScalarParameter() {bUseCustomPrimitiveData=true, defaultValue=1, group=SpawnEffect, parameterName=DissolveGradientPercentage, primitiveDataIndex=12}
  n3 = FunctionInput(Preview=@n2) {bUsePreviewValueAsDefault=true, inputName=Rez In Percentage, inputType=FunctionInput_Scalar}
  n4 = TextureCoordinate()
  n5 = ScalarParameter() {defaultValue=0.02, parameterName=GridDistortionAmount}
  n6 = Subtract(
   A=WorldPosition() {worldPositionShaderOffset=WPT_ExcludeAllShaderOffsets}.XYZ
   B=VectorParameter() {parameterName=DissolveGradientBaseLocation, defaultValue=<rgba 0 0 0 1>}.RGB
  )
  n7 = VectorParameter() {parameterName=DissolveGradientDirectionalVector, defaultValue=<rgba 1 0 0 1>}
  n8 = Clamp(None=Divide(A=DotProduct(A=@n6, B=@n7.RGB), B=ScalarParameter() {defaultValue=200, parameterName=DissolveGradientDistance}) {constB=30})
  n9 = Add(
   A=Multiply(A=TextureSample(UVs=@n4) {texture=&T_NoisePack_01}.R, B=@n5) {constB=-0.015}
   B=FunctionOutput(None=FunctionInput(Preview=@n8) {bUsePreviewValueAsDefault=true, inputName=Animation Gradient, inputType=FunctionInput_Scalar}) {outputName=AnimationGradient}
  )
  n10 = Add(A=@n5)
  n11 = MaterialFunctionCall(Transition Phase=Clamp(None=FunctionOutput(None=@n3) {outputName=Res In Percentage}), Height Texture=Clamp(None=Divide(A=@n9, B=@n10))) {materialFunction=&HeightLerp}
  n12 = TextureCoordinate()
  n13 = ScalarParameter() {defaultValue=3.4424, parameterName=texture pattern tiling}
  n14 = OneMinus(None=TextureSample(UVs=Multiply(A=@n12, B=@n13)) {samplerType=SAMPLERTYPE_LinearColor, texture=/Game/Packages/Fortress_SharedMaterials/Textures/T_Slowfield_GridSquares_01}.R)
  n15 = ScalarParameter() {defaultValue=0.2478, parameterName=PatternAnimationOffset}
  n16 = TextureSample(UVs=MaterialFunctionCall(
    Transition Phase=FunctionOutput(None=@n11.Lerp Alpha No Contrast) {outputName=Animation With Distortion}
    Height Texture=Multiply(A=@n14, B=@n15) {constB=0.2}
   ) {materialFunction=&HeightLerp}.Alpha) {samplerType=SAMPLERTYPE_LinearColor, texture=/Game/Packages/Fortnite_UniqueMaterials_1/Effects/ShieldPowerAnim}
  n17 = FunctionOutput(None=ScalarParameter() {defaultValue=1, parameterName=FirstSpawnPassComplete}) {outputName=FirstSpawnPassComplete}
  FunctionOutput(None=Clamp(None=Multiply(A=@n16.A, B=@n17))) {outputName=Modulative Opacity}
  SphereMask(
   A=@n1.RGB
   B=MaterialFunctionCall() {materialFunction=/Game/Packages/Fortnite_UniqueMaterials_1/Weapons/WeaponLocalPosition}.Weapon Local Position Pixel Shader
   Radius=@n1.RGB
  ) {hardnessPercent=0}
 ]}
