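# Masked master material for the vehicle derez and spawn dissolve, layering a block-pattern glow over a blue base function.
# At Epic quality, custom primitive data shifts a threshold along a pre-skinned directional gradient, exploding and smearing vertices outward.
# That gradient height-lerps with a scrolling world-aligned shield texture for dithered opacity and fake-chromatic edge emissive.
# Other quality levels skip the vertex offset and mask with VehicleResInEffect and panning noise.
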Material /Game/Packages/Fortress_SharedMaterials/Base_Material/VehicleDeRezShader
graph {
 properties {
  bUseMaterialAttributes=true
  bUsedWithMeshDeformer=true
  bUsedWithSkeletalMesh=true
  blendMode=BLEND_Masked
  pixelDepthOffsetMode=PDOM_Legacy
  usedShadingModels=Default Lit
 }
 nodes [
  n1 = GetMaterialAttributes(None=MaterialFunctionCall!(BaseMaterial, Rim) {materialFunction=/Game/Packages/Fortress_SharedMaterials/Base_Material_Functions/MF_VehicleDerez_BaseBlue}.Result) {attributeGetTypes [#1]}
  n2 = Constant() {r=1}
  n3 = Constant() {r=0.5}
  n4 = NamedRerouteUsage()
  n5 = Multiply(
   A=OneMinus(None=MaterialFunctionCall!(ThresholdValue=@n2, ThresholdRange=@n3, Alpha=@n4, TwoSidedThreshold) {materialFunction=/Engine/Functions/Engine_MaterialFunctions02/Masking/ThresholdWithRange}.Result)
   B=VectorParameter() {parameterName=BlocksColor, defaultValue=<rgba 0 0.1759 0.3 1>}.RGB
  )
  n6 = Power(
   Base=TextureSample(UVs=TextureCoordinate() {uTiling=3, vTiling=3}) {samplerSource=SSM_Wrap_WorldGroupSettings, texture=/Game/Athena/SafeZone/Textures/T_JanusShield3_M}.R
   Exp=ScalarParameter() {defaultValue=0.5, parameterName=BlocksContrast}
  ) {constExponent=0.5}
  n7 = SetMaterialAttributes(MaterialAttributes=@n1.MaterialAttributes, Emissive Color=Add(A=@n1.Emissive Color, B=Multiply(A=@n5, B=@n6))) {attributeSetTypes [#1]}
  n8 = GetMaterialAttributes(None=@n7) {attributeGetTypes [#2]}
  n9 = ObjectPositionWS()
  n10 = WorldPosition()
  n11 = ScalarParameter() {parameterName=HT_CrunchVerts}
  n12 = ScalarParameter() {bUseCustomPrimitiveData=true, group=SpawnEffect, parameterName=DissolveGradientPercentage, primitiveDataIndex=12}
  n13 = LinearInterpolate(
   B=ScalarParameter() {defaultValue=1.5, parameterName=ThresholdLeadingValue}
   Alpha=OneMinus(None=NamedRerouteDeclaration(None=@n12) {name=Dissolve Control, nodeColor=<rgba 1 0 0.375 1>})
  ) {constB=1.2}
  n14 = PreSkinnedPosition()
  n15 = VectorParameter() {parameterName=DissolveGradientDirectionalVector, defaultValue=<rgba 1 0 0 1>}
  n16 = Clamp(None=Divide(A=DotProduct(A=@n14, B=@n15.RGB), B=ScalarParameter() {defaultValue=400, parameterName=DissolveGradientDistance}) {constB=30})
  n17 = OneMinus(None=MaterialFunctionCall(
    ThresholdValue=@n13
    ThresholdRange=ScalarParameter() {defaultValue=1, parameterName=ThresholdRange}
    Alpha=@n16
    TwoSidedThreshold=StaticBool()
   ) {materialFunction=&ThresholdWithRange}.Result)
  n18 = Multiply(A=VertexNormalWS(), B=ScalarParameter() {defaultValue=20, parameterName=ExplodeIntensity})
  n19 = Multiply(A=VectorParameter() {parameterName=ExplodeSmear, defaultValue=<rgba 100 0 0 1>}.RGB, B=@n17)
  n20 = MaterialFunctionCall!(WPO=Multiply(A=@n17, B=Add(A=@n18, B=@n19))) {materialFunction=/Game/Packages/Fortress_SharedMaterials/Base_Material_Functions/Fortnite_HideElementsWithVertexColor}
  n21 = Add(A=Multiply(A=Subtract(A=@n9, B=@n10.XYZ), B=@n11), B=Multiply(A=OneMinus(None=@n11), B=Add(A=@n8.World Position Offset, B=@n20.Result)))
  n22 = VertexInterpolator(VS=@n14)
  n23 = Multiply(A=Multiply(A=Time(), B=ScalarParameter() {defaultValue=40, parameterName=Speed}), B=Constant3Vector() {constant=<rgba -1 0 0 0>})
  n24 = ComponentMask(None=MaterialFunctionCall(
    TextureObject=TextureObject() {texture=&T_JanusShield3_M}
    TextureSize=ScalarParameter() {defaultValue=150, parameterName=Size}
    WorldPosition=Add(A=@n22.PS, B=@n23)
   ) {materialFunction=&WorldAlignedTexture}.XY Texture) {r=true}
  n25 = MaterialFunctionCall(
   Transition Phase=MaterialFunctionCall(
    ThresholdValue=@n13
    ThresholdRange=ScalarParameter() {defaultValue=1, parameterName=ThresholdRange}
    Alpha=VertexInterpolator(VS=@n16).PS
    TwoSidedThreshold=StaticBool()
   ) {materialFunction=&ThresholdWithRange}.Result
   Height Texture=@n24
  ) {materialFunction=&HeightLerp}
  n26 = MaterialFunctionCall(In=@n25.Lerp Alpha No Contrast, Contrast=ScalarParameter() {defaultValue=0.5, parameterName=OpacityMaskingContrast}) {materialFunction=&CheapContrast}
  n27 = GetMaterialAttributes(None=SetMaterialAttributes!(
    MaterialAttributes=SetMaterialAttributes!(MaterialAttributes=@n8.MaterialAttributes, World Position Offset=@n21) {attributeSetTypes [#2]}
    Opacity Mask=MaterialFunctionCall(Alpha Threshold=@n26.Result) {materialFunction=&DitherTemporalAA}.Result
   ) {attributeSetTypes [#3]}) {attributeGetTypes [#1]}
  n28 = Multiply(A=@n27.Emissive Color, B=@n25.Lerp Alpha No Contrast)
  n29 = MaterialFunctionCall(Float3=@n22.PS) {materialFunction=&BreakOutFloat3Components}
  n30 = ScalarParameter() {defaultValue=0.01, parameterName=FakeChromaticLookZShift}
  n31 = Add(A=Add(A=@n24, B=Multiply(A=@n29.R, B=@n30)), B=Multiply(A=Time()) {constB=-1})
  n32 = Multiply(
   A=MaterialFunctionCall(
    ThresholdValue=ScalarParameter() {defaultValue=0.2, parameterName=FakeChromaticMaskThreshold}
    ThresholdRange=ScalarParameter() {defaultValue=0.8, parameterName=FakeChromaticMaskRange}
    Alpha=@n25.Lerp Alpha No Contrast
    TwoSidedThreshold=StaticBool() {value=true}
   ) {materialFunction=&ThresholdWithRange}.Result
   B=Multiply(
    A=TextureSample(UVs=@n31) {texture=&T_Inky_Smoke_Tile_Chromatic}.RGB
    B=ScalarParameter() {defaultValue=20, parameterName=FakeChromaticIntensity}
   )
  )
  n33 = TextureSample(UVs=Panner(Coordinate=TextureCoordinate() {uTiling=1.5, vTiling=1.5}) {speedY=0.1}) {samplerType=SAMPLERTYPE_LinearColor, texture=/Game/Packages/Fortress_Sky/Materials/Noise32x322_nearestNeighboor}
  n34 = MaterialFunctionCall!(Animation Gradient, Rez In Percentage) {materialFunction=/Game/Packages/Fortress_SharedMaterials/Base_Material_Functions/VehicleResInEffect}
  n35 = Add(A=Multiply(A=@n33.G, B=@n33.G), B=LinearInterpolate(Alpha=@n34.Animation With Distortion) {constA=-1, constB=2})
  n36 = SetMaterialAttributes!(MaterialAttributes=@n7, Opacity Mask=Multiply(A=Clamp(None=@n35), B=@n34.Animation With Distortion)) {attributeSetTypes [#3]}
  QualitySwitch(
   Default=@n36
   Low=@n36
   High=@n36
   Medium=@n36
   Epic=SetMaterialAttributes!(MaterialAttributes=@n27.MaterialAttributes, Emissive Color=Add(A=@n28, B=@n32)) {attributeSetTypes [#1]}
  )
  SphereMask(
   A=MaterialFunctionCall(Float3=@n14) {materialFunction=&BreakOutFloat3Components}.R
   B=Constant()
   Radius=ScalarParameter() {defaultValue=180, parameterName=dist}
   Hardness=Constant()
  )
 ]
}
