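# Material function that cuts a material away along a UV axis by overriding its opacity mask.
# A selectable texture coordinate channel, flipped by a static switch, is offset by a Fade-driven lerp between min and max, then floored to make a hard moving edge.
# That edge is multiplied by an input mask, green vertex color by default, and blends opacity mask toward zero through MatLayerBlend_OverrideOpacityMask.
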MaterialFunction /Game/Packages/Fortress_SharedMaterials/Base_Material/MF_Fade_Along_UV
graph {nodes [
  n1 = StaticComponentMaskParameter(None=TextureCoordinate()) {defaultR=true, group=Fade Along UV, parameterName=UVFadeMask}
  n2 = OneMinus(None=@n1)
  n3 = StaticBoolParameter() {group=Fade Along UV, parameterName=Invert Mask Fade Direction, sortPriority=4}
  n4 = ScalarParameter() {defaultValue=1, group=Fade Along UV, parameterName=U_Fade_Min, sortPriority=1}
  n5 = ScalarParameter() {group=Fade Along UV, parameterName=U_Fade_Max, sortPriority=2}
  n6 = ScalarParameter() {group=Fade Along UV, parameterName=Fade, sortPriority=3}
  n7 = Floor(None=Add(A=StaticSwitch(True=@n1, False=@n2, Value=@n3), B=LinearInterpolate(A=@n4, B=@n5, Alpha=@n6)))
  n8 = FunctionInput(Preview=VertexColor().G) {
   bUsePreviewValueAsDefault=true
   description=By Default Green Vertex Color
   inputName=Mask for Fade
   inputType=FunctionInput_Scalar
   sortPriority=1
  }
  FunctionOutput(None=MaterialFunctionCall!(
    Material=FunctionInput() {inputName=Material, inputType=FunctionInput_MaterialAttributes}
    NewOpacityMask=Constant()
    Mask=Multiply(A=@n7, B=@n8)
   ) {materialFunction=/Engine/Functions/MaterialLayerFunctions/MatLayerBlend_OverrideOpacityMask}.Blended Material) {outputName=Fade in OpacityMask}
 ]}
